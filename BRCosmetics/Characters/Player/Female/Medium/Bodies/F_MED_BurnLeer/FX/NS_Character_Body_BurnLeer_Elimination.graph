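# Niagara system for the BurnLeer character body's elimination effect: a burst of rising embers plus six-second body flames that fade in and out.
# Particle partitions index arrays of bones or sockets, offsets and scales, so mesh and sprite flames follow the skeletal mesh.
# Mesh flames pass smoothed lagging positions to dynamic or custom particle parameters.
# User flags fade selected partitions when hands are hidden and set front-end preview opacity on one fire emitter.
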NiagaraSystem /BRCosmetics/Characters/Player/Female/Medium/Bodies/F_MED_BurnLeer/FX/NS_Character_Body_BurnLeer_Elimination
dependencies {
 usedDataInterfaces [
  &NiagaraDataInterfaceArrayFloat3
  &NiagaraDataInterfaceArrayInt32
  &NiagaraDataInterfaceArrayFloat
  &NiagaraDataInterfaceCurve
  &NiagaraDataInterfaceSkeletalMesh
 ]
 usedDynamicInputs [
  &SetBoolByFloatComparison
  &Boolean_AndOperation
  &SetBoolByIntComparison
 ]
 usedModules [
  &self:SystemSpawnScript.SystemScriptSource.SystemScriptGraph.NiagaraNodeAssignment_0.SetVariables_#1
  &self:SystemSpawnScript.SystemScriptSource.SystemScriptGraph.NiagaraNodeAssignment_1.SetVariables_#2
  &SystemState
  /Game/Effects/Niagara/Modules/Logic/UpdateValueStateChange
 ]
}
emitters {
 NE_Burst_Embers {
  emitterUpdateScript [
   &EmitterState()
   &SpawnBurst_Instantaneous(Spawn Count=24, Use Spawn Probability=false, UseLoopCountLimit=false)
  ]
  particleSpawnScript [
   &PartitionParticles(Number Of Partitions=12)
   &InitializeParticle(
    Color Mode=::NewEnumerator0
    Lifetime Max=0.9
    Lifetime Min=0.5
    Lifetime Mode=::NewEnumerator1
    Sprite Rotation Mode=::NewEnumerator2
    Sprite Size Mode=::NewEnumerator1
    Uniform Sprite Size Max=15
    UsePositionOffset=false
    Write Material Random=true
   )
   &AddVelocity(
    Module.Use Falloff Distance=false
    Override Seed=false
    Velocity Mode=::NewEnumerator1
    Velocity Origin=$Particles.SkeletalMeshLocation.SampledPosition
    Velocity Speed=&RandomRangeFloat(Minimum=25, Maximum=150)
   )
  ]
  particleUpdateScript [
   &ParticleState()
   &SkeletalMeshLocation(
    Bone / Socket Index=&SelectIntFromArray(
     Direct Array Index=$Particles.PartitionParticles.Partition
     Int Selection Array=DataInterface() {propertyValues {IntData [20,21,1,2,1,2,5,6,5,6,20,21]}}
    )
    Bone Sampling Mode=::NewEnumerator4
    MaskSpawnGroup=false
    Position Sampling=::NewEnumerator1
    Skeletal Mesh=$System.Skeletal Mesh Reader
   )
   &ScaleSpriteSize(Uniform Curve Sprite Scale=DataInterface() {propertyValues {Curve {
       defaultValue=0
       keys [
        {arriveTangent=9.0725, interpMode=RCIM_Cubic, leaveTangent=9.0725, tangentMode=RCTM_User, time=0, value=0}
        {arriveTangent=-2.5249, interpMode=RCIM_Cubic, leaveTangent=-2.5249, tangentMode=RCTM_User, time=0.1408, value=0.9847}
        {arriveTangent=-0.0405, interpMode=RCIM_Cubic, leaveTangent=-0.0405, tangentMode=RCTM_User, time=1, value=0}
       ]
      }}})
   &Drag(Drag=10, Use Linear Drag=true, Use Rotational Drag=true)
   &GravityForce(Gravity=&RandomRangeVector(Minimum=<0 0 250>, Maximum=<0 1 500>))
   &CurlNoiseForce(
    Bias Noise=false
    Noise Frequency=20
    Noise Strength=&RandomRangeFloat(Minimum=50, Maximum=150)
    Pan Noise=false
   )
   &SolveForcesAndVelocity(Clamp Velocity=false, Limit Acceleration=false)
   &DynamicMaterialParameters(
    Index 0 Param 1=&FloatFromCurve(FloatCurve=DataInterface() {propertyValues {Curve {
        defaultValue=1
        keys [{interpMode=RCIM_Cubic, tangentMode=RCTM_User, time=0, value=0}, {interpMode=RCIM_Cubic, tangentMode=RCTM_User, time=1.0018, value=0.5073}]
       }}})
    Index 0 Param 2=&RandomRangeFloat(Minimum=0, Maximum=1)
    Index 0 Param 3=&RandomRangeFloat(Minimum=0, Maximum=1)
    Module.Index0Float0Write=true
    Module.Index0Float1Write=true
    Module.Index0Float2Write=true
    Module.Index0Float3Write=true
   )
   &ScaleColor(
    Scale Alpha=&FloatFromCurve(
     FloatCurve=DataInterface() {propertyValues {Curve {keys [<at 0 0>,<at 0.5 1>,<at 1 0>]}}}
     Scale Curve=&Lerp_Float(
      A=1
      B=&SelectFloatFromArray(
       Direct Array Index=$Particles.PartitionParticles.Partition
       Float Selection Array=DataInterface() {propertyValues {FloatData [1,1,1,1,1,1,0,0,0,0,1,1]}}
      )
      Alpha=&MakeCustomFloatFromBool(Bool=$User.bisHandHidden, True Float=1, False Float=0)
     )
    )
    ScaleA=true
    ScaleRGB=true
   )
   &CameraOffset(Camera Offset Amount=5)
  ]
  rendererClasses [&NiagaraSpriteRendererProperties]
  renderers [{rendererClass=&NiagaraSpriteRendererProperties}]
 }
 NE_Fire_01 {
  emitterUpdateScript [
   &EmitterState(
    Inactive Response=::NewEnumerator1
    Life Cycle Mode=::NewEnumerator1
    Loop Behavior=::NewEnumerator1
    Loop Delay=0.05
    Loop Duration=6
    UseLoopDelay=true
   )
   &SpawnBurst_Instantaneous(Spawn Count=12, Use Spawn Probability=false, UseLoopCountLimit=false)
  ]
  particleSpawnScript [
   &PartitionParticles(Number Of Partitions=12)
   &InitializeParticle(
    Color Maximum=<rgba 1 1 1 0.2>
    Color Minimum=<rgba 1 1 1 0.1>
    Color Mode=::NewEnumerator2
    Lifetime=6
    Mesh Scale=&Multiply_Vector(
     A=&SelectVectorFromArray(
      Direct Array Index=$Particles.PartitionParticles.Partition
      Vector Selection Array=DataInterface() {propertyValues {FloatData [
         <0.8 0.4 15>
         <0.8 0.4 5>
         <0.5 0.8 20>
         <0.5 0.8 -40>
         <0.7 1.2 30>
         <0.6 1.2 -60>
         <0.6 1 30>
         <0.6 1 -40>
         <0.6 0.7 30>
         <0.6 0.75 -30>
         <0.6 0.8 1>
         <0.6 0.8 1>
        ]}}
     )
     B=<0.95 0.95 1>
    )
    Mesh Scale Mode=::NewEnumerator4
    UsePositionOffset=false
    Write Material Random=true
   )
   &InitialMeshOrientation(
    Mesh Orientation Mode=::NewEnumerator2
    Rotation=&SelectVectorFromArray(
     Direct Array Index=$Particles.PartitionParticles.Partition
     Vector Selection Array=DataInterface() {propertyValues {FloatData [
        <0 0 0.25>
        <0 0 0.25>
        <0 0.05 0.25>
        <0 0 0.25>
        <0 -0.035 0.25>
        <0 0 0.25>
        <0 0 0.25>
        <0 0 0.25>
        <0 0 0.25>
        <0 0 0.25>
        <0 0 0.25>
        <0 0 0.25>
       ]}}
    )
    Use Rotation Vector=true
   )
  ]
  particleUpdateScript [
   &ParticleState()
   &SkeletalMeshLocation(
    Bone / Socket Index=&SelectIntFromArray(
     Direct Array Index=$Particles.PartitionParticles.Partition
     Int Selection Array=DataInterface() {propertyValues {IntData [20,21,1,2,1,2,5,6,5,6,20,21]}}
    )
    Bone Sampling Mode=::NewEnumerator4
    MaskSpawnGroup=false
    Sampled Position Offset=&SelectVectorFromArray(
     Direct Array Index=$Particles.PartitionParticles.Partition
     Vector Selection Array=DataInterface() {propertyValues {FloatData [<11 0 0>, <-10 0 -2>, <5 0 2>, <-5 0 0>, <17 0 2>, <-15 0 0>, <6 0 2>, <-6 0 -2>, <18 0 0>, <-18 0 0>, <5 0 2>, <-5 0 -2>]}}
    )
    Skeletal Mesh=$System.Skeletal Mesh Reader
   )
   &DynamicMaterialParameters(
    Index 0 Param 1=&RandomRangeFloat(Minimum=0.4, Maximum=0.8)
    Index 0 Param 2=&RandomRangeFloat(Minimum=0.4, Maximum=0.8)
    Index 0 Param 4=&RandomRangeFloat(Minimum=1, Maximum=1.3)
    Index 2 Param 1=&RandomRangeFloat(Minimum=0.8, Maximum=1)
    Index 2 Param 2=&Multiply_Float(
     A=&SelectFloatFromArray(
      Direct Array Index=$Particles.PartitionParticles.Partition
      Float Selection Array=DataInterface() {propertyValues {FloatData [1.3,1.5,0.5,0.8,1,0.8,0.85,0.8,0.75,0.6,1.5,1.5]}}
     )
     B=0.65
    )
    Module.Index0Float0Write=true
    Module.Index0Float1Write=true
    Module.Index0Float2Write=true
    Module.Index0Float3Write=true
    Module.Index2Float0Write=true
    Module.Index2Float1Write=true
    Module.Index2Float2Write=true
    Module.Index2Float3Write=true
    Write Parameter Index 2=::NewEnumerator1
   )
   &InterpolateOverTime(Rate Of Change=5, Target Vector Value:NiagaraPosition=$Particles.Position, Variable Type=::NewEnumerator2)
   /SetVariables(Particles.Interpolate Dynamic=&MakeFloat4FromVectorAndFloat("Vector (RGB)"=$Particles.InterpolateOverTime.InterpolateOverTime_V3_.Moving Average, "Float (Alpha)"=0.4)) {isSetParameters=true}
   &ScaleColor(
    Scale Alpha=&MakeCustomFloatFromBool(
     Bool=$User.bIsFrontEndPreview
     True Float=0.3
     False Float=&FloatFromCurve(
      FloatCurve=DataInterface() {propertyValues {Curve {keys [<at 0 0>,<at 0.1 1>,<at 0.5 1>,<at 1 0>]}}}
      Scale Curve=&Lerp_Float(
       A=0.05
       B=&SelectFloatFromArray(
        Direct Array Index=$Particles.PartitionParticles.Partition
        Float Selection Array=DataInterface() {propertyValues {FloatData [0.05,0.05,0.05,0.05,0.05,0.05,0,0,0,0,0.05,0.05]}}
       )
       Alpha=&MakeCustomFloatFromBool(Bool=$User.bisHandHidden, True Float=1, False Float=0)
      )
     )
    )
    Scale RGB=&MakeVectorFromLinearColorRGB(Color=<rgba 1 1 1 0.5>)
    ScaleA=true
    ScaleRGB=true
   )
   &CameraOffset(Camera Offset Amount=-10)
  ]
  rendererClasses [&NiagaraMeshRendererProperties]
  renderers [{rendererClass=&NiagaraMeshRendererProperties, rendererIndex=0}]
 }
 NE_Fire_02 {
  emitterUpdateScript [
   &EmitterState(
    Inactive Response=::NewEnumerator1
    Life Cycle Mode=::NewEnumerator1
    Loop Behavior=::NewEnumerator1
    Loop Delay=0.05
    Loop Duration=6
    UseLoopDelay=true
   )
   &SpawnBurst_Instantaneous(Spawn Count=10, Use Spawn Probability=false, UseLoopCountLimit=false)
  ]
  particleSpawnScript [
   &PartitionParticles(Number Of Partitions=10)
   &InitializeParticle(
    Lifetime=6
    Mesh Scale=&Multiply_VectorByFloat(Vector=<0.3 1 0.3>, Float=&RandomRangeFloat(Minimum=0.7, Maximum=0.8))
    Mesh Scale Mode=::NewEnumerator4
    Sprite Rotation Angle=-0.25
    Sprite Rotation Mode=::NewEnumerator1
    Sprite Size Mode=::NewEnumerator3
    Uniform Sprite Size=&Multiply_Float(
     A=&SelectFloatFromArray(
      Direct Array Index=$Particles.PartitionParticles.Partition
      Float Selection Array=DataInterface() {propertyValues {FloatData [25,25,20,20,20,20,18,18,20,20]}}
     )
     B=0.8
    )
    UsePositionOffset=false
    Write Material Random=true
   )
   /SetVariables(Particles.Random=&RandomRangeFloat(Minimum=0, Maximum=1)) {isSetParameters=true}
   &InitialMeshOrientation(Mesh Orientation Mode=::NewEnumerator2, Rotation=<0 0 0.25>, Use Rotation Vector=true)
   &InitialMeshOrientation(Mesh Orientation Mode=::NewEnumerator2, Rotation=<0.25 0 0>, Use Rotation Vector=true) {module=InitialMeshOrientation001}
   &CameraOffset(Camera Offset Amount=&Multiply_Float(
     A=&SelectFloatFromArray(
      Direct Array Index=$Particles.PartitionParticles.Partition
      Float Selection Array=DataInterface() {propertyValues {FloatData [8,8,8,8,5,2,1,1,1,1]}}
     )
     B=0.7
    ))
  ]
  particleUpdateScript [
   &ParticleState()
   &SkeletalMeshLocation(
    Bone / Socket Index=&SelectIntFromArray(
     Direct Array Index=$Particles.PartitionParticles.Partition
     Int Selection Array=DataInterface() {propertyValues {IntData [1,2,1,2,5,5,7,8,6,6]}}
    )
    Bone Sampling Mode=::NewEnumerator4
    MaskSpawnGroup=false
    Sampled Position Offset=&SelectVectorFromArray(
     Direct Array Index=$Particles.PartitionParticles.Partition
     Vector Selection Array=DataInterface() {propertyValues {FloatData [<0 0 0>, <0 0 0>, <13 0 0>, <-13 0 0>, <0 0 0>, <0 0 0>, <0 0 0>, <0 0 0>, <0 0 0>, <0 0 0>]}}
    )
    Skeletal Mesh=$System.Skeletal Mesh Reader
   )
   &SolveForcesAndVelocity(Clamp Velocity=false, Limit Acceleration=false)
   &ScaleColor(
    Scale Alpha=&FloatFromCurve(
     FloatCurve=DataInterface() {propertyValues {Curve {keys [<at 0 0>,<at 0.1 0.5>,<at 0.5 0.5>,<at 1 0>]}}}
     Scale Curve=&Lerp_Float(
      A=0.5
      B=&SelectFloatFromArray(
       Direct Array Index=$Particles.PartitionParticles.Partition
       Float Selection Array=DataInterface() {propertyValues {FloatData [0.5,0.5,0.5,0.5,0,0,0,0,0,0]}}
      )
      Alpha=&MakeCustomFloatFromBool(Bool=$User.bisHandHidden, True Float=1, False Float=0)
     )
    )
    Scale RGB=&MakeVectorFromLinearColorRGB(Color=&RandomRangeLinearColor(Minimum=<rgba 1 0.4277 0.1651 1>, Maximum=<rgba 1 0.6637 0.2813 1>))
    ScaleA=true
    ScaleRGB=true
   )
  ]
  rendererClasses [&NiagaraSpriteRendererProperties]
  renderers [{rendererClass=&NiagaraSpriteRendererProperties}]
 }
 NE_Fire_Spline_01 {
  emitterUpdateScript [
   &EmitterState(
    Inactive Response=::NewEnumerator1
    Life Cycle Mode=::NewEnumerator1
    Loop Behavior=::NewEnumerator1
    Loop Delay=0.05
    Loop Duration=6
    UseLoopDelay=true
   )
   &SpawnBurst_Instantaneous(Spawn Count=8, Use Spawn Probability=false, UseLoopCountLimit=false)
  ]
  particleSpawnScript [
   &PartitionParticles(Number Of Partitions=8)
   &InitializeParticle(
    Color=<rgba 2 1.2852 0 0.05>
    Lifetime=6
    Mesh Scale=&MergeHelper_VectorSwitch(
     0=<-0.36 1 0.5>
     1=<0.36 1 0.3>
     2=<-0.36 1 0.3>
     3=<-0.36 1 0.15>
     4=<0.36 1 0.3>
     5=<-0.36 1 0.3>
     6=<0.36 1 0.3>
     7=<0.36 1 0.3>
     Index=$Particles.PartitionParticles.Partition
     Num=::NewEnumerator6
    )
    Mesh Scale Mode=::NewEnumerator4
    UsePositionOffset=false
    Write Material Random=true
   )
   &InitialMeshOrientation(Mesh Orientation Mode=::NewEnumerator2, Rotation=<-0.25 0 0>, Use Rotation Vector=true)
   /SetVariables(Particles.FireColorOutside:LinearColor=<rgba 1 0.2 0 0.05>, Particles.FireColorOutsideLifetime=6) {isSetParameters=true, module=SetVariables_#3}
   &InitialMeshOrientation(
    Mesh Orientation Mode=::NewEnumerator2
    Rotation=&MergeHelper_VectorSwitch(
     0=<0.05 0 0.3>
     1=<0.05 0 0.3>
     2=<0.05 0 0.2>
     3=<0 0 0.2>
     4=<0 0 0.3>
     5=<0 0 0.2>
     6=<0 0 0.25>
     7=<0 0 0.4>
     Index=$Particles.PartitionParticles.Partition
     Num=::NewEnumerator6
    )
    Use Rotation Vector=true
   ) {module=InitialMeshOrientation001}
  ]
  particleUpdateScript [
   &CameraOffset(Camera Offset Amount=5)
   &ParticleState()
   &Timeline(Play Rate=1.7, Timeline End=10000000)
   &SkeletalMeshLocation(
    Bone / Socket Index=&SelectIntFromArray(Direct Array Index=$Particles.PartitionParticles.Partition, Int Selection Array=DataInterface() {propertyValues {IntData [0,1,2,3,4,5,0,0]}})
    MaskSpawnGroup=false
    Mesh Sampling Type=::NewEnumerator5
    Sampled Position Offset=&MergeHelper_VectorSwitch(
     0=<5 0 0>
     1=<5 0 0>
     2=<5 0 0>
     3=<5 0 0>
     4=<5 0 0>
     5=<5 0 0>
     6=<5 0 0>
     7=<5 -4 0>
     Index=$Particles.PartitionParticles.Partition
     Num=::NewEnumerator6
    )
    Skeletal Mesh=$System.Skeletal Mesh Reader
    Socket Sampling Mode=::NewEnumerator5
   )
   &ScaleMeshSize(Scale Factor=&Multiply_Vector(
     A=&SelectVectorFromArray(
      Direct Array Index=$Particles.PartitionParticles.Partition
      Vector Selection Array=DataInterface() {propertyValues {FloatData [<0.2 0.2 0.22>, <0.2 0.2 0.22>, <0.2 0.2 0.22>, <0.2 0.2 0.22>, <0.2 0.2 0.22>, <0.2 0.2 0.22>, <0.2 0.2 0.22>, <0.2 0.2 0.22>]}}
     )
     B=&Lerp_Vector(
      A=&Lerp_Vector(A=<1 1 1>, B=<1 1 1>, Alpha=$System.InterpolateOverTime001.InterpolateOverTime_Float.Moving Average)
      B=<0.8 0.8 0.8>
      Alpha=$System.InterpolateOverTime.InterpolateOverTime_Float.Moving Average
     )
    ))
   &ScaleMeshSize(Scale Factor=&FN_FrontEndVector3Modifier(Gameplay=<0.8 0.8 1>, Front End=<0.8 0.8 1>)) {module=ScaleMeshSize001}
   &ScaleColor(
    Scale Alpha=&FloatFromCurve(FloatCurve=DataInterface() {propertyValues {Curve {keys [<at 0 0>,<at 0.1 2>,<at 0.5 2>,<at 1 0>]}}}, Scale Curve=0.5)
    Scale RGB=<0.4 0.4 0.4>
    ScaleA=true
    ScaleRGB=true
   )
   /SetVariables(Particles.UpdateFireColorOutsideAge:NiagaraFloat=&Divide_Float(A=$Particles.Age, B=$Particles.Initial.FireColorOutsideLifetime)) {isSetParameters=true, module=SetVariables_#4}
   /SetVariables(Particles.UpdateFireColorOutside:LinearColor=&MakeLinearColorFromVectorAndFloat(
     "Vector (RGB)"=&MakeVectorFromLinearColorRGB(Color=$Particles.Initial.FireColorOutside)
     "Float (Alpha)"=&FloatFromCurve(
      FloatCurve=DataInterface() {propertyValues {Curve {keys [<at 0 0>,<at 0.1 1>,<at 0.5 1>,<at 1 0>]}}}
      CurveIndex=$Particles.UpdateFireColorOutsideAge
      Scale Curve=0.5
     )
    )) {isSetParameters=true, module=SetVariables_#5}
   &InterpolateOverTime(
    Rate Of Change:NiagaraFloat=$System.Interpolate Rate
    Target Vector Value:NiagaraPosition=$Particles.Position
    Variable Type=::NewEnumerator2
   )
   /SetVariables(Particles.Interpolate DY:Vector4f=&MakeFloat4FromVectorAndFloat("Vector (RGB)"=$Particles.InterpolateOverTime.InterpolateOverTime_V3_.Moving Average, "Float (Alpha)"=$System.Interpolate Internsity)) {isSetParameters=true, module=SetVariables_#6}
   &DynamicMaterialParameters(
    Index 0 Param 1=0.3
    Index 0 Param 2=$Particles.Timeline.Time
    Index 0 Param 3=&FN_FrontEndFloatModifier(Gameplay=10, Front End=5)
    Index 0 Param 4=0.1
    Index 2 Param 1=0.7
    Module.Index0Param0Write=true
    Module.Index0Param1Write=true
    Module.Index0Param2Write=true
    Module.Index0Param3Write=true
    Module.Index2Param0Write=true
    Module.Index2Param1Write=true
    Module.Index2Param2Write=true
    Module.Index2Param3Write=true
    Write Parameter Index 2=true
   )
  ]
  rendererClasses [&NiagaraMeshRendererProperties]
  renderers [
   {rendererClass=&NiagaraMeshRendererProperties, rendererIndex=0}
   {rendererClass=&NiagaraMeshRendererProperties, rendererIndex=1}
   {rendererClass=&NiagaraMeshRendererProperties, rendererIndex=2}
  ]
 }
 NE_Fire_Spline_02 {
  emitterUpdateScript [
   &EmitterState(
    Inactive Response=::NewEnumerator1
    Life Cycle Mode=::NewEnumerator1
    Loop Behavior=::NewEnumerator1
    Loop Delay=0.05
    Loop Duration=6
    UseLoopDelay=true
   )
   &SpawnBurst_Instantaneous(Spawn Count=12, Use Spawn Probability=false, UseLoopCountLimit=false)
  ]
  particleSpawnScript [
   &PartitionParticles(Number Of Partitions=12)
   &InitializeParticle(
    Color=<rgba 2 1.2852 0 0.05>
    Lifetime=6
    Mesh Scale=&SelectVectorFromArray(
     Direct Array Index=0
     Vector Selection Array=DataInterface() {propertyValues {FloatData [
        <-0.36 1 0.55>
        <-0.36 1 0.55>
        <0.36 1 0.65>
        <-0.36 1 0.65>
        <-0.36 1 0.75>
        <-0.36 1 0.75>
        <-0.36 1 0.55>
        <-0.36 1 0.55>
        <-0.36 1 0.75>
        <-0.36 1 0.75>
        <-0.36 1 0.65>
        <-0.36 1 0.65>
       ]}}
    )
    Mesh Scale Mode=::NewEnumerator4
    UsePositionOffset=false
    Write Material Random=true
   )
   &InitialMeshOrientation(Mesh Orientation Mode=::NewEnumerator2, Use Rotation Vector=true)
   /SetVariables(Particles.FireColorOutside:LinearColor=<rgba 1 0.2 0 0.05>, Particles.FireColorOutsideLifetime=6) {isSetParameters=true, module=SetVariables_#3}
   &InitialMeshOrientation(
    Mesh Orientation Mode=::NewEnumerator2
    Rotation=&SelectVectorFromArray(
     Direct Array Index=$Particles.PartitionParticles.Partition
     Vector Selection Array=DataInterface() {propertyValues {FloatData [
        <0 0 0>
        <0 0.5 0>
        <0 -0.1 -0.2>
        <0 0.4 -0.1>
        <0 -0.1 -0.2>
        <0 0.4 -0.1>
        <-0.2 0 0>
        <0.2 0.3 -0.2>
        <-0.2 0 0>
        <0.2 0.3 -0.2>
        <-0.1 0 0>
        <0 0.5 0>
       ]}}
    )
    Use Rotation Vector=true
   ) {module=InitialMeshOrientation001}
  ]
  particleUpdateScript [
   &CameraOffset()
   &ParticleState()
   &Timeline(Play Rate=1.7, Timeline End=10000000)
   &SkeletalMeshLocation(
    Bone / Socket Index=&SelectIntFromArray(
     Direct Array Index=$Particles.PartitionParticles.Partition
     Int Selection Array=DataInterface() {propertyValues {IntData [20,21,1,2,1,2,5,6,5,6,20,21]}}
    )
    Bone Sampling Mode=::NewEnumerator4
    MaskSpawnGroup=false
    Sampled Position Offset=&SelectVectorFromArray(
     Direct Array Index=$Particles.PartitionParticles.Partition
     Vector Selection Array=DataInterface() {propertyValues {FloatData [<15 0 0>, <-12 0 -2>, <5 0 3>, <-5 0 0>, <17 0 2>, <-15 0 0>, <6 0 2>, <-6 0 -2>, <18 0 0>, <-18 0 0>, <10 0 5>, <-5 0 -4>]}}
    )
    Skeletal Mesh=$System.Skeletal Mesh Reader
   )
   &ScaleMeshSize(Scale Factor=&Multiply_Vector(
     A=&SelectVectorFromArray(
      Direct Array Index=$Particles.PartitionParticles.Partition
      Vector Selection Array=DataInterface() {propertyValues {FloatData [
         <0.2 0.2 0.22>
         <0.2 0.2 0.22>
         <0.2 0.2 0.22>
         <0.2 0.2 0.22>
         <0.2 0.2 0.22>
         <0.2 0.2 0.22>
         <0.2 0.2 0.22>
         <0.2 0.2 0.22>
         <0.2 0.2 0.22>
         <0.2 0.2 0.22>
         <0.2 0.2 0.22>
         <0.2 0.2 0.22>
        ]}}
     )
     B=&Lerp_Vector(
      A=&Lerp_Vector(A=<1 1 1>, B=<1 1 1>, Alpha=$System.InterpolateOverTime001.InterpolateOverTime_Float.Moving Average)
      B=<0.8 0.8 0.8>
      Alpha=$System.InterpolateOverTime.InterpolateOverTime_Float.Moving Average
     )
    ))
   &ScaleMeshSize(Scale Factor=&FN_FrontEndVector3Modifier(Gameplay=<0.8 0.8 1>, Front End=<0.8 0.8 1>)) {module=ScaleMeshSize001}
   &ScaleColor(
    Scale Alpha=&FloatFromCurve(
     FloatCurve=DataInterface() {propertyValues {Curve {keys [<at 0 0>,<at 0.1 2>,<at 0.5 2>,<at 1 0>]}}}
     Scale Curve=&Lerp_Float(
      A=0.5
      B=&SelectFloatFromArray(
       Direct Array Index=$Particles.PartitionParticles.Partition
       Float Selection Array=DataInterface() {propertyValues {FloatData [0.5,0.5,0.5,0.5,0.5,0.5,0,0,0,0,0.5,0.5]}}
      )
      Alpha=&MakeCustomFloatFromBool(Bool=$User.bisHandHidden, True Float=1, False Float=0)
     )
    )
    Scale RGB=<0.4 0.4 0.4>
    ScaleA=true
    ScaleRGB=true
   )
   /SetVariables(Particles.UpdateFireColorOutsideAge:NiagaraFloat=&Divide_Float(A=$Particles.Age, B=$Particles.Initial.FireColorOutsideLifetime)) {isSetParameters=true, module=SetVariables_#4}
   /SetVariables(Particles.UpdateFireColorOutside:LinearColor=&MakeLinearColorFromVectorAndFloat(
     "Vector (RGB)"=&MakeVectorFromLinearColorRGB(Color=$Particles.Initial.FireColorOutside)
     "Float (Alpha)"=&FloatFromCurve(
      FloatCurve=DataInterface() {propertyValues {Curve {keys [<at 0 0>,<at 0.1 1>,<at 0.5 1>,<at 1 0>]}}}
      CurveIndex=$Particles.UpdateFireColorOutsideAge
      Scale Curve=&Lerp_Float(
       A=0.5
       B=&SelectFloatFromArray(
        Direct Array Index=$Particles.PartitionParticles.Partition
        Float Selection Array=DataInterface() {propertyValues {FloatData [0.5,0.5,0.5,0.5,0.5,0.5,0,0,0,0,0.5,0.5]}}
       )
       Alpha=&MakeCustomFloatFromBool(Bool=$User.bisHandHidden, True Float=1, False Float=0)
      )
     )
    )) {isSetParameters=true, module=SetVariables_#5}
   &InterpolateOverTime(
    Rate Of Change:NiagaraFloat=$System.Interpolate Rate
    Target Vector Value:NiagaraPosition=$Particles.Position
    Variable Type=::NewEnumerator2
   )
   /SetVariables(Particles.Interpolate DY:Vector4f=&MakeFloat4FromVectorAndFloat("Vector (RGB)"=$Particles.InterpolateOverTime.InterpolateOverTime_V3_.Moving Average, "Float (Alpha)"=$System.Interpolate Internsity)) {isSetParameters=true, module=SetVariables_#6}
   &DynamicMaterialParameters(
    Index 0 Param 1=0.6
    Index 0 Param 2=$Particles.Timeline.Time
    Index 0 Param 3=&FN_FrontEndFloatModifier(Gameplay=10, Front End=5)
    Index 0 Param 4=0.5
    Index 2 Param 1=0.7
    Module.Index0Param0Write=true
    Module.Index0Param1Write=true
    Module.Index0Param2Write=true
    Module.Index0Param3Write=true
    Module.Index2Param0Write=true
    Module.Index2Param1Write=true
    Module.Index2Param2Write=true
    Module.Index2Param3Write=true
    Write Parameter Index 2=true
   )
  ]
  rendererClasses [&NiagaraMeshRendererProperties]
  renderers [
   {rendererClass=&NiagaraMeshRendererProperties, rendererIndex=0}
   {rendererClass=&NiagaraMeshRendererProperties, rendererIndex=1}
   {rendererClass=&NiagaraMeshRendererProperties, rendererIndex=2}
  ]
 }
}
summary {
 systemName=NS_Character_Body_BurnLeer_Elimination
 emitters [
  {emitterName=NE_Fire_01, rendererClasses [&NiagaraMeshRendererProperties]}
  {emitterName=NE_Fire_02, rendererClasses [&NiagaraSpriteRendererProperties]}
  {emitterName=NE_Fire_Spline_01, rendererClasses [&NiagaraMeshRendererProperties]}
  {emitterName=NE_Fire_Spline_02, rendererClasses [&NiagaraMeshRendererProperties]}
  {emitterName=NE_Burst_Embers, rendererClasses [&NiagaraSpriteRendererProperties]}
 ]
 userVariables [
  User.bIsFrontEnd {type=NiagaraBool}
  User.bIsFrontEndPreview {type=NiagaraBool}
  User.bisHandHidden {type=NiagaraBool}
  User.IsReacting {type=NiagaraFloat, defaultValue=0, description=""}
  User.IsReactive {type=NiagaraFloat, defaultValue=1, description=""}
  User.IsSkydiving {type=NiagaraBool}
  User.KillCount {type=NiagaraInt32}
 ]
}
