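# Niagara cosmetic effect that dresses the BurnLeer character's skirt in fire, embers and motes.
# NE_Fire_Spline pins ten partitioned mesh flames to skeletal bones or sockets each update, writing smoothed positions to a particle attribute for lag.
# Sprite embers and curl-noise motes spawn on the skeletal mesh.
# User flags hide embers and motes while skydiving, shift flame color and camera offset when reacting, and kill flames on Change_OffPhysics.
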NiagaraSystem /BRCosmetics/Characters/Player/Female/Medium/Bodies/F_MED_BurnLeer/FX/NS_Character_BurnLeer_Skirt
dependencies {
 usedDataInterfaces [
  &NiagaraDataInterfacePlatformSet
  &NiagaraDataInterfaceArrayFloat
  &NiagaraDataInterfaceArrayInt32
  &NiagaraDataInterfaceArrayFloat3
  &NiagaraDataInterfaceCurve
  &NiagaraDataInterfaceSpriteRendererInfo
  &NiagaraDataInterfaceColorCurve
  &NiagaraDataInterfaceVectorCurve
  &NiagaraDataInterfaceSkeletalMesh
 ]
 usedDynamicInputs [&SetBoolByFloatComparison, &VectorLength]
 usedModules [
  &self:SystemSpawnScript.SystemScriptSource.SystemScriptGraph.NiagaraNodeAssignment_1.SetVariables_#1
  &self:SystemSpawnScript.SystemScriptSource.SystemScriptGraph.NiagaraNodeAssignment_0.SetVariables_#2
  &SystemState
  &self:SystemSpawnScript.SystemScriptSource.SystemScriptGraph.NiagaraNodeAssignment_2.SetVariables_#3
  &self:SystemSpawnScript.SystemScriptSource.SystemScriptGraph.NiagaraNodeAssignment_5.SetVariables_#4
 ]
}
emitters {
 NE_Embers {
  emitterUpdateScript [
   &EmitterState()
   &SpawnRate(SpawnRate=5, Use Spawn Probability=false)
  ]
  particleSpawnScript [
   &InitializeParticle(
    Color=<rgba 5 0.25 0 1>
    Lifetime=&Lerp_Float(
     A=&RandomRangeFloat(Minimum=1, Maximum=1.5)
     B=&RandomRangeFloat(Minimum=0.2, Maximum=0.5)
     Alpha=&MakeCustomFloatFromBool(Bool=$System.bIsRunning, True Float=1, False Float=0)
    )
    Position Offset=<0 0 30>
    Position Offset Coordinate Space=::Local
    Sprite Rotation Mode=::NewEnumerator2
    Sprite Size Mode=::NewEnumerator1
    Uniform Sprite Size Max=3.25
    Uniform Sprite Size Min=1.75
    UsePositionOffset=true
    Write Material Random=true
   )
   &ShapeLocation(MaskSpawnGroup=false, Override Seed=false, Sphere Radius=10, UseOwnerScale=true)
   &SkeletalMeshLocation(MaskSpawnGroup=false, Mesh Sampling Type=::NewEnumerator5, OverrideRandomSeed=false, Skeletal Mesh=$System.Skeletal Mesh)
  ]
  particleUpdateScript [
   &ParticleState()
   &ScaleSpriteSize(Uniform Curve Sprite Scale=DataInterface() {propertyValues {Curve {keys [
        {arriveTangent=0.5516, interpMode=RCIM_Cubic, leaveTangent=0.5516, tangentMode=RCTM_User, time=0, value=0}
        {arriveTangent=5.476, interpMode=RCIM_Cubic, leaveTangent=-0.224, tangentMode=RCTM_Break, time=0.4001, value=1.0058}
        {arriveTangent=0.0706, interpMode=RCIM_Cubic, leaveTangent=-23.6527, tangentMode=RCTM_Break, time=0.9106, value=1}
        <at 1 0>
       ]}}})
   &SpriteRotationRate(Rotation Rate=&AngleConversion(
     Angle:NiagaraFloat=&RandomRangeFloat(Minimum=-1, Maximum=1)
     Angle Input:ENiagara_AngleInput=::NewEnumerator1
     Angle Output:ENiagara_AngleInput=::NewEnumerator0
    ))
   &Drag(Drag=&RandomRangeFloat(Minimum=6, Maximum=8), Use Linear Drag=true, Use Rotational Drag=true)
   &AccelerationForce(Acceleration=<0 0 80>)
   &SolveForcesAndVelocity(Clamp Velocity=false, Limit Acceleration=false)
   &SubUVAnimation(
    Random Start Frame=true
    "Sprite Renderer "=DataInterface() {propertyValues {SpriteRenderer=&self:NE_Embers_0.NiagaraSpriteRendererProperties_0}}
    UseEndFrame=false
    UseStartFrame=false
   )
   &ScaleColor(
    Scale Alpha=&FloatFromCurve(
     FloatCurve=DataInterface() {propertyValues {Curve {keys [<at 0 0>,<at 0.2 1>,<at 1 0.2>]}}}
     Scale Curve=&Lerp_Float(A=1, B=0, Alpha=&MakeCustomFloatFromBool(Bool=$User.IsSkydiving, True Float=1, False Float=0))
    )
    Scale RGB=&VectorFromCurve(VectorCurve=DataInterface() {propertyValues {
       XCurve {keys [
         <at 0 500 RCIM_Cubic>
         {arriveTangent=-500, interpMode=RCIM_Cubic, leaveTangent=11.6698, tangentMode=RCTM_Break, time=0.8, value=0}
         <at 1 0 RCIM_Cubic>
        ]}
       YCurve {keys [
         <at 0 62.9696 RCIM_Cubic>
         {arriveTangent=-62.9696, interpMode=RCIM_Cubic, leaveTangent=-0.0144, tangentMode=RCTM_Break, time=0.8, value=0}
         <at 1 0 RCIM_Cubic>
        ]}
       ZCurve {keys [<at 0 0 RCIM_Cubic>,<at 1 0 RCIM_Cubic>]}
      }})
    ScaleA=true
    ScaleRGB=true
   )
  ]
  rendererClasses [&NiagaraSpriteRendererProperties]
  renderers [{rendererClass=&NiagaraSpriteRendererProperties}]
 }
 NE_Fire_Spline {
  emitterUpdateScript [
   &EmitterState(
    Enable Distance Culling=true
    Inactive Response=::NewEnumerator1
    Life Cycle Mode=::NewEnumerator1
    Loop Behavior=::NewEnumerator1
    Loop Duration Mode=::NewEnumerator1
    Max Distance Response=::SleepAndClearParticles
    MaxDistance=1000
    Module.bUseMaxDistance=true
    Module.bUseMinDistance=true
    Reset Age On Awaken=true
    Scalability Mode=::NewEnumerator1
    UseLoopDelay=false
   )
   &SpawnBurst_Instantaneous(
    Spawn Count=10
    Spawn Probability=&MakeCustomFloatFromBool(Bool=$User.Change_OffPhysics, True Float=0, False Float=1)
    Use Spawn Probability=true
    UseLoopCountLimit=false
   )
  ]
  particleSpawnScript [
   &PartitionParticles(Number Of Partitions=10)
   &InitializeParticle(
    Color=<rgba 2 1.2852 0 0.1>
    Mesh Scale=&MergeHelper_VectorSwitch(
     0=<0.36 1 0.5>
     1=<1 1 1>
     2=<-1 1 1>
     3=<-0.36 1 0.5>
     4=<0.36 1 0.5>
     5=<-0.36 1 0.5>
     6=<0.36 1 0.75>
     7=<-0.36 1 0.75>
     8=<-0.36 1 1>
     9=<0.36 1 1>
     Index=$Particles.PartitionParticles.Partition
     Num=::NewEnumerator8
    )
    Mesh Scale Mode=::NewEnumerator4
    UsePositionOffset=false
    Write Material Random=true
   )
   &InitialMeshOrientation(Mesh Orientation Mode=::NewEnumerator2, Rotation=<-0.25 0 0>, Use Rotation Vector=true)
   &InitialMeshOrientation(
    Mesh Orientation Mode=::NewEnumerator2
    Rotation=&self:ScratchDynamicInput(
     TargetPlatformActive:Vector3f=&MergeHelper_VectorSwitch(
      0=<0.05 0 0.25>
      1=<0 0 0.3>
      2=<0 0 0.2>
      3=<0.05 0 0.25>
      4=<0 0 0.3>
      5=<0 0 0.2>
      6=<0 0 0.25>
      7=<0 0 0.2>
      8=<0 0.03 0.15>
      9=<0 0.085 0.335>
      Index=$Particles.PartitionParticles.Partition
      Num=::NewEnumerator8
     )
     TargetPlatformNoActive:Vector3f=&MergeHelper_VectorSwitch(
      0=<0.05 0 0.25>
      1=<0 0 0.3>
      2=<0 0 0.2>
      3=<0.05 0 0.25>
      4=<0 0 0.3>
      5=<0 0 0.2>
      6=<0 0 0.25>
      7=<0 0 0.2>
      8=<0 0.03 0.15>
      9=<0 0.05 0.35>
      Index=$Particles.PartitionParticles.Partition
      Num=::NewEnumerator8
     )
     TargetPlatform:NiagaraDataInterfacePlatformSet=DataInterface() {propertyValues {Platforms {qualityLevelMask=-2}}}
    )
    Use Rotation Vector=true
   ) {module=InitialMeshOrientation001}
  ]
  particleUpdateScript [
   &CameraOffset(Camera Offset Amount=&Lerp_Float(
     A=&Lerp_Float(
      A=&SelectFloatFromArray(
       Direct Array Index=$Particles.PartitionParticles.Partition
       Float Selection Array=DataInterface() {propertyValues {FloatData [0,-10,-10,0,-10,-10,0,0,0,0]}}
      )
      B=-50
      Alpha=&MakeCustomFloatFromBool(Bool=$User.IsSkydiving, True Float=1, False Float=0)
     )
     B=10
     Alpha=$User.IsReacting
    ))
   &ParticleState(Kill Particles When Lifetime Has Elapsed=false)
   &Timeline(Play Rate=1.7, Timeline End=10000000)
   &SkeletalMeshLocation(
    Bone / Socket Index=&SelectIntFromArray(
     Direct Array Index=$Particles.PartitionParticles.Partition
     Int Selection Array=DataInterface() {propertyValues {IntData [0,1,1,2,3,4,5,6,7,8]}}
    )
    MaskSpawnGroup=false
    Mesh Sampling Type=::NewEnumerator5
    Sampled Position Offset=&Lerp_Vector(
     A=&MergeHelper_VectorSwitch(
      0=<0 0 0>
      1=<0 0 0>
      2=<0 0 0>
      3=<0 0 0>
      4=<0 0 2>
      5=<0 0 2>
      6=<0 0 0>
      7=<0 2 0>
      8=<0 0 0>
      9=<0 0 0>
      Index=$Particles.PartitionParticles.Partition
      Num=::NewEnumerator8
     )
     B=&Lerp_Vector(
      A=&MergeHelper_VectorSwitch(
       0=<0 0 0>
       1=<0 0 0>
       2=<0 0 0>
       3=<0 0 0>
       4=<0 0 2>
       5=<0 0 2>
       6=<0 0 0>
       7=<0 2 0>
       8=<0 0 0>
       9=<0 0 0>
       Index=$Particles.PartitionParticles.Partition
       Num=::NewEnumerator8
      )
      B=&MergeHelper_VectorSwitch(
       0=<0 0 0>
       1=<0 0 0>
       2=<0 0 0>
       3=<0 0 0>
       4=<0 0 2>
       5=<0 0 2>
       6=<0 0 0>
       7=<0 2 0>
       8=<0 0 0>
       9=<0 0 0>
       Index=$Particles.PartitionParticles.Partition
       Num=::NewEnumerator8
      )
      Alpha=&MakeCustomFloatFromBool(Bool=$User.IsSkydiving, True Float=1, False Float=0)
     )
     Alpha=$System.InterpolateOverTime.InterpolateOverTime_Float.Moving Average
    )
    Skeletal Mesh=$System.Skeletal Mesh
    Socket Sampling Mode=::NewEnumerator5
   )
   &ScaleMeshSize(Scale Factor=&Multiply_Vector(
     A=&SelectVectorFromArray(
      Direct Array Index=$Particles.PartitionParticles.Partition
      Vector Selection Array=DataInterface() {propertyValues {FloatData [
         <0.2 0.2 0.22>
         <0.2 0.2 0.22>
         <0.2 0.2 0.22>
         <0.2 0.2 0.22>
         <0.2 0.2 0.22>
         <0.2 0.2 0.22>
         <0.2 0.2 0.22>
         <0.2 0.2 0.22>
         <0.2 0.2 0.22>
         <0.2 0.2 0.22>
        ]}}
     )
     B=&Lerp_Vector(
      A=&Lerp_Vector(A=<1 1 1>, B=<1 1 1>, Alpha=$System.InterpolateOverTime001.InterpolateOverTime_Float.Moving Average)
      B=<0.8 0.8 0.8>
      Alpha=$System.InterpolateOverTime.InterpolateOverTime_Float.Moving Average
     )
    ))
   &ScaleMeshSize(Scale Factor=&FN_FrontEndVector3Modifier(Gameplay=<0.8 0.8 1>, Front End=<0.8 0.8 1>)) {module=ScaleMeshSize001}
   &ScaleColor(Scale Alpha=2, Scale RGB=<0.4 0.4 0.4>, ScaleA=true, ScaleRGB=true)
   &InterpolateOverTime(
    Rate Of Change:NiagaraFloat=$System.Interpolate Rate
    Target Vector Value:NiagaraPosition=$Particles.Position
    Variable Type=::NewEnumerator2
   )
   /SetVariables(Particles.Interpolate DY:Vector4f=&MakeFloat4FromVectorAndFloat("Vector (RGB)"=$Particles.InterpolateOverTime.InterpolateOverTime_V3_.Moving Average, "Float (Alpha)"=$System.Interpolate Internsity)) {isSetParameters=true}
   &DynamicMaterialParameters(
    Index 0 Param 1=&SelectFloatFromArray(
     Direct Array Index=$Particles.PartitionParticles.Partition
     Float Selection Array=DataInterface() {propertyValues {FloatData [0.6,0.6,0.6,0.6,0.6,0.6,0.6,0.6,0.6,0.6]}}
    )
    Index 0 Param 2=$Particles.Timeline.Time
    Index 0 Param 3=&FN_FrontEndFloatModifier(Gameplay=10, Front End=5)
    Index 0 Param 4=&SelectFloatFromArray(
     Direct Array Index=$Particles.PartitionParticles.Partition
     Float Selection Array=DataInterface() {propertyValues {FloatData [0.1,0.5,0.1,0.5,0.25,0.1,0.5,0.25,0.1,0.1]}}
    )
    Index 2 Param 1=0.7
    Module.Index0Param0Write=true
    Module.Index0Param1Write=true
    Module.Index0Param2Write=true
    Module.Index0Param3Write=true
    Module.Index2Param0Write=true
    Module.Index2Param1Write=true
    Module.Index2Param2Write=true
    Module.Index2Param3Write=true
    Write Parameter Index 2=true
   )
   /SetVariables(Particles.FireColorOutside:LinearColor=&LerpLinearColors(StartColor=<rgba 1.5 0.3 0 0.1>, EndColor=<rgba 1 0.2 0 0.25>, LerpFactor=$User.IsReacting)) {isSetParameters=true}
   &KillParticles(Kill Particles=$User.Change_OffPhysics)
  ]
  rendererClasses [&NiagaraMeshRendererProperties]
  renderers [
   {rendererClass=&NiagaraMeshRendererProperties, rendererIndex=0}
   {rendererClass=&NiagaraMeshRendererProperties, rendererIndex=1}
   {rendererClass=&NiagaraMeshRendererProperties, rendererIndex=2}
  ]
 }
 NE_Motes {
  emitterUpdateScript [
   &EmitterState(Life Cycle Mode=::NewEnumerator1, Loop Delay=1, Loop Duration=1, UseLoopDelay=false)
   &SpawnRate(
    Spawn Probability=&MakeCustomFloatFromBool(Bool=$User.IsSkydiving, True Float=0, False Float=1)
    SpawnRate=3
    Use Spawn Probability=true
   )
  ]
  particleSpawnScript [
   /SetVariables(Particles.PerParticleRandom:NiagaraFloat=&RandomRangeFloat(Minimum=0, Maximum=1)) {isSetParameters=true}
   &InitializeParticle(
    Color Mode=::NewEnumerator0
    Lifetime=&Lerp_Float(
     A=&Lerp_Float(A=1, B=2, Alpha=$Particles.PerParticleRandom)
     B=&Lerp_Float(A=0.2, B=0.5, Alpha=$Particles.PerParticleRandom)
     Alpha=&MakeCustomFloatFromBool(Bool=$System.bIsRunning, True Float=1, False Float=0)
    )
    Position Offset=<0 0 5>
    Position Offset Coordinate Space=::Local
    Sprite Size=&Multiply_Vector2D(
     A=<0.5 1.5>
     B=&MakeVector2D(
      X=&FloatFromCurve(
       FloatCurve=DataInterface() {propertyValues {Curve {keys [<at 0 1>,<at 0.8 2 RCIM_Constant>,<at 0.95 3 RCIM_Constant>]}}}
       CurveIndex=$Particles.PerParticleRandom
      )
      Y=&FloatFromCurve(
       FloatCurve=DataInterface() {propertyValues {Curve {keys [<at 0 1>,<at 0.8 2 RCIM_Constant>,<at 0.95 3>]}}}
       CurveIndex=$Particles.PerParticleRandom
      )
     )
    )
    Sprite Size Mode=::NewEnumerator4
    UsePositionOffset=true
    Write Material Random=true
   )
   &AddVelocity(
    Cone Axis=<0 0 1>
    Module.UseConeFalloff=false
    Module.UseVelDistribution=false
    Override Seed=false
    Velocity Mode=::NewEnumerator2
    Velocity Speed=100
   )
   &SkeletalMeshLocation(MaskSpawnGroup=false, Mesh Sampling Type=::NewEnumerator5, OverrideRandomSeed=false, Skeletal Mesh=$System.Skeletal Mesh)
  ]
  particleUpdateScript [
   &ParticleState()
   &Drag(
    Drag=&FloatFromCurve(
     FloatCurve=DataInterface() {propertyValues {Curve {keys [<at 0 0.5>,<at 0.5 0.5>,<at 1 1>]}}}
     Scale Curve=&RandomRangeFloat(Minimum=6, Maximum=22)
    )
    Use Linear Drag=true
    Use Rotational Drag=false
   )
   &CurlNoiseForce(Bias Noise=false, Noise Frequency=150, Noise Strength=500, Pan Noise=false) {module=CurlNoiseForce001}
   &SolveForcesAndVelocity(Clamp Velocity=false, Limit Acceleration=false)
   &ScaleColor(
    Scale Alpha=&Lerp_Float(A=1, B=0, Alpha=&MakeCustomFloatFromBool(Bool=$User.IsSkydiving, True Float=1, False Float=0))
    Scale RGB=&MakeVectorFromLinearColorRGB(Color=&MergeHelper_ColorSwitch(
      0=&ColorFromCurve(ColorCurve=DataInterface() {propertyValues {
         AlphaCurve {keys [<at 0 1 RCIM_Cubic>]}
         BlueCurve {keys [<at 0 0 RCIM_Cubic>,<at 0.4579 0 RCIM_Cubic>]}
         GreenCurve {keys [<at 0 1.8333 RCIM_Cubic>,<at 0.4579 0.1 RCIM_Cubic>]}
         RedCurve {keys [<at 0 10 RCIM_Cubic>,<at 0.4579 1 RCIM_Cubic>]}
        }})
      1=&ColorFromCurve(ColorCurve=DataInterface() {propertyValues {
         AlphaCurve {keys [<at 0 1 RCIM_Cubic>]}
         BlueCurve {keys [<at 0 10 RCIM_Cubic>,<at 0.4579 1 RCIM_Cubic>]}
         GreenCurve {keys [<at 0 2.4349 RCIM_Cubic>,<at 0.4579 0.1667 RCIM_Cubic>]}
         RedCurve {keys [<at 0 0 RCIM_Cubic>,<at 0.4579 0 RCIM_Cubic>]}
        }})
      2=?
      3=?
      Index=&MakeIntFromFloat(Float=0)
      Num=::NewEnumerator0
     ))
    ScaleA=true
    ScaleRGB=true
   )
   &ScaleSpriteSizeBySpeed(
    Max Scale Factor=<1.666 6>
    Min Scale Factor=<1.333 1.333>
    Sample Scale Factor By Curve=true
    Scale Factor Curve=DataInterface() {propertyValues {Curve {keys [<at 0 0>, {arriveTangent=1.4989, interpMode=RCIM_Cubic, tangentMode=RCTM_Break, time=1, value=1}]}}}
    Velocity Threshold=999
   )
   &ScaleSpriteSize(
    Uniform Curve Scale=&MakeFloatFromVector(Vector=$Engine.Owner.Scale, Channel=::NewEnumerator0)
    Uniform Curve Sprite Scale=DataInterface() {propertyValues {Curve {keys [
        {tangentMode=RCTM_User, time=0, value=1}
        {arriveTangent=-0.0405, interpMode=RCIM_Cubic, leaveTangent=-0.0405, tangentMode=RCTM_User, time=0.7009, value=1}
        {arriveTangent=-8.0089, interpMode=RCIM_Cubic, leaveTangent=-8.0089, tangentMode=RCTM_User, time=1, value=0}
       ]}}}
   )
  ]
  rendererClasses [&NiagaraSpriteRendererProperties]
  renderers [{rendererClass=&NiagaraSpriteRendererProperties}]
 }
}
summary {
 systemName=NS_Character_BurnLeer_Skirt
 emitters [
  {emitterName=NE_Fire_Spline, rendererClasses [&NiagaraMeshRendererProperties]}
  {emitterName=NE_Embers, rendererClasses [&NiagaraSpriteRendererProperties]}
  {emitterName=NE_Motes, rendererClasses [&NiagaraSpriteRendererProperties]}
 ]
 userVariables [
  User.bisFootHidden {type=NiagaraBool}
  User.bIsFrontEnd {type=NiagaraBool}
  User.bIsFrontEndPreview {type=NiagaraBool}
  User.bisHandHidden {type=NiagaraBool}
  User.bisHeadHidden {type=NiagaraBool}
  User.Change_OffPhysics {type=NiagaraBool}
  User.Emote begin {type=NiagaraBool}
  User.IsReacting {type=NiagaraFloat, defaultValue=0, description=""}
  User.IsSkydiving {type=NiagaraBool}
  User.VariantID {type=NiagaraFloat, defaultValue=0, description=""}
 ]
}
