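# Parent material for the eyes of the BurnLeer female medium cosmetic body, built on material attributes for skeletal meshes with mesh deformer support.
# An eye function specific to BurnLeer feeds the shared 2023 MF_Eyelash function, whose result passes through MF_TechArtOverrides into the material.
# The shading model comes from the graph, with Default Lit and Subsurface in use.
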Material /BRCosmetics/Characters/Player/Female/Medium/Bodies/F_MED_BurnLeer/Materials/M_BurnLeer_Eyes_Parent
graph {
 properties {
  bUseMaterialAttributes=true
  bUsedWithMeshDeformer=true
  bUsedWithSkeletalMesh=true
  pixelDepthOffsetMode=PDOM_Legacy
  shadingModel=MSM_FromMaterialExpression
  usedShadingModels=Default Lit | Subsurface
 }
 nodes [MaterialFunctionCall(In=MaterialFunctionCall!(Material=MaterialFunctionCall() {materialFunction=&/MF_eyes_BurnLeer}.Result) {materialFunction=/Game/Global/Materials/Character_Parent_2023/Eye_Parent_2023/MF_Eyelash}.Result) {materialFunction=&MF_TechArtOverrides}]
}
