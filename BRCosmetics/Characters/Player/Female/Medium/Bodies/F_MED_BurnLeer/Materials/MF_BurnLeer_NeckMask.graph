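# Material function for the BurnLeer female character body that builds a procedural neck mask and a face mask.
# It compares pre-skinned local position, passed through a vertex interpolator, against three parameterized sphere masks, so the masks stay fixed to the mesh during animation.
# The second and third spheres are subtracted from the first, then the result is cubed, saturated and scaled by a blend parameter.
# The third sphere is also output directly as FaceMask.
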MaterialFunction /BRCosmetics/Characters/Player/Female/Medium/Bodies/F_MED_BurnLeer/Materials/MF_BurnLeer_NeckMask
graph {nodes [
  n1 = VertexInterpolator(VS=LocalPosition() {localOrigin=InstancePreSkinning}.XYZ)
  n2 = VectorParameter() {group=BurnLeer_NeckMask, parameterName=BurnLeer_NeckMask_pos1, defaultValue=<rgba -0.576 -5.6837 139.3837 1>}
  n3 = SphereMask(
   A=@n1.PS
   B=@n2.RGB
   Radius=ScalarParameter() {defaultValue=43.8403, group=BurnLeer_NeckMask, parameterName=BurnLeer_NeckMaskRadius1}
   Hardness=ScalarParameter() {defaultValue=1, group=BurnLeer_NeckMask, parameterName=BurnLeer_NeckMaskHardness1}
  )
  n4 = VectorParameter() {group=BurnLeer_NeckMask, parameterName=BurnLeer_NeckMask_pos2, defaultValue=<rgba -0.576 -2.1961 145.7796 1>}
  n5 = SphereMask(
   A=@n1.PS
   B=@n4.RGB
   Radius=ScalarParameter() {defaultValue=13.93, group=BurnLeer_NeckMask, parameterName=BurnLeer_NeckMaskRadius2}
   Hardness=ScalarParameter() {group=BurnLeer_NeckMask, parameterName=BurnLeer_NeckMaskHardness2}
  )
  n6 = SphereMask(
   A=@n1.PS
   B=VectorParameter() {group=BurnLeer_NeckMask, parameterName=BurnLeer_NeckMask_pos3, defaultValue=<rgba 0 29.6635 170 1>}.RGB
   Radius=ScalarParameter() {defaultValue=43, group=BurnLeer_NeckMask, parameterName=BurnLeer_NeckMaskRadius3}
   Hardness=ScalarParameter() {defaultValue=0.8, group=BurnLeer_NeckMask, parameterName=BurnLeer_NeckMaskHardness3}
  )
  n7 = Power(Base=Subtract(A=Subtract(A=@n3, B=@n5), B=@n6)) {constExponent=3}
  FunctionOutput(None=Multiply(A=Saturate(None=@n7), B=ScalarParameter() {defaultValue=1, group=BurnLeer_NeckMask, parameterName=BurnLeer_NeckMask_Blend}))
  FunctionOutput(None=@n6) {outputName=FaceMask}
 ]}
