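# Material function for the BurnLeer character body that outputs an animated reactive fire effect, scaled by a mask input that defaults to a texture channel.
# Panned, distorted noise is sampled on triplanar front and side projections and blended, with named reroutes for distortion and a Fresnel-based gradient lookup for emissive color.
# Fire color blends between two parameters by Fresnel, and tinted flowing smoke is added only on High and Epic quality.
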MaterialFunction /BRCosmetics/Characters/Player/Female/Medium/Bodies/F_MED_BurnLeer/Materials/MF_BurnLeer_ReactiveFire
graph {nodes [
  n1 = NamedRerouteUsage()
  ComponentMask(None=@n1) {g=true, r=true}
  n2 = TextureSample() {
   samplerSource=SSM_Wrap_WorldGroupSettings
   samplerType=SAMPLERTYPE_Masks
   texture=/BRCosmetics/Characters/Player/Female/Medium/Bodies/F_MED_BurnLeer/Textures/T_BurnLeer_Body_FX3
  }
  n3 = FunctionInput(Preview=@n2.R) {bUsePreviewValueAsDefault=true, inputName=ReactiveFireMask, inputType=FunctionInput_Scalar}
  n4 = Constant()
  n5 = MaterialFunctionCall!(In_FireSmoke_TAMask, In_FireSmoke_UV=Multiply(A=TextureCoordinate(), B=Constant() {r=1.9357}) {constB=10}) {materialFunction=/BRCosmetics/Gadgets/Assets/VinderTech_GliderChute/Glider_RampCanal/Materials/MFW_FlowingSmoke}
  n6 = ScalarParameter() {defaultValue=0.6, group=BurnLeer, parameterName=ReactiveFirestIntensity}
  n7 = VectorParameter() {group=BurnLeer, parameterName=Reactive_FresnelColorA, defaultValue=<rgba 1.3 1.3 1.3 1>}
  n8 = VectorParameter() {group=BurnLeer, parameterName=Reactive_FresnelColorB, defaultValue=<rgba 20 0.3528 0 1>}
  n9 = Fresnel(ExponentIn=LinearInterpolate(A=ScalarParameter() {defaultValue=2, group=BurnLeer, parameterName=Reactive_Fresnel_Power}, Alpha=NamedRerouteUsage())) {exponent=1}
  n10 = Add(
   A=Multiply(
    A=MaterialFunctionCall!(
     Low _ PC Low and Med=@n4
     "High _ XB1, PS4"=@n5.Result
     "Epic _ PC High, XSX, PS5"=@n5.Result
     --------
     "Mobile Low _ PC Perf, Android"=@n4
     Mobile Med _ Switch=@n4
    ) {materialFunction=/Game/Global/MaterialFunctions/TechArt/MF_Fortnite_QualitySwitch_Vector3}.Result
    B=VectorParameter() {group=BurnLeer, parameterName=Reactive_SmokeColor, defaultValue=<rgba 0.41 0.41 0 1>}.RGB
   )
   B=Multiply(A=Multiply(A=@n1, B=@n6), B=LinearInterpolate(A=@n7.RGB, B=@n8.RGB, Alpha=@n9) {constA=1})
  )
  FunctionOutput(None=NamedRerouteDeclaration(None=Multiply(A=@n10, B=@n3)) {name=ReactiveFire, nodeColor=<rgba 0.8281 1 0 1>})
  n11 = TextureSample(UVs=MaterialFunctionCall(X=NamedRerouteUsage(), Y=Constant() {r=0.5}) {materialFunction=&MakeFloat2}.Result) {texture=/BRCosmetics/Characters/Player/Male/Medium/Bodies/M_MED_TurboFueled/FX/T_TurboFueled_Gradient}
  n12 = ScalarParameter() {defaultValue=3, group=BurnLeer, parameterName=Reactive_ColorPower}
  NamedRerouteDeclaration(None=Multiply(A=Power(Base=@n11.RGB, Exp=@n12), B=ScalarParameter() {defaultValue=0.608, group=BurnLeer, parameterName=Reactive_ColorIntensity})) {name=EmissiveOUT_1, nodeColor=<rgba 1 0.0938 0 1>}
  n13 = Power(Base=NamedRerouteUsage(), Exp=Constant() {r=0.5})
  NamedRerouteDeclaration(None=Multiply(A=ConstantBiasScale(None=@n13) {bias=-0.5, scale=2}, B=Constant() {r=0.4})) {name=Distortion, nodeColor=<rgba 1 0.7031 0 1>}
  n14 = Saturate(None=DotProduct(A=CameraVectorWS(), B=PixelNormalWS()))
  n15 = Constant() {r=1.2258}
  n16 = Add(A=NamedRerouteUsage(), B=NamedRerouteDeclaration(None=Power(Base=@n14, Exp=@n15)) {name=FresenlValue, nodeColor=<rgba 0.7344 0 1 1>})
  NamedRerouteDeclaration(None=LinearInterpolate(Alpha=Saturate(None=@n16)) {constB=0.99}) {name=GradientUVX, nodeColor=<rgba 0 1 0.0625 1>}
  n17 = Multiply(A=NamedRerouteUsage(), B=Constant() {r=0.9003}) {constB=4}
  n18 = FunctionInput(Preview=ComponentMask(None=VectorParameter() {group=BurnLeer, parameterName=ReactiveFireSpeed, defaultValue=<rgba 0 0.3 0 1>}.RGB) {g=true, r=true}) {bUsePreviewValueAsDefault=true, inputName=ReactiveFireSpeed, inputType=FunctionInput_Vector2}
  n19 = Multiply(A=NamedRerouteUsage(), B=Constant() {r=4.7613}) {constB=11}
  n20 = ComponentMask(None=Constant4Vector() {constant=<rgba -0.2 0.3 0 1>}.RGB) {g=true, r=true}
  n21 = TextureSample(UVs=MaterialFunctionCall(UVs=Panner(Coordinate=@n19, Speed=@n20) {speedX=0.07, speedY=0.07}, "Rotation Angle (0-1)"=Constant()) {materialFunction=&CustomRotator}.Rotated Values) {samplerType=SAMPLERTYPE_Masks, texture=&T_BurnLeer_Body_FX3}
  n22 = Constant() {r=0.01}
  n23 = TextureSampleParameter2D(UVs=Add(A=Panner(Coordinate=@n17, Speed=@n18) {speedY=0.3}, B=Multiply(A=@n21.G, B=@n22))) {
   group=BurnLeer
   parameterName=T_ReactiveFire
   samplerSource=SSM_Wrap_WorldGroupSettings
   samplerType=SAMPLERTYPE_LinearColor
   texture=/Game/Global/Textures/Default/Noise/T_Curie_NoisePack
  }
  n24 = Multiply(A=NamedRerouteUsage(), B=Constant() {r=0.9003}) {constB=4}
  n25 = Multiply(A=NamedRerouteUsage(), B=Constant() {r=4.7613}) {constB=11}
  n26 = ComponentMask(None=Constant4Vector() {constant=<rgba -0.2 0.3 0 1>}.RGB) {g=true, r=true}
  n27 = TextureSample(UVs=MaterialFunctionCall(UVs=Panner(Coordinate=@n25, Speed=@n26) {speedX=0.07, speedY=0.07}, "Rotation Angle (0-1)"=Constant()) {materialFunction=&CustomRotator}.Rotated Values) {samplerType=SAMPLERTYPE_Masks, texture=&T_BurnLeer_Body_FX3}
  n28 = Constant() {r=0.01}
  n29 = TextureSampleParameter2D(UVs=Add(A=Panner(Coordinate=@n24, Speed=@n18) {speedY=0.2}, B=Multiply(A=@n27.G, B=@n28))) {
   group=BurnLeer
   parameterName=T_ReactiveFire
   samplerSource=SSM_Wrap_WorldGroupSettings
   samplerType=SAMPLERTYPE_LinearColor
   texture=&T_Curie_NoisePack
  }
  n30 = MaterialFunctionCall() {materialFunction=/BRCosmetics/Characters/Player/Male/Medium/Bodies/M_MED_BirchPalace/Skins/Gas/Materials/MF_TriplanarBlendMask}
  n31 = LinearInterpolate(A=@n23.RGB, B=@n29.RGB, Alpha=Saturate(None=Multiply(A=@n30.AxisMask_X) {constB=2}))
  NamedRerouteDeclaration(None=Saturate(None=ComponentMask(None=@n31) {g=true, r=true})) {name=DistortionNoiseSample, nodeColor=<rgba 1 0 0.5391 1>}
  n32 = MaterialFunctionCall!(
   TriplanarUV_UseClamp?=Constant()
   TriplanarUV_FlipUV=Constant()
   TriplanarUV_CenterPosition=Constant4Vector() {constant=<rgba 0 0 0 1>}.RGB
   TriplanarUV_UVScale=Constant() {r=30}
  ) {materialFunction=/BRCosmetics/Characters/Player/Male/Medium/Bodies/M_MED_BirchPalace/Skins/Gas/Materials/MF_TriplanarUV}
  NamedRerouteDeclaration(None=@n32.UV_FrontProjection) {name=UV Front, nodeColor=<rgba 0.4062 0 1 1>}
  NamedRerouteDeclaration(None=@n32.UV_SideProjection) {name=Uvside, nodeColor=<rgba 0.2188 0 1 1>}
  NamedRerouteUsage()
 ]}
