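# Material function for the BurnLeer female character body, named for a neck fresnel effect, that outputs a tinted rim term.
# A Fresnel node with a fixed exponent is multiplied by a constant HDR red color to form the single function output.
# It takes no inputs, so the falloff and color are hardcoded inside the function.
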MaterialFunction /BRCosmetics/Characters/Player/Female/Medium/Bodies/F_MED_BurnLeer/Materials/MF_Burnleer_NeckFresnel
graph {nodes [
  n1 = Constant() {r=2.463}
  FunctionOutput(None=Multiply(A=Fresnel(ExponentIn=@n1), B=Constant4Vector() {constant=<rgba 2 0 0.0047 1>}.RGB))
 ]}
