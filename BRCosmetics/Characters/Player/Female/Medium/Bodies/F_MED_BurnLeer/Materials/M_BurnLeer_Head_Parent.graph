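# Parent material for the BurnLeer character's head, inserting reactive fire effects between the shared character PreFX and PostFX functions.
# A neck Fresnel, excluded from a pre-skinned-position sphere mask and scaled by FireReactive and EnableReactive, adds to base color.
# Emissive combines tinted fire from a ReactiveFire function, driven by neck masking, with forehead texture reroutes.
# A face mask from the neck mask function sets opacity under expression-driven shading models.
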Material /BRCosmetics/Characters/Player/Female/Medium/Bodies/F_MED_BurnLeer/Materials/M_BurnLeer_Head_Parent
graph {
 properties {
  bAutomaticallySetUsageInEditor=false
  bUseMaterialAttributes=true
  bUsedWithMeshDeformer=true
  bUsedWithMorphTargets=true
  bUsedWithSkeletalMesh=true
  pixelDepthOffsetMode=PDOM_Legacy
  shadingModel=MSM_FromMaterialExpression
  usedShadingModels=Default Lit | Subsurface | Clear Coat | Subsurface Profile | Cloth
 }
 nodes [
  n1 = MaterialFunctionCall() {materialFunction=&MF_CharacterParent_2023_PreFX}
  n2 = GetMaterialAttributes(None=@n1.Result) {attributeGetTypes [#1, #2, #3]}
  n3 = LocalPosition() {localOrigin=InstancePreSkinning}
  n4 = Multiply(
   A=SphereMask(
    A=VertexInterpolator(VS=@n3.XYZ).PS
    B=Constant4Vector() {constant=<rgba 0 6.2428 149.8385 1>}.RGB
    Radius=Constant() {r=6.4255}
    Hardness=Constant() {r=0.256}
   )
   B=Constant() {r=1}
  ) {constB=2}
  n5 = Multiply(
   A=OneMinus(None=Saturate(None=@n4))
   B=MaterialFunctionCall() {materialFunction=&/MF_Burnleer_NeckFresnel}.Result
  )
  n6 = NamedRerouteDeclaration(None=Multiply(
    A=ScalarParameter() {group=BurnLeer, parameterName=FireReactive}
    B=ScalarParameter() {defaultValue=1, group=BurnLeer, parameterName=EnableReactive}
   )) {name=FireReactive, nodeColor=<rgba 1 0 0.2344 1>}
  n7 = MaterialFunctionCall() {materialFunction=&/MF_BurnLeer_NeckMask}
  n8 = SetMaterialAttributes!(
   MaterialAttributes=@n2.MaterialAttributes
   Emissive Color=NamedRerouteUsage()
   Base Color=Add(A=Multiply(A=@n5, B=@n6), B=@n2.Base Color)
   Fuzz Color or Subsurface Color
   Opacity=LinearInterpolate(B=Constant() {r=1}, Alpha=Power(Base=@n7.FaceMask)) {constB=0}
  ) {attributeSetTypes [
    #1
    #3
    #2
    #4
   ]}
  MaterialFunctionCall(Material=@n8, M=@n1.M, Pre Skinned Local Position=@n1.M, Distance Blend=@n1.M, Fresnel=@n1.M) {materialFunction=&MF_CharacterParent_2023_PostFX}
  n9 = TextureSample() {texture=/BRCosmetics/Characters/Player/Female/Medium/Bodies/F_MED_BurnLeer/Textures/T_BurnLeer_foreHead_E}
  NamedRerouteDeclaration(None=@n9.RGB) {name=foreHead_E, nodeColor=<rgba 0 0.6719 1 1>}
  NamedRerouteDeclaration(None=@n9.R) {name=foreHeadMask, nodeColor=<rgba 0.9922 1 0 1>}
  n10 = Constant3Vector() {constant=<rgba 0 0.25 0 0>}
  n11 = Add(
   A=MaterialFunctionCall() {materialFunction=&MF_BurnLeer_NeckMask}.Result
   B=NamedRerouteUsage()
  )
  n12 = Multiply(
   A=Power(
    Base=MaterialFunctionCall!(ReactiveFireSpeed=@n10, ReactiveFireMask=@n11) {materialFunction=&/MF_BurnLeer_ReactiveFire}.Result
    Exp=Constant() {r=1.5}
   )
   B=VectorParameter() {group=BurnLeer, parameterName=ReactiveFireColor, defaultValue=<rgba 1 0.4313 0 1>}.RGB
  )
  n13 = Multiply(A=NamedRerouteUsage(), B=@n12)
  n14 = Multiply(A=@n12, B=Power(Base=NamedRerouteUsage()))
  n15 = NamedRerouteUsage()
  NamedRerouteDeclaration(None=Add(A=Add(A=@n13, B=@n14), B=Power(Base=@n15) {constExponent=1.5})) {name=Out EmissiveColor, nodeColor=<rgba 0.6875 0 1 1>}
 ]
}
