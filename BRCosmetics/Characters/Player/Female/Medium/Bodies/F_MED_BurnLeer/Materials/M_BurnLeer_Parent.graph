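# Parent character material for the BurnLeer female outfit, adding animated fire, ember and glow effects to cloth and body.
# It inserts effects between the shared 2023 character PreFX and PostFX functions, overriding base color and emissive via material attributes.
# Panned noise distorts packed-mask and emissive lookups for flames, burning edges and internal glow, localized by sphere masks and Fresnel.
# A FireReactive parameter boosts flames; quality switches restrict embers and smoke to higher tiers.
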Material /BRCosmetics/Characters/Player/Female/Medium/Bodies/F_MED_BurnLeer/Materials/M_BurnLeer_Parent
graph {
 properties {
  bAutomaticallySetUsageInEditor=false
  bUseMaterialAttributes=true
  bUsedWithMeshDeformer=true
  bUsedWithMorphTargets=true
  bUsedWithSkeletalMesh=true
  pixelDepthOffsetMode=PDOM_Legacy
  shadingModel=MSM_FromMaterialExpression
  usedShadingModels=Default Lit | Subsurface | Clear Coat | Subsurface Profile | Cloth
 }
 nodes [
  n1 = MaterialFunctionCall() {materialFunction=&MF_CharacterParent_2023_PreFX}
  MaterialFunctionCall(Material=NamedRerouteUsage(), M=@n1.M, Pre Skinned Local Position=@n1.M, Distance Blend=@n1.M, Fresnel=@n1.M) {materialFunction=&MF_CharacterParent_2023_PostFX}
  Multiply()
  Multiply()
  n2 = TextureSampleParameter2D() {
   group=06_Emissive
   parameterName=Emissive
   samplerSource=SSM_Wrap_WorldGroupSettings
   sortPriority=1
   texture=&T_BlacksRGB
  }
  n3 = NamedRerouteUsage()
  n4 = NamedRerouteUsage()
  n5 = Add(A=Add(A=@n4, B=NamedRerouteUsage()), B=NamedRerouteUsage())
  n6 = ComponentMask(None=VertexInterpolator(VS=LocalPosition() {localOrigin=InstancePreSkinning}.XYZ).PS) {r=true}
  n7 = Constant()
  n8 = Constant() {r=32.8015}
  NamedRerouteDeclaration(None=LinearInterpolate(A=@n2.RGB, B=Add(A=@n3, B=@n5), Alpha=SphereMask(A=@n6, B=@n7, Radius=@n8))) {name=Cloth_Fire, nodeColor=<rgba 0.5 1 0 1>}
  n9 = MaterialFunctionCall!(
   TriplanarUV_UseClamp?=ScalarParameter() {group=TriplanarUV, parameterName=TriplanarUV_3D_UseClamp}
   TriplanarUV_FlipUV=ScalarParameter() {group=TriplanarUV, parameterName=TriplanarUV_3D_FlipUV}
   TriplanarUV_CenterPosition=VectorParameter() {group=TriplanarUV, parameterName=TriplanarUV_3D_CenterPosition, defaultValue=<rgba 0 0 0 1>}.RGB
   TriplanarUV_UVScale=ScalarParameter() {defaultValue=82, group=TriplanarUV, parameterName=TriplanarUV_3D_UVSize}
  ) {materialFunction=/BRCosmetics/Characters/Player/Male/Medium/Bodies/M_MED_BirchPalace/Skins/Gas/Materials/MF_TriplanarUV}
  NamedRerouteDeclaration(None=@n9.UV_FrontProjection) {name=TriplannarUV_front, nodeColor=<rgba 1 0.3516 0 1>}
  n10 = Constant4Vector() {constant=<rgba 0.544 0.8629 0 0>}
  n11 = TextureSampleParameter2D() {
   parameterName=T_Cloth_Fire_Mask
   samplerSource=SSM_Wrap_WorldGroupSettings
   samplerType=SAMPLERTYPE_Masks
   texture=/BRCosmetics/Characters/Player/Female/Medium/Bodies/F_MED_BurnLeer/Textures/T_BurnLeer_Body_FX2
  }
  n12 = Multiply(
   A=SphereMask(A=TextureCoordinate(), B=ComponentMask(None=@n10.RGB) {g=true, r=true}, Radius=Constant() {r=0.296})
   B=Saturate(None=Multiply(A=@n11.R) {constB=3})
  )
  NamedRerouteDeclaration(None=@n12) {name=FixlegSeamMask, nodeColor=<rgba 0 1 0.7188 1>}
  n13 = TextureCoordinate()
  n14 = ComponentMask(None=Constant4Vector() {constant=<rgba 0.544 0.664 0 1>}.RGB) {g=true, r=true}
  n15 = ScalarParameter() {defaultValue=0.1, group=BurnLeer, parameterName=chestMaskRadius}
  NamedRerouteDeclaration(None=Multiply(A=SphereMask(A=@n13, B=@n14, Radius=@n15), B=NamedRerouteUsage())) {name=ChestMask, nodeColor=<rgba 1 0 0.2812 1>}
  NamedRerouteDeclaration(None=@n1.Result) {name=InMA, nodeColor=<rgba 0.0312 0 1 1>}
  NamedRerouteDeclaration(None=@n9.UV_SideProjection) {name=TriplannarUV_Side, nodeColor=<rgba 0.1953 1 0 1>}
  n16 = SphereMask(
   A=VertexInterpolator(VS=LocalPosition() {localOrigin=InstancePreSkinning}.XYZ).PS
   B=Constant4Vector() {constant=<rgba 0 0 0 1>}.RGB
   Radius=Constant() {r=98.7459}
  )
  NamedRerouteDeclaration(None=@n16) {name=Hand Mask, nodeColor=<rgba 0 0.0156 1 1>}
  n17 = TextureSample() {
   samplerSource=SSM_Wrap_WorldGroupSettings
   samplerType=SAMPLERTYPE_Masks
   texture=/BRCosmetics/Characters/Player/Female/Medium/Bodies/F_MED_BurnLeer/Textures/T_BurnLeer_Body_FX3
  }
  n18 = LinearInterpolate(
   A=Constant() {r=1}
   B=VectorParameter() {group=BurnLeer, parameterName=ReactiveFireColor, defaultValue=<rgba 1 0.4313 0 1>}.RGB
   Alpha=Multiply(
    A=NamedRerouteDeclaration(None=@n17.R) {name=ReactiveFireMask, nodeColor=<rgba 1 0.8906 0 1>}
    B=ScalarParameter() {defaultValue=0.9, group=BurnLeer, parameterName=ReactiveFireIntensity}
   )
  ) {constA=1}
  n19 = TextureSample() {
   samplerSource=SSM_Wrap_WorldGroupSettings
   samplerType=SAMPLERTYPE_Masks
   texture=/BRCosmetics/Characters/Player/Female/Medium/Bodies/F_MED_BurnLeer/Textures/T_BurnLeer_Body_FX4
  }
  n20 = Constant3Vector() {constant=<rgba 0 0.25 0 0>}
  n21 = Constant3Vector() {constant=<rgba 0 0.1 0 0>}
  n22 = NamedRerouteUsage()
  n23 = TextureSample() {
   samplerSource=SSM_Wrap_WorldGroupSettings
   samplerType=SAMPLERTYPE_Masks
   texture=&T_BurnLeer_Body_FX3
  }
  n24 = NamedRerouteUsage()
  n25 = Multiply(
   A=MaterialFunctionCall() {materialFunction=&/MF_BurnLeer_NeckMask}.Result
   B=@n24
  )
  n26 = Multiply(
   A=LinearInterpolate(B=Constant4Vector() {constant=<rgba 1 0.4178 0.3217 1>}.RGB, Alpha=@n19.R) {constA=1}
   B=MaterialFunctionCall!(ReactiveFireSpeed=LinearInterpolate(A=@n20, B=@n21, Alpha=@n22), ReactiveFireMask=LinearInterpolate(A=@n23.R, B=@n25, Alpha=@n24)) {materialFunction=&/MF_BurnLeer_ReactiveFire}.Result
  )
  NamedRerouteDeclaration(None=Power(Base=Multiply(A=@n18, B=@n26), Exp=Constant() {r=1.5})) {name=ReactiveFire, nodeColor=<rgba 0.8281 1 0 1>}
  n27 = NamedRerouteUsage()
  n28 = TextureSampleParameter2D() {
   group=06_Emissive
   parameterName=Emissive
   samplerSource=SSM_Wrap_WorldGroupSettings
   sortPriority=1
   texture=&T_BlacksRGB
  }
  n29 = Multiply(A=@n28.RGB) {constB=1.6}
  n30 = ScalarParameter() {defaultValue=0.85, group=BurnLeer, parameterName=FingerEmissivePower}
  NamedRerouteDeclaration(None=LinearInterpolate(
    A=Add(A=@n27, B=@n29)
    B=VectorParameter() {group=BurnLeer, parameterName=FingerEmissiveColor, defaultValue=<rgba 1 0 0 1>}.RGB
    Alpha=Fresnel(ExponentIn=@n30)
   )) {name=FingerEmissive, nodeColor=<rgba 0.5703 0 1 1>}
  n31 = TextureSampleParameter2D() {
   parameterName=T_Cloth_Fire_Mask
   samplerSource=SSM_Wrap_WorldGroupSettings
   samplerType=SAMPLERTYPE_Masks
   texture=&T_BurnLeer_Body_FX2
  }
  NamedRerouteDeclaration(None=@n31.R) {name=SolidFireMask, nodeColor=<rgba 0.2656 0 1 1>}
  n32 = TextureSample() {
   samplerSource=SSM_Wrap_WorldGroupSettings
   samplerType=SAMPLERTYPE_Masks
   texture=&T_BurnLeer_Body_FX4
  }
  NamedRerouteDeclaration(None=@n32.B) {name=FingerMask, nodeColor=<rgba 0 0.5312 1 1>}
  n33 = Constant()
  n34 = MaterialFunctionCall!(In_FireSmoke_TAMask, In_FireSmoke_UV=Multiply(A=TextureCoordinate(), B=Constant() {r=2}) {constB=10}) {materialFunction=/BRCosmetics/Gadgets/Assets/VinderTech_GliderChute/Glider_RampCanal/Materials/MFW_FlowingSmoke}
  n35 = MaterialFunctionCall!(
   Low _ PC Low and Med=@n33
   "High _ XB1, PS4"=@n34.Result
   "Epic _ PC High, XSX, PS5"=@n34.Result
   --------
   "Mobile Low _ PC Perf, Android"=@n33
   Mobile Med _ Switch=@n33
  ) {materialFunction=/Game/Global/MaterialFunctions/TechArt/MF_Fortnite_QualitySwitch_Vector3}
  n36 = OneMinus(None=Fresnel(ExponentIn=Constant() {r=0.0751}))
  n37 = Fresnel(ExponentIn=Constant() {r=1.0576})
  n38 = Constant() {r=1.5}
  n39 = TextureSampleParameter2D() {
   group=00_Texture Inputs
   parameterName=Mask
   samplerSource=SSM_Wrap_WorldGroupSettings
   samplerType=SAMPLERTYPE_Masks
   sortPriority=3
   texture=&T_BlackMask
  }
  n40 = TextureSample(UVs=Panner(Coordinate=NamedRerouteUsage()) {speedX=0.02, speedY=0.08}) {samplerSource=SSM_Wrap_WorldGroupSettings, samplerType=SAMPLERTYPE_Grayscale, texture=&LowResBlurredNoise}
  n41 = Multiply(A=@n40.R) {constB=2}
  n42 = Constant() {r=1.5}
  n43 = Saturate(None=Multiply(A=@n39.R, B=Power(Base=@n41, Exp=@n42)))
  n44 = Fresnel(ExponentIn=Constant() {r=0.03})
  n45 = Multiply(A=Multiply(A=OneMinus(None=@n44), B=NamedRerouteUsage()), B=Constant() {r=31.4781})
  n46 = Constant() {r=0.9878}
  n47 = Multiply(A=@n43, B=Saturate(None=Power(Base=@n45, Exp=@n46)))
  n48 = Multiply(A=TextureCoordinate(), B=Constant() {r=2}) {constB=2}
  n49 = Multiply(A=NamedRerouteUsage(), B=Constant() {r=2}) {constB=2}
  n50 = MaterialFunctionCall() {materialFunction=/BRCosmetics/Characters/Player/Male/Medium/Bodies/M_MED_BirchPalace/Skins/Gas/Materials/MF_TriplanarBlendMask}
  n51 = LinearInterpolate(
   A=TextureSample(UVs=Panner(Coordinate=@n48) {speedY=0.005}) {samplerType=SAMPLERTYPE_Masks, texture=/Game/Global/Textures/Default/Noise/NoisePattern}.R
   B=TextureSample(UVs=Panner(Coordinate=@n49) {speedY=0.005}) {samplerType=SAMPLERTYPE_Masks, texture=&NoisePattern}.R
   Alpha=Saturate(None=Multiply(A=@n50.AxisMask_X) {constB=2})
  )
  n52 = Constant() {r=3}
  n53 = LinearInterpolate(B=Constant() {r=13.3538}, Alpha=Saturate(None=Power(Base=@n51, Exp=@n52))) {constB=2}
  n54 = NamedRerouteUsage()
  n55 = Constant() {r=6.1108}
  n56 = TextureSample(UVs=Panner(Coordinate=Multiply(A=@n54, B=@n55) {constB=2}) {speedX=0.02, speedY=0.3}) {samplerSource=SSM_Wrap_WorldGroupSettings, samplerType=SAMPLERTYPE_Grayscale, texture=&LowResBlurredNoise}
  n57 = Saturate(None=Power(Base=Multiply(A=@n56.R) {constB=2}, Exp=Constant() {r=1.5}))
  n58 = NamedRerouteUsage()
  n59 = Constant() {r=3}
  n60 = TextureSample(UVs=BumpOffset(Coordinate=Multiply(A=@n58, B=@n59) {constB=2}, Height=Constant() {r=-0.05})) {samplerSource=SSM_Wrap_WorldGroupSettings, samplerType=SAMPLERTYPE_Grayscale, texture=&LowResBlurredNoise}
  n61 = Power(Base=Multiply(A=@n60.R, B=Constant() {r=1.7084}), Exp=Constant() {r=1.0532})
  n62 = Add(A=Saturate(None=Multiply(A=@n53, B=@n57)), B=LinearInterpolate(A=Constant() {r=0.15}, Alpha=Saturate(None=@n61)) {constA=0.51})
  n63 = Add(A=Multiply(A=Multiply(A=@n37, B=@n38), B=@n43), B=LinearInterpolate(B=Multiply(A=@n47, B=@n62), Alpha=NamedRerouteUsage()))
  n64 = TextureSampleParameter2D() {
   group=00_Texture Inputs
   parameterName=Diffuse
   samplerSource=SSM_Wrap_WorldGroupSettings
   sortPriority=0
   texture=/Game/Global/Textures/Default/Blanks/T_MidgreysRGB
  }
  n65 = LinearInterpolate(A=ScalarParameter() {defaultValue=5, group=BurnLeer, parameterName=InternalGlowintensity}, B=Constant() {r=2}, Alpha=@n16)
  n66 = TextureSampleParameter2D() {
   group=00_Texture Inputs
   parameterName=Diffuse
   samplerSource=SSM_Wrap_WorldGroupSettings
   sortPriority=0
   texture=&T_MidgreysRGB
  }
  n67 = Add(
   A=Multiply(A=Multiply(A=@n35.Result, B=@n36), B=Constant4Vector() {constant=<rgba 0.3 0.0157 0.0247 1>}.RGB)
   B=Multiply(A=Multiply(A=@n63, B=@n64.RGB), B=Multiply(A=@n65, B=@n66.RGB))
  )
  n68 = NamedRerouteUsage()
  NamedRerouteDeclaration(None=LinearInterpolate(A=Multiply(A=@n67, B=@n68), Alpha=NamedRerouteUsage()) {constB=0}) {name=InternalGlow, nodeColor=<rgba 1 0 0.8203 1>}
  NamedRerouteDeclaration(None=LinearInterpolate(A=Power(Base=@n31.B) {constExponent=1.2}, Alpha=NamedRerouteUsage()) {constB=0}) {name=Burning EdgeMask, nodeColor=<rgba 1 0 0.375 1>}
  NamedRerouteDeclaration(None=Multiply(A=ScalarParameter() {parameterName=FireReactive}, B=ScalarParameter() {defaultValue=1, group=BurnLeer, parameterName=EnableReactive})) {name=FireReactive, nodeColor=<rgba 1 0 0.2344 1>}
  n69 = GetMaterialAttributes(None=NamedRerouteUsage()) {attributeGetTypes [#1, #2, #3]}
  n70 = LinearInterpolate(Alpha=Subtract(A=NamedRerouteUsage(), B=NamedRerouteUsage())) {constA=1, constB=0.6}
  n71 = Multiply(
   A=NamedRerouteUsage()
   B=MaterialFunctionCall() {materialFunction=&/MF_Burnleer_NeckFresnel}.Result
  )
  NamedRerouteDeclaration(None=LinearInterpolate(A=Multiply(A=@n69.Base Color, B=@n70) {constB=0.6}, B=Add(A=@n69.Base Color, B=@n71), Alpha=NamedRerouteUsage())) {name=Out_BaseColor, nodeColor=<rgba 0.3125 1 0 1>}
  n72 = Add(A=NamedRerouteUsage(), B=NamedRerouteUsage()) {constB=0}
  n73 = Add(A=NamedRerouteUsage())
  n74 = TextureSampleParameter2D() {
   group=06_Emissive
   parameterName=Emissive
   samplerSource=SSM_Wrap_WorldGroupSettings
   sortPriority=1
   texture=&T_BlacksRGB
  }
  n75 = OneMinus(None=NamedRerouteUsage())
  n76 = LinearInterpolate(
   A=@n72
   B=Add(A=@n72, B=NamedRerouteUsage())
   Alpha=LinearInterpolate(A=NamedRerouteUsage(), B=Saturate(None=@n73), Alpha=Multiply(A=@n74.R, B=@n75))
  )
  n77 = NamedRerouteUsage()
  n78 = Multiply(A=NamedRerouteUsage()) {constB=3}
  n79 = Saturate(None=Subtract(A=@n32.G, B=Multiply(A=@n77, B=@n78)))
  n80 = SetMaterialAttributes!(
   MaterialAttributes=@n69.MaterialAttributes
   Emissive Color=LinearInterpolate(A=LinearInterpolate(A=@n76, Alpha=@n79) {constB=0}, B=NamedRerouteUsage(), Alpha=NamedRerouteUsage())
   Base Color=NamedRerouteUsage()
   Fuzz Color or Subsurface Color
   Opacity=Constant()
  ) {attributeSetTypes [
    #1
    #3
    #2
    #4
   ]}
  NamedRerouteDeclaration(None=@n80) {name=OutMA, nodeColor=<rgba 1 0.6328 0 1>}
  NamedRerouteDeclaration(None=ComponentMask(None=@n1.M) {b=true}) {name=Skin, nodeColor=<rgba 0 1 0.8125 1>}
  n81 = TextureCoordinate()
  n82 = ScalarParameter() {defaultValue=11.5422, parameterName=SolidFireNoiseTiling}
  n83 = VectorParameter() {parameterName=SolidFireSpeed, defaultValue=<rgba 0 0.4 0 1>}
  n84 = TextureSample(UVs=Panner(Coordinate=Multiply(A=@n81, B=@n82) {constB=11}, Speed=ComponentMask(None=@n83.RGB) {g=true, r=true}) {speedX=0.07, speedY=0.07}) {
   samplerSource=SSM_Wrap_WorldGroupSettings
   samplerType=SAMPLERTYPE_Masks
   texture=&T_BurnLeer_Body_FX3
  }
  n85 = ScalarParameter() {defaultValue=0.01, parameterName=SolidFireNoise}
  n86 = Constant()
  n87 = NamedRerouteUsage()
  n88 = TextureSampleParameter2D() {
   group=06_Emissive
   parameterName=Emissive
   samplerSource=SSM_Wrap_WorldGroupSettings
   sortPriority=1
   texture=&T_BlacksRGB
  }
  n89 = Power(Base=Multiply(A=@n88.RGB, B=Constant() {r=1.1287}) {constB=2}, Exp=Constant() {r=1})
  n90 = Multiply(
   A=ComponentMask(None=@n84.G) {g=true, r=true}
   B=LinearInterpolate(A=LinearInterpolate(A=@n85, B=@n86, Alpha=@n87), Alpha=Saturate(None=@n89)) {constB=0}
  ) {constB=0.05}
  n91 = NamedRerouteUsage()
  n92 = TextureSampleParameter2D(UVs=Add(A=TextureCoordinate(), B=LinearInterpolate(A=@n90, Alpha=@n91) {constB=0})) {
   group=06_Emissive
   parameterName=Emissive
   samplerSource=SSM_Wrap_WorldGroupSettings
   sortPriority=1
   texture=&T_BlacksRGB
  }
  NamedRerouteDeclaration(None=Multiply(A=@n92.RGB) {constB=0.9}) {name=SolidFire, nodeColor=<rgba 0.6875 1 0 1>}
  n93 = TextureCoordinate()
  n94 = Constant() {r=15.9602}
  n95 = Constant4Vector() {constant=<rgba 0 1 0 1>}
  n96 = TextureSample(UVs=Panner(Coordinate=Multiply(A=@n93, B=@n94) {constB=11}, Speed=ComponentMask(None=@n95.RGB) {g=true, r=true}) {speedX=0.07, speedY=0.07}) {samplerSource=SSM_Wrap_WorldGroupSettings, samplerType=SAMPLERTYPE_Grayscale, texture=&LowResBlurredNoise}
  n97 = ComponentMask(None=@n96.G) {g=true, r=true}
  n98 = LinearInterpolate(A=Constant() {r=0.02}, B=Constant() {r=0.002}, Alpha=NamedRerouteUsage()) {constB=0.005}
  n99 = TextureSampleParameter2D(UVs=Add(A=TextureCoordinate(), B=Multiply(A=@n97, B=@n98))) {
   parameterName=T_Cloth_Fire_Mask
   samplerSource=SSM_Wrap_WorldGroupSettings
   samplerType=SAMPLERTYPE_Masks
   texture=&T_BurnLeer_Body_FX2
  }
  n100 = Power(Base=NamedRerouteDeclaration(None=@n99.G) {name=fire, nodeColor=<rgba 0 1 0.4609 1>}, Exp=Constant() {r=2.5})
  n101 = LinearInterpolate(Alpha=ScalarParameter() {parameterName=FireReactive}) {constA=1, constB=6}
  NamedRerouteDeclaration(None=Multiply(A=Multiply(A=@n100, B=@n101), B=Constant4Vector() {constant=<rgba 0.7 0.0561 0 1>}.RGB)) {name=FireFlame, nodeColor=<rgba 0.9922 0 1 1>}
  n102 = TextureSampleParameter2D() {
   group=06_Emissive
   parameterName=Emissive
   samplerSource=SSM_Wrap_WorldGroupSettings
   sortPriority=1
   texture=&T_BlacksRGB
  }
  n103 = Power(Base=Multiply(A=@n102.RGB, B=Constant() {r=1.4777}) {constB=11}, Exp=Constant() {r=7.05})
  n104 = Multiply(A=@n102.RGB) {constB=0.3}
  n105 = NamedRerouteUsage()
  n106 = Max(A=Saturate(None=LinearInterpolate(A=@n103, B=@n104, Alpha=@n105)), B=NamedRerouteUsage())
  n107 = NamedRerouteUsage()
  NamedRerouteDeclaration(None=Multiply(A=Multiply(A=@n106, B=@n107), B=VectorParameter() {parameterName=BurningEdge_Brightness, defaultValue=<rgba 25 1.0322 0 1>}.RGB)) {name=Burning Edge, nodeColor=<rgba 0 1 0.9063 1>}
  n108 = Constant()
  n109 = Multiply(A=TextureCoordinate(), B=Constant() {r=15})
  n110 = NamedRerouteUsage()
  n111 = Constant() {r=0.1}
  n112 = TextureSample(UVs=Add(A=Panner(Coordinate=@n109) {speedY=0.2}, B=Multiply(A=@n110, B=@n111) {constB=0.2})) {
   samplerSource=SSM_Wrap_WorldGroupSettings
   samplerType=SAMPLERTYPE_Masks
   texture=&T_BurnLeer_Body_FX3
  }
  n113 = Multiply(A=Multiply(A=@n112.B, B=NamedRerouteUsage()), B=Power(Base=NamedRerouteUsage()))
  n114 = Constant4Vector() {constant=<rgba 10 0.8015 0 1>}
  NamedRerouteDeclaration(None=MaterialFunctionCall(
    Low _ PC Low and Med=@n108
    "High _ XB1, PS4"=@n108
    "Epic _ PC High, XSX, PS5"=Multiply(A=@n113, B=@n114.RGB)
    "Mobile Low _ PC Perf, Android"=@n108
    Mobile Med _ Switch=@n108
   ) {materialFunction=&MF_Fortnite_QualitySwitch_Vector3}.Result) {name=FireEmber, nodeColor=<rgba 1 0.5859 0 1>}
  n115 = TextureCoordinate()
  n116 = Constant() {r=32}
  n117 = TextureSample(UVs=Panner(Coordinate=Multiply(A=@n115, B=@n116) {constB=11}, Speed=Constant() {r=0.3}) {speedX=0.07, speedY=0.07}) {samplerSource=SSM_Wrap_WorldGroupSettings, samplerType=SAMPLERTYPE_Grayscale, texture=&LowResBlurredNoise}
  n118 = Power(Base=Multiply(A=@n117.R) {constB=3})
  NamedRerouteDeclaration(None=LinearInterpolate(A=Saturate(None=@n118), Alpha=@n12)) {name=FireEmberNoise, nodeColor=<rgba 0.2656 0 1 1>}
  NamedRerouteDeclaration(None=ComponentMask(None=@n1.M) {r=true}) {name=AO, nodeColor=<rgba 0 1 0.2266 1>}
  Reroute(None=@n4)
 ]
}
