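# Parent hair material for the female BurnLeer character body, adding a reactive fire glow to the shared 2023 character material stack.
# Character PreFX output blends with the FN Hair 2023 function by hair mask, behind a static switch, with an optional flyaway-hair pass.
# FireReactive, gated by EnableReactive, shifts the fire mask from powered green to red channel and scales intensity.
# The reactive fire result adds to emissive before the shared PostFX function.
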Material /BRCosmetics/Characters/Player/Female/Medium/Bodies/F_MED_BurnLeer/Materials/M_HairParent_BurnLeer_Parent
graph {
 properties {
  bUseMaterialAttributes=true
  bUsedWithMeshDeformer=true
  bUsedWithMorphTargets=true
  bUsedWithSkeletalMesh=true
  pixelDepthOffsetMode=PDOM_Legacy
  shadingModel=MSM_FromMaterialExpression
  usedShadingModels=Default Lit | Subsurface | Clear Coat | Subsurface Profile | Cloth
 }
 nodes [
  n1 = MaterialFunctionCall() {materialFunction=&MF_CharacterParent_2023_PreFX}
  n2 = MaterialFunctionCall!(In=@n1.Result, Mask Input=@n1.Result) {materialFunction=/Game/Global/Materials/Character_Parent_2023/Hair_Parent_2023/MaterialFunctions/MF_FN_Hair_2023}
  n3 = BlendMaterialAttributes(
   A=@n1.Result
   B=StaticSwitchParameter(True=@n2.Result, False=@n1.Result) {defaultValue=true, group=Skin FX, parameterName=UseFNHair2_0}
   Alpha=@n2.HairMask
  )
  n4 = GetMaterialAttributes(None=StaticSwitchParameter(
    True=MaterialFunctionCall!(MaterialInput=@n3) {materialFunction=/Game/Global/Materials/Character_Parent_2023/Hair_Parent_2023/MaterialFunctions/MF_FlyAwayHair_2023}.Result
    False=@n3
   ) {group=Skin FX, parameterName=IsFlyAwayHair?}) {attributeGetTypes [#1]}
  n5 = Multiply(A=@n4.Emissive Color)
  n6 = TextureSampleParameter2D() {
   group=00_Texture Inputs
   parameterName=Mask
   samplerSource=SSM_Wrap_WorldGroupSettings
   samplerType=SAMPLERTYPE_Masks
   sortPriority=3
   texture=&T_BlackMask
  }
  n7 = Power(Base=@n6.G)
  n8 = NamedRerouteUsage()
  n9 = Multiply(
   A=MaterialFunctionCall!(ReactiveFireSpeed, ReactiveFireMask=LinearInterpolate(A=@n7, B=@n6.R, Alpha=@n8)) {materialFunction=&/MF_BurnLeer_ReactiveFire}.Result
   B=LinearInterpolate(A=Constant() {r=0.3}, Alpha=@n8) {constB=0.5}
  )
  MaterialFunctionCall(
   Material=SetMaterialAttributes(MaterialAttributes=@n4.MaterialAttributes, Emissive Color=Add(A=@n5, B=@n9)) {attributeSetTypes [#1]}
   M=@n1.M
   Pre Skinned Local Position=@n1.M
   Distance Blend=@n1.M
   Fresnel=@n1.M
  ) {materialFunction=&MF_CharacterParent_2023_PostFX}
  NamedRerouteDeclaration(None=Multiply(A=ScalarParameter() {parameterName=FireReactive}, B=ScalarParameter() {defaultValue=1, parameterName=EnableReactive})) {name=FireReactive, nodeColor=<rgba 1 0 0.2344 1>}
  Reroute(None=@n1.Tech Art Mask)
 ]
}
