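# Control Rig for the F_MED_BurnLeer female body cosmetic, running shared body deformation, then skirt and collar corrective passes.
# SkirtFix blends skirt masters between thighs and pelvis, layering offsets weighted by spherical pose readers and thigh pitch.
# It also eases skirt bones onto the thighs during GoldenGloryEmote when RBAN is disabled.
# Later passes add constant offsets, interpolated crouch, directional crouch-walk and slide fixes, and clavicle-driven collar rotation.
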ControlRigBlueprint /BRCosmetics/Characters/Player/Female/Medium/Bodies/F_MED_BurnLeer/Meshes/F_MED_BurnLeer_CtrlRig
graphs [
 New Function_ContainedGraph {nodes [
   RigVMFunctionEntryNode()
   > RigVMFunction_Sequence_1 = RigVMAggregateNode()
   &RigUnit_ModifyTransforms(
    ItemToModify=((Item=(Type=Bone,Name="thigh_l"),Transform=(Rotation=(X=0,Y=0,Z=0,W=1),Translation=(X=0,Y=0,Z=0),Scale3D=(X=1,Y=1,Z=1))))
    Weight=1
    WeightMinimum=0
    ExecutePin=@RigVMFunction_Sequence_1.A
   )
   > &RigUnit_ModifyTransforms(
    ItemToModify=((Item=(Type=Bone,Name="thigh_r"),Transform=(Rotation=(X=0,Y=0,Z=0,W=1),Translation=(X=0,Y=0,Z=0),Scale3D=(X=1,Y=1,Z=1))))
    Weight=1
    WeightMinimum=0
   )
   > &RigUnit_ParentConstraint(
    AdvancedSettings=(InterpolationType=Average,RotationOrderForFilter=XZY)
    Child=(Type=Bone,Name="skirt_L_f_Master")
    Filter=(TranslationFilter=(bX=false,bY=false,bZ=false),RotationFilter=(bX=True,bY=True,bZ=True),ScaleFilter=(bX=True,bY=True,bZ=True))
    Parents=((Item=(Type=Bone,Name="thigh_l"),Weight=1),(Item=(Type=Bone,Name="pelvis"),Weight=0.2))
    Weight=1
   )
   > &RigUnit_ParentConstraint(
    AdvancedSettings=(InterpolationType=Average,RotationOrderForFilter=XZY)
    Child=(Type=Bone,Name="dyn_skirt_l_master")
    Filter=(TranslationFilter=(bX=false,bY=false,bZ=false),RotationFilter=(bX=True,bY=True,bZ=True),ScaleFilter=(bX=True,bY=True,bZ=True))
    Parents=((Item=(Type=Bone,Name="thigh_l"),Weight=1),(Item=(Type=Bone,Name="pelvis"),Weight=0.5))
    Weight=1
   )
   > &RigUnit_ParentConstraint(
    AdvancedSettings=(InterpolationType=Average,RotationOrderForFilter=XZY)
    Child=(Type=Bone,Name="skirt_L_b_Master")
    Filter=(TranslationFilter=(bX=false,bY=false,bZ=false),RotationFilter=(bX=True,bY=True,bZ=True),ScaleFilter=(bX=True,bY=True,bZ=True))
    Parents=((Item=(Type=Bone,Name="thigh_l"),Weight=0.5),(Item=(Type=Bone,Name="pelvis"),Weight=0.5))
    Weight=1
   )
   > &RigUnit_ParentConstraint(
    AdvancedSettings=(InterpolationType=Average,RotationOrderForFilter=XZY)
    Child=(Type=Bone,Name="skirt_b_m_Master")
    Filter=(TranslationFilter=(bX=false,bY=false,bZ=false),RotationFilter=(bX=True,bY=True,bZ=True),ScaleFilter=(bX=True,bY=True,bZ=True))
    Parents=((Item=(Type=Bone,Name="thigh_l"),Weight=0.2),(Item=(Type=Bone,Name="pelvis"),Weight=0.5))
    Weight=1
   )
   > &RigUnit_ParentConstraint(
    AdvancedSettings=(InterpolationType=Average,RotationOrderForFilter=XZY)
    Child=(Type=Bone,Name="skirt_r_f_Master")
    Filter=(TranslationFilter=(bX=false,bY=false,bZ=false),RotationFilter=(bX=True,bY=True,bZ=True),ScaleFilter=(bX=True,bY=True,bZ=True))
    Parents=((Item=(Type=Bone,Name="thigh_r"),Weight=1),(Item=(Type=Bone,Name="pelvis"),Weight=0.2))
    Weight=1
   )
   > &RigUnit_ParentConstraint(
    AdvancedSettings=(InterpolationType=Average,RotationOrderForFilter=XZY)
    Child=(Type=Bone,Name="dyn_skirt_r_master")
    Filter=(TranslationFilter=(bX=false,bY=false,bZ=false),RotationFilter=(bX=True,bY=True,bZ=True),ScaleFilter=(bX=True,bY=True,bZ=True))
    Parents=((Item=(Type=Bone,Name="thigh_r"),Weight=0.8),(Item=(Type=Bone,Name="pelvis"),Weight=0.2))
    Weight=1
   )
   > &RigUnit_ParentConstraint(
    AdvancedSettings=(InterpolationType=Average,RotationOrderForFilter=XZY)
    Child=(Type=Bone,Name="skirt_r_b_Master")
    Filter=(TranslationFilter=(bX=false,bY=false,bZ=false),RotationFilter=(bX=True,bY=True,bZ=True),ScaleFilter=(bX=True,bY=True,bZ=True))
    Parents=((Item=(Type=Bone,Name="thigh_r"),Weight=0.4),(Item=(Type=Bone,Name="pelvis"),Weight=0.2))
    Weight=1
   )
   > &RigUnit_ParentConstraint(
    AdvancedSettings=(InterpolationType=Average,RotationOrderForFilter=XZY)
    Child=(Type=Bone,Name="skirt_b_m_Master")
    Filter=(TranslationFilter=(bX=false,bY=false,bZ=false),RotationFilter=(bX=True,bY=True,bZ=True),ScaleFilter=(bX=True,bY=True,bZ=True))
    Parents=((Item=(Type=Bone,Name="thigh_r"),Weight=0.1),(Item=(Type=Bone,Name="pelvis"),Weight=0.2))
    Weight=1
   )
   > RigVMFunction_Sequence = RigVMAggregateNode()
   Multiply = &RigVMFunction_MathDoubleMul(A=RigVMVariableNode() {variable=R_Leg_Rotback}.Value, B=RigVMVariableNode() {variable=L_Leg_Rotback}.Value)
   &RigUnit_ModifyTransforms(
    ItemToModify="((Item=(Type=Bone,Name=\"None\"),Transform=(Rotation=(X=0,Y=0,Z=0,W=1),Translation=(X=0,Y=0,Z=0),Scale3D=(X=1,Y=1,Z=1))),(Item=(Type=Bone,Name=\"skirt_b_m_Master\"),Transform=(Rotation=(X=0,Y=0.766,Z=0,W=0.6428),Translation=(X=0,Y=0,Z=0),Scale3D=(X=1,Y=1,Z=1)) ...(truncated)"
    WeightMinimum=0
    Weight=@Multiply.Result
    ExecutePin=@RigVMFunction_Sequence.C
   )
   RigVMFunction_MathBoolNot = &RigVMFunction_MathBoolNot(Value=RigVMVariableNode() {variable=RBAN Enable}.Value)
   RigVMFunction_MathBoolAnd = &RigVMFunction_MathBoolAnd(B=@RigVMFunction_MathBoolNot.Result, A=RigVMVariableNode() {variable=GoldenGloryEmote}.Value)
   RigVMFunction_MathBoolToFloat = &RigVMFunction_MathBoolToFloat(Value=@RigVMFunction_MathBoolAnd.Result)
   AlphaInterp = &RigVMFunction_AlphaInterp(InterpSpeedDecreasing=5, InterpSpeedIncreasing=5, Scale=1, bClampResult=true, Value=@RigVMFunction_MathBoolToFloat.Result)
   > &RigUnit_ParentConstraint(
    AdvancedSettings=(InterpolationType=Average,RotationOrderForFilter=XZY)
    Child=(Type=Bone,Name="dyn_skirt_r_f_01")
    Parents=((Item=(Type=Bone,Name="thigh_twist_01_r"),Weight=1))
    Weight=@AlphaInterp.Result
   )
   > &RigUnit_ParentConstraint(
    AdvancedSettings=(InterpolationType=Average,RotationOrderForFilter=XZY)
    Child=(Type=Bone,Name="dyn_skirt_r_f_02")
    Parents=((Item=(Type=Bone,Name="thigh_twist_01_r"),Weight=1))
    Weight=@AlphaInterp.Result
   )
   > &RigUnit_ParentConstraint(
    AdvancedSettings=(InterpolationType=Average,RotationOrderForFilter=XZY)
    Child=(Type=Bone,Name="dyn_skirt_r_f_03")
    Parents=((Item=(Type=Bone,Name="thigh_twist_01_r"),Weight=1))
    Weight=@AlphaInterp.Result
   )
   > &RigUnit_ParentConstraint(
    AdvancedSettings=(InterpolationType=Average,RotationOrderForFilter=XZY)
    Child=(Type=Bone,Name="dyn_skirt_r_01")
    Parents=((Item=(Type=Bone,Name="thigh_r"),Weight=1))
    Weight=@AlphaInterp.Result
   )
   > &RigUnit_ParentConstraint(
    AdvancedSettings=(InterpolationType=Average,RotationOrderForFilter=XZY)
    Child=(Type=Bone,Name="dyn_skirt_r_02")
    Parents=((Item=(Type=Bone,Name="thigh_twist_01_r"),Weight=1))
    Weight=@AlphaInterp.Result
   )
   > &RigUnit_ParentConstraint(
    AdvancedSettings=(InterpolationType=Average,RotationOrderForFilter=XZY)
    Child=(Type=Bone,Name="dyn_skirt_r_03")
    Parents=((Item=(Type=Bone,Name="thigh_twist_01_r"),Weight=1))
    Weight=@AlphaInterp.Result
   )
   > &RigUnit_ParentConstraint(
    AdvancedSettings=(InterpolationType=Average,RotationOrderForFilter=XZY)
    Child=(Type=Bone,Name="dyn_chain_02")
    Parents=((Item=(Type=Bone,Name="thigh_twist_01_r"),Weight=1))
    Weight=@AlphaInterp.Result
   )
   > &RigUnit_ParentConstraint(
    AdvancedSettings=(InterpolationType=Average,RotationOrderForFilter=XZY)
    Child=(Type=Bone,Name="dyn_skirt_l_f1")
    Filter=(TranslationFilter=(bX=True,bY=True,bZ=True),RotationFilter=(bX=False,bY=False,bZ=False),ScaleFilter=(bX=True,bY=True,bZ=True))
    Parents=((Item=(Type=Bone,Name="thigh_twist_01_l"),Weight=1))
    Weight=@AlphaInterp.Result
   )
   > &RigUnit_ParentConstraint(
    AdvancedSettings=(InterpolationType=Average,RotationOrderForFilter=XZY)
    Child=(Type=Bone,Name="dyn_skirt_l_f2")
    Parents=((Item=(Type=Bone,Name="thigh_twist_01_l"),Weight=1))
    Weight=@AlphaInterp.Result
   )
   > &RigUnit_ParentConstraint(
    AdvancedSettings=(InterpolationType=Average,RotationOrderForFilter=XZY)
    Child=(Type=Bone,Name="dyn_skirt_l_f3")
    Parents=((Item=(Type=Bone,Name="thigh_twist_01_l"),Weight=1))
    Weight=@AlphaInterp.Result
   )
   > &RigUnit_ParentConstraint(
    AdvancedSettings=(InterpolationType=Average,RotationOrderForFilter=XZY)
    Child=(Type=Bone,Name="dyn_skirt_l_01")
    Parents=((Item=(Type=Bone,Name="thigh_twist_01_l"),Weight=1))
    Weight=@AlphaInterp.Result
   )
   > &RigUnit_ParentConstraint(
    AdvancedSettings=(InterpolationType=Average,RotationOrderForFilter=XZY)
    Child=(Type=Bone,Name="dyn_skirt_l_02")
    Parents=((Item=(Type=Bone,Name="thigh_twist_01_l"),Weight=1))
    Weight=@AlphaInterp.Result
   )
   > &RigUnit_ParentConstraint(
    AdvancedSettings=(InterpolationType=Average,RotationOrderForFilter=XZY)
    Child=(Type=Bone,Name="dyn_skirt_l_03")
    Parents=((Item=(Type=Bone,Name="thigh_twist_01_l"),Weight=1))
    Weight=@AlphaInterp.Result
   )
   > &RigUnit_ModifyTransforms(
    ItemToModify="((Item=(Type=Bone,Name=\"dyn_skirt_l_master\"),Transform=(Rotation=(X=0,Y=0,Z=0,W=1),Translation=(X=0,Y=0,Z=1.3593),Scale3D=(X=1,Y=1,Z=1))),(Item=(Type=Bone,Name=\"skirt_L_b_Master\"),Transform=(Rotation=(X=0,Y=0,Z=0,W=1),Translation=(X=0,Y=0,Z=1.3663),Scale3D ...(truncated)"
    WeightMinimum=0
    Weight=@AlphaInterp.Result
   )
   RigVMFunctionReturnNode(ExecuteContext=@RigVMFunction_Sequence.D)
   n1 = &RigUnit_SphericalPoseReader(
    ActiveRegionScaleFactors=(PositiveWidth=1,NegativeWidth=1,PositiveHeight=1,NegativeHeight=1)
    ActiveRegionSize=0.02
    Debug=(bDrawDebug=false,bDraw2D=False,bDrawLocalAxes=False,DebugScale=25,DebugSegments=20,DebugThickness=0.25)
    DriverAxis=(X=1,Y=0,Z=0)
    DriverItem=(Type=Bone,Name="thigh_r")
    FalloffRegionScaleFactors=(PositiveWidth=1,NegativeWidth=1,PositiveHeight=1,NegativeHeight=1)
    FalloffSize=0.6
    FlipHeightScaling=true
    FlipWidthScaling=true
    OptionalParentItem=(Type=Bone,Name="None")
    RotationOffset=(X=-90,Y=0,Z=-20)
    ExecutePin=@RigVMFunction_Sequence.A
   )
   > RigVMVariableNode(Value=@n1.OutputParam) {variable=R Leg Front SPR}
   > &RigUnit_ModifyTransforms(
    ItemToModify="((Item=(Type=Bone,Name=\"None\"),Transform=(Rotation=(X=0,Y=0,Z=0,W=1),Translation=(X=0,Y=0,Z=0),Scale3D=(X=1,Y=1,Z=1))),(Item=(Type=Bone,Name=\"skirt_r_f_Master\"),Transform=(Rotation=(X=0.311,Y=0.0197,Z=-0.4405,W=-0.842),Translation=(X=-0.9265,Y=0.7182,Z=3.4 ...(truncated)"
    WeightMinimum=0
    Weight=@n1.OutputParam
   )
   > n2 = &RigUnit_SphericalPoseReader(
    ActiveRegionScaleFactors=(PositiveWidth=1,NegativeWidth=1,PositiveHeight=1,NegativeHeight=1)
    ActiveRegionSize=0.05
    Debug=(bDrawDebug=false,bDraw2D=false,bDrawLocalAxes=False,DebugScale=25,DebugSegments=20,DebugThickness=0.25)
    DriverAxis=(X=1,Y=0,Z=0)
    DriverItem=(Type=Bone,Name="thigh_r")
    FalloffRegionScaleFactors=(PositiveWidth=1,NegativeWidth=1.075,PositiveHeight=1,NegativeHeight=1)
    FalloffSize=0.6
    FlipHeightScaling=true
    FlipWidthScaling=true
    OptionalParentItem=(Type=Bone,Name="None")
    RotationOffset=(X=0,Y=20,Z=0)
   )
   RigVMFunction_AnimEvalRichCurve = &RigVMFunction_AnimEvalRichCurve(
    Curve="(EditorCurveData=(Keys=((InterpMode=RCIM_Cubic,TangentMode=RCTM_User,ArriveTangent=1.3147,LeaveTangent=1.0898),(InterpMode=RCIM_Cubic,TangentMode=RCTM_User,Time=0.5,Value=0.525,ArriveTangent=0.9879,LeaveTangent=0.9879),(InterpMode=RCIM_Cubic,TangentMode=RC ...(truncated)"
    SourceMinimum=0
    TargetMinimum=0
    Value=@n2.OutputParam
   )
   > &RigUnit_ModifyTransforms(
    ItemToModify="((Item=(Type=Bone,Name=\"None\"),Transform=(Rotation=(X=0,Y=0,Z=0,W=1),Translation=(X=0,Y=0,Z=0),Scale3D=(X=1,Y=1,Z=1))),(Item=(Type=Bone,Name=\"skirt_r_f_Master\"),Transform=(Rotation=(X=0.0004,Y=-0.0953,Z=-0.0047,W=-0.9954),Translation=(X=0,Y=0,Z=0),Scale3D= ...(truncated)"
    WeightMinimum=0
    Weight=@RigVMFunction_AnimEvalRichCurve.Result
   )
   > n3 = &RigUnit_SphericalPoseReader(
    ActiveRegionScaleFactors=(PositiveWidth=1,NegativeWidth=1,PositiveHeight=1,NegativeHeight=1)
    ActiveRegionSize=0.05
    Debug=(bDrawDebug=false,bDraw2D=False,bDrawLocalAxes=False,DebugScale=25,DebugSegments=20,DebugThickness=0.25)
    DriverAxis=(X=1,Y=0,Z=0)
    DriverItem=(Type=Bone,Name="thigh_r")
    FalloffRegionScaleFactors=(PositiveWidth=1,NegativeWidth=1,PositiveHeight=1,NegativeHeight=1)
    FalloffSize=0.6
    FlipHeightScaling=true
    FlipWidthScaling=true
    OptionalParentItem=(Type=Bone,Name="None")
    RotationOffset=(X=90,Y=0,Z=20)
   )
   > RigVMVariableNode(Value=@n3.OutputParam) {variable=R_Leg_Rotback}
   > &RigUnit_ModifyTransforms(
    ItemToModify="((Item=(Type=Bone,Name=\"None\"),Transform=(Rotation=(X=0,Y=0,Z=0,W=1),Translation=(X=0,Y=0,Z=0),Scale3D=(X=1,Y=1,Z=1))),(Item=(Type=Bone,Name=\"skirt_r_f_Master\"),Transform=(Rotation=(X=-0.0026,Y=-0.0982,Z=0.0336,W=0.9946),Translation=(X=-0.3289,Y=1.1882,Z=9 ...(truncated)"
    WeightMinimum=0
    Weight=@n3.OutputParam
   )
   > n4 = &RigUnit_SphericalPoseReader(
    ActiveRegionScaleFactors=(PositiveWidth=1,NegativeWidth=1,PositiveHeight=1,NegativeHeight=1)
    ActiveRegionSize=0.05
    Debug=(bDrawDebug=false,bDraw2D=False,bDrawLocalAxes=False,DebugScale=25,DebugSegments=20,DebugThickness=0.25)
    DriverAxis=(X=1,Y=0,Z=0)
    DriverItem=(Type=Bone,Name="thigh_r")
    FalloffRegionScaleFactors=(PositiveWidth=1,NegativeWidth=1,PositiveHeight=1,NegativeHeight=1)
    FalloffSize=0.6
    FlipHeightScaling=true
    FlipWidthScaling=true
    OptionalParentItem=(Type=Bone,Name="None")
    RotationOffset=(X=-40,Y=20,Z=0)
   )
   > RigVMVariableNode(Value=@n4.OutputParam) {variable=R Leg Front Mid SPR}
   > &RigUnit_ModifyTransforms(
    ItemToModify="((Item=(Type=Bone,Name=\"None\"),Transform=(Rotation=(X=0,Y=0,Z=0,W=1),Translation=(X=0,Y=0,Z=0),Scale3D=(X=1,Y=1,Z=1))),(Item=(Type=Bone,Name=\"skirt_r_f_Master\"),Transform=(Rotation=(X=0,Y=0,Z=0,W=1),Translation=(X=0,Y=0,Z=0),Scale3D=(X=1,Y=1,Z=1))),(Item=( ...(truncated)"
    WeightMinimum=0
    Weight=@n4.OutputParam
   )
   > n5 = &RigUnit_SphericalPoseReader(
    ActiveRegionScaleFactors=(PositiveWidth=1,NegativeWidth=1,PositiveHeight=1,NegativeHeight=1)
    ActiveRegionSize=0.05
    Debug=(bDrawDebug=false,bDraw2D=False,bDrawLocalAxes=False,DebugScale=25,DebugSegments=20,DebugThickness=0.25)
    DriverAxis=(X=1,Y=0,Z=0)
    DriverItem=(Type=Bone,Name="thigh_r")
    FalloffRegionScaleFactors=(PositiveWidth=1,NegativeWidth=1,PositiveHeight=1,NegativeHeight=1)
    FalloffSize=0.6
    FlipHeightScaling=true
    FlipWidthScaling=true
    OptionalParentItem=(Type=Bone,Name="None")
    RotationOffset=(X=40,Y=20,Z=0)
   )
   > &RigUnit_ModifyTransforms(
    ItemToModify="((Item=(Type=Bone,Name=\"None\"),Transform=(Rotation=(X=0,Y=0,Z=0,W=1),Translation=(X=0,Y=0,Z=0),Scale3D=(X=1,Y=1,Z=1))),(Item=(Type=Bone,Name=\"skirt_r_f_Master\"),Transform=(Rotation=(X=0,Y=0,Z=0,W=1),Translation=(X=0,Y=0,Z=0),Scale3D=(X=1,Y=1,Z=1))),(Item=( ...(truncated)"
    WeightMinimum=0
    Weight=@n5.OutputParam
   )
   > n6 = &RigUnit_SphericalPoseReader(
    ActiveRegionScaleFactors=(PositiveWidth=1,NegativeWidth=1,PositiveHeight=1,NegativeHeight=1)
    ActiveRegionSize=0.1
    Debug=(bDrawDebug=false,bDraw2D=false,bDrawLocalAxes=False,DebugScale=25,DebugSegments=20,DebugThickness=0.25)
    DriverAxis=(X=1,Y=0,Z=0)
    DriverItem=(Type=Bone,Name="thigh_r")
    FalloffRegionScaleFactors=(PositiveWidth=1,NegativeWidth=1,PositiveHeight=1,NegativeHeight=1)
    FalloffSize=0.23
    FlipHeightScaling=true
    FlipWidthScaling=true
    OptionalParentItem=(Type=Bone,Name="None")
    RotationOffset=(X=0,Y=-130,Z=0)
   )
   > &RigUnit_ModifyTransforms(
    ItemToModify="((Item=(Type=Bone,Name=\"None\"),Transform=(Rotation=(X=0,Y=0,Z=0,W=1),Translation=(X=0,Y=0,Z=0),Scale3D=(X=1,Y=1,Z=1))),(Item=(Type=Bone,Name=\"skirt_r_f_Master\"),Transform=(Rotation=(X=-0.0046,Y=0.0338,Z=-0.1358,W=-0.9902),Translation=(X=-0.0785,Y=-0.022,Z= ...(truncated)"
    WeightMinimum=0
    Weight=@n6.OutputParam
   )
   n7 = &RigUnit_SphericalPoseReader(
    ActiveRegionScaleFactors=(PositiveWidth=1,NegativeWidth=1,PositiveHeight=1,NegativeHeight=1)
    ActiveRegionSize=0.05
    Debug=(bDrawDebug=false,bDraw2D=False,bDrawLocalAxes=False,DebugScale=25,DebugSegments=20,DebugThickness=0.25)
    DriverAxis=(X=-1,Y=0,Z=0)
    DriverItem=(Type=Bone,Name="thigh_l")
    FalloffRegionScaleFactors=(PositiveWidth=1,NegativeWidth=1,PositiveHeight=1,NegativeHeight=1)
    FalloffSize=0.6
    FlipHeightScaling=true
    FlipWidthScaling=true
    OptionalParentItem=(Type=Bone,Name="None")
    RotationOffset=(X=90,Y=0,Z=-20)
    ExecutePin=@RigVMFunction_Sequence.B
   )
   > RigVMVariableNode(Value=@n7.OutputParam) {variable=L Leg Front SPR}
   > &RigUnit_ModifyTransforms(
    ItemToModify="((Item=(Type=Bone,Name=\"skirt_L_f_Master\"),Transform=(Rotation=(X=0.1755,Y=0.2769,Z=-0.4331,W=0.8396),Translation=(X=-0.0597,Y=-0.2423,Z=6.2361),Scale3D=(X=1,Y=1,Z=1))),(Item=(Type=Bone,Name=\"dyn_skirt_l_master\"),Transform=(Rotation=(X=0.0188,Y=0.0046,Z=-0 ...(truncated)"
    WeightMinimum=0
    Weight=@n7.OutputParam
   )
   > n8 = &RigUnit_SphericalPoseReader(
    ActiveRegionScaleFactors=(PositiveWidth=1,NegativeWidth=1,PositiveHeight=1,NegativeHeight=1)
    ActiveRegionSize=0.05
    Debug=(bDrawDebug=false,bDraw2D=false,bDrawLocalAxes=False,DebugScale=25,DebugSegments=20,DebugThickness=0.25)
    DriverAxis=(X=-1,Y=0,Z=0)
    DriverItem=(Type=Bone,Name="thigh_l")
    FalloffRegionScaleFactors=(PositiveWidth=1,NegativeWidth=1.075,PositiveHeight=1,NegativeHeight=1)
    FalloffSize=0.6
    FlipHeightScaling=true
    FlipWidthScaling=true
    OptionalParentItem=(Type=Bone,Name="None")
    RotationOffset=(X=180,Y=20,Z=0)
   )
   n9 = &RigVMFunction_AnimEvalRichCurve(
    Curve="(EditorCurveData=(Keys=((InterpMode=RCIM_Cubic,TangentMode=RCTM_User,ArriveTangent=1.3147,LeaveTangent=1.0898),(InterpMode=RCIM_Cubic,TangentMode=RCTM_User,Time=0.5,Value=0.525,ArriveTangent=0.9879,LeaveTangent=0.9879),(InterpMode=RCIM_Cubic,TangentMode=RC ...(truncated)"
    SourceMinimum=0
    TargetMinimum=0
    Value=@n8.OutputParam
   )
   > &RigUnit_ModifyTransforms(
    ItemToModify="((Item=(Type=Bone,Name=\"skirt_L_f_Master\"),Transform=(Rotation=(X=0,Y=0,Z=-0.0162,W=0.9999),Translation=(X=0,Y=0,Z=0),Scale3D=(X=1,Y=1,Z=1))),(Item=(Type=Bone,Name=\"dyn_skirt_l_master\"),Transform=(Rotation=(X=0.0286,Y=-0.3212,Z=-0.0099,W=0.9465),Translatio ...(truncated)"
    WeightMinimum=0
    Weight=@n9.Result
   )
   > n10 = &RigUnit_SphericalPoseReader(
    ActiveRegionScaleFactors=(PositiveWidth=1,NegativeWidth=1,PositiveHeight=1,NegativeHeight=1)
    ActiveRegionSize=0.05
    Debug=(bDrawDebug=false,bDraw2D=False,bDrawLocalAxes=False,DebugScale=25,DebugSegments=20,DebugThickness=0.25)
    DriverAxis=(X=-1,Y=0,Z=0)
    DriverItem=(Type=Bone,Name="thigh_l")
    FalloffRegionScaleFactors=(PositiveWidth=1,NegativeWidth=1,PositiveHeight=1,NegativeHeight=1)
    FalloffSize=0.6
    FlipHeightScaling=true
    FlipWidthScaling=true
    OptionalParentItem=(Type=Bone,Name="None")
    RotationOffset=(X=-90,Y=0,Z=20)
   )
   > RigVMVariableNode(Value=@n10.OutputParam) {variable=L_Leg_Rotback}
   > &RigUnit_ModifyTransforms(
    ItemToModify="((Item=(Type=Bone,Name=\"skirt_L_f_Master\"),Transform=(Rotation=(X=0,Y=0,Z=-0.0174,W=0.9998),Translation=(X=0,Y=0,Z=5.2239),Scale3D=(X=1,Y=1,Z=1))),(Item=(Type=Bone,Name=\"dyn_skirt_l_master\"),Transform=(Rotation=(X=0,Y=0,Z=0,W=1),Translation=(X=0,Y=0,Z=0),S ...(truncated)"
    WeightMinimum=0
    Weight=@n10.OutputParam
   )
   > n11 = &RigUnit_SphericalPoseReader(
    ActiveRegionScaleFactors=(PositiveWidth=1,NegativeWidth=1,PositiveHeight=1,NegativeHeight=1)
    ActiveRegionSize=0.05
    Debug=(bDrawDebug=false,bDraw2D=False,bDrawLocalAxes=False,DebugScale=25,DebugSegments=20,DebugThickness=0.25)
    DriverAxis=(X=-1,Y=0,Z=0)
    DriverItem=(Type=Bone,Name="thigh_l")
    FalloffRegionScaleFactors=(PositiveWidth=1,NegativeWidth=1,PositiveHeight=1,NegativeHeight=1)
    FalloffSize=0.6
    FlipHeightScaling=true
    FlipWidthScaling=true
    OptionalParentItem=(Type=Bone,Name="None")
    RotationOffset=(X=140,Y=30,Z=0)
   )
   > RigVMVariableNode(Value=@n11.OutputParam) {variable=L Leg Front Mid SPR}
   > &RigUnit_ModifyTransforms(
    ItemToModify="((Item=(Type=Bone,Name=\"skirt_L_f_Master\"),Transform=(Rotation=(X=0.1515,Y=0.0062,Z=-0.0529,W=0.987),Translation=(X=-0.7409,Y=-4.3717,Z=0.8569),Scale3D=(X=1,Y=1,Z=1))),(Item=(Type=Bone,Name=\"dyn_skirt_l_master\"),Transform=(Rotation=(X=0.0513,Y=-0.0389,Z=0. ...(truncated)"
    WeightMinimum=0
    Weight=@n11.OutputParam
   )
   > n12 = &RigUnit_SphericalPoseReader(
    ActiveRegionScaleFactors=(PositiveWidth=1,NegativeWidth=1,PositiveHeight=1,NegativeHeight=1)
    ActiveRegionSize=0.05
    Debug=(bDrawDebug=false,bDraw2D=False,bDrawLocalAxes=False,DebugScale=25,DebugSegments=20,DebugThickness=0.25)
    DriverAxis=(X=-1,Y=0,Z=0)
    DriverItem=(Type=Bone,Name="thigh_l")
    FalloffRegionScaleFactors=(PositiveWidth=1,NegativeWidth=1,PositiveHeight=1,NegativeHeight=1)
    FalloffSize=0.6
    FlipHeightScaling=true
    FlipWidthScaling=true
    OptionalParentItem=(Type=Bone,Name="None")
    RotationOffset=(X=-140,Y=30,Z=0)
   )
   > &RigUnit_ModifyTransforms(
    ItemToModify="((Item=(Type=Bone,Name=\"skirt_L_f_Master\"),Transform=(Rotation=(X=0.2331,Y=-0.0719,Z=-0.0596,W=0.968),Translation=(X=0,Y=0,Z=0),Scale3D=(X=1,Y=1,Z=1))),(Item=(Type=Bone,Name=\"dyn_skirt_l_master\"),Transform=(Rotation=(X=-0.0035,Y=0.0331,Z=-0.2262,W=0.9735), ...(truncated)"
    WeightMinimum=0
    Weight=@n12.OutputParam
   )
   > n13 = &RigUnit_SphericalPoseReader(
    ActiveRegionScaleFactors=(PositiveWidth=1,NegativeWidth=1,PositiveHeight=1,NegativeHeight=1)
    ActiveRegionSize=0.1
    Debug=(bDrawDebug=false,bDraw2D=false,bDrawLocalAxes=False,DebugScale=25,DebugSegments=20,DebugThickness=0.25)
    DriverAxis=(X=-1,Y=0,Z=0)
    DriverItem=(Type=Bone,Name="thigh_l")
    FalloffRegionScaleFactors=(PositiveWidth=1,NegativeWidth=1,PositiveHeight=1,NegativeHeight=1)
    FalloffSize=0.23
    FlipHeightScaling=true
    FlipWidthScaling=true
    OptionalParentItem=(Type=Bone,Name="None")
    RotationOffset=(X=0,Y=42,Z=0)
   )
   > &RigUnit_ModifyTransforms(
    ItemToModify="((Item=(Type=Bone,Name=\"skirt_L_f_Master\"),Transform=(Rotation=(X=-0.0028,Y=-0.0211,Z=-0.1769,W=0.984),Translation=(X=0,Y=0,Z=0.8982),Scale3D=(X=1,Y=1,Z=1))),(Item=(Type=Bone,Name=\"dyn_skirt_l_master\"),Transform=(Rotation=(X=0.0001,Y=-0.1803,Z=-0.0001,W=0. ...(truncated)"
    WeightMinimum=0
    Weight=@n13.OutputParam
   )
   Get Delta Ref_1 = RigVMFunctionReferenceNode(Item=(Type=Bone,Name="thigh_l"), ExecuteContext=@RigVMFunction_Sequence_1.B)
   > &RigUnit_ModifyTransforms(
    ItemToModify="((Item=(Type=Bone,Name=\"dyn_skirt_l_master\"),Transform=(Rotation=(X=0,Y=-0.0872,Z=0,W=0.9962),Translation=(X=0,Y=0,Z=0),Scale3D=(X=1,Y=1,Z=1))),(Item=(Type=Bone,Name=\"skirt_L_f_Master\"),Transform=(Rotation=(X=-0.0073,Y=-0.0864,Z=-0.0844,W=0.9927),Translati ...(truncated)"
    WeightMinimum=0
    Weight=&RigVMFunction_MathDoubleRemap(SourceMaximum=-50, SourceMinimum=0, TargetMaximum=1, TargetMinimum=0, Value=@"Get Delta Ref_1".Result.Rotation.Pitch).Result
   )
   > Get Delta Ref_2 = RigVMFunctionReferenceNode(Item=(Type=Bone,Name="thigh_r"))
   > &RigUnit_ModifyTransforms(
    ItemToModify="((Item=(Type=Bone,Name=\"dyn_skirt_r_master\"),Transform=(Rotation=(X=0,Y=-0.0872,Z=0,W=0.9962),Translation=(X=0,Y=0,Z=0),Scale3D=(X=1,Y=1,Z=1))),(Item=(Type=Bone,Name=\"skirt_r_b_Master\"),Transform=(Rotation=(X=0,Y=0,Z=-0.2344,W=0.9722),Translation=(X=0,Y=0, ...(truncated)"
    WeightMinimum=0
    Weight=&RigVMFunction_MathDoubleRemap(SourceMaximum=-50, SourceMinimum=0, TargetMaximum=1, TargetMinimum=0, Value=@"Get Delta Ref_2".Result.Rotation.Pitch).Result
   )
   &RigUnit_ModifyTransforms(
    ExecutePin=()
    ItemToModify="((Item=(Type=Bone,Name=\"None\"),Transform=(Rotation=(X=0,Y=0,Z=0,W=1),Translation=(X=0,Y=0,Z=0),Scale3D=(X=1,Y=1,Z=1))),(Item=(Type=Bone,Name=\"skirt_r_f_Master\"),Transform=(Rotation=(X=0.311,Y=0.0197,Z=-0.4405,W=-0.842),Translation=(X=-0.9265,Y=0.7182,Z=3.4 ...(truncated)"
    WeightMinimum=0
    Weight=@n1.OutputParam
   )
  ]}
 New Function_ContainedGraph {nodes [
   n1 = RigVMFunctionEntryNode()
   RigVMFunction_Sequence = &RigVMFunction_Sequence(ExecuteContext=@n1.ExecuteContext)
   &RigUnit_ModifyTransforms(
    ItemToModify="((Item=(Type=Bone,Name=\"skirt_b_m_Master\"),Transform=(Rotation=(X=0,Y=-0.0057,Z=0,W=-1),Translation=(X=0.022,Y=0.0001,Z=-0.2227),Scale3D=(X=1,Y=1,Z=1))),(Item=(Type=Bone,Name=\"skirt_r_b_Master\"),Transform=(Rotation=(X=0,Y=-0.0226,Z=0,W=0.9997),Translation= ...(truncated)"
    Weight=1
    WeightMinimum=0
    ExecutePin=@RigVMFunction_Sequence.A
   )
   RigVMFunctionReturnNode(ExecuteContext=@RigVMFunction_Sequence.B)
   &RigUnit_ModifyTransforms(
    ExecutePin=()
    ItemToModify="((Item=(Type=Bone,Name=\"skirt_b_m_Master\"),Transform=(Rotation=(X=0,Y=-0.0159,Z=0,W=0.9999),Translation=(X=-0.0269,Y=0,Z=1.3191),Scale3D=(X=1,Y=1,Z=1))),(Item=(Type=Bone,Name=\"skirt_r_b_Master\"),Transform=(Rotation=(X=0,Y=-0.0226,Z=0,W=0.9997),Translation= ...(truncated)"
    Weight=1
    WeightMinimum=0
   )
  ]}
 New Function_ContainedGraph {nodes [
   RigVMFunctionEntryNode()
   > RigVMFunction_Sequence = RigVMAggregateNode()
   AlphaInterp = &RigVMFunction_AlphaInterp(InterpSpeedDecreasing=10, InterpSpeedIncreasing=10, Scale=1, bClampResult=False, Value=RigVMVariableNode() {variable=OnlyCrouching}.Value)
   &RigUnit_ModifyTransforms(
    ItemToModify="((Item=(Type=Bone,Name=\"skirt_b_m_Master\"),Transform=(Rotation=(X=-0.0049,Y=-0.0433,Z=-0.1131,W=0.9926),Translation=(X=0.0114,Y=-0.0026,Z=-0.1343),Scale3D=(X=1,Y=1,Z=1))),(Item=(Type=Bone,Name=\"skirt_L_b_Master\"),Transform=(Rotation=(X=0,Y=-0.14,Z=0,W=0.99 ...(truncated)"
    WeightMinimum=0
    Weight=@AlphaInterp.Result
    ExecutePin=@RigVMFunction_Sequence.A
   )
   Multiply = &RigVMFunction_MathDoubleMul(B=RigVMVariableNode() {variable=MoveForward}.Value, A=RigVMVariableNode() {variable=CrouchWalk}.Value)
   n1 = &RigVMFunction_AlphaInterp(InterpSpeedDecreasing=8, InterpSpeedIncreasing=10, Scale=1, bClampResult=False, Value=@Multiply.Result)
   > &RigUnit_ModifyTransforms(
    ItemToModify="((Item=(Type=Bone,Name=\"skirt_b_m_Master\"),Transform=(Rotation=(X=0.0019,Y=0.0436,Z=-0.0436,W=0.9981),Translation=(X=0,Y=0,Z=0),Scale3D=(X=1,Y=1,Z=1))),(Item=(Type=Bone,Name=\"skirt_L_b_Master\"),Transform=(Rotation=(X=0.0038,Y=-0.0435,Z=0.0871,W=0.9952),Tra ...(truncated)"
    WeightMinimum=0
    Weight=@n1.Result
   )
   > &RigUnit_ModifyTransforms(
    ItemToModify="((Item=(Type=Bone,Name=\"skirt_r_f_Master\"),Transform=(Rotation=(X=0,Y=-0.1564,Z=0,W=0.9877),Translation=(X=0,Y=0,Z=0),Scale3D=(X=1,Y=1,Z=1))),(Item=(Type=Bone,Name=\"dyn_skirt_r_f_02\"),Transform=(Rotation=(X=0,Y=0,Z=0,W=1),Translation=(X=0,Y=0,Z=0),Scale3D= ...(truncated)"
    WeightMinimum=0
    Weight=@n1.Result
   )
   n2 = &RigVMFunction_AlphaInterp(
    InterpSpeedDecreasing=8
    InterpSpeedIncreasing=10
    Scale=1
    bClampResult=False
    Value=&RigVMFunction_MathDoubleMul(B=RigVMVariableNode() {variable=MoveBackward}.Value, A=RigVMVariableNode() {variable=CrouchWalk}.Value).Result
   )
   > &RigUnit_ModifyTransforms(
    ItemToModify="((Item=(Type=Bone,Name=\"skirt_b_m_Master\"),Transform=(Rotation=(X=0,Y=0.0436,Z=0,W=0.999),Translation=(X=0,Y=0,Z=0),Scale3D=(X=1,Y=1,Z=1))),(Item=(Type=Bone,Name=\"skirt_L_b_Master\"),Transform=(Rotation=(X=0.0038,Y=-0.0435,Z=0.0871,W=0.9952),Translation=(X= ...(truncated)"
    WeightMinimum=0
    Weight=@n2.Result
   )
   n3 = &RigVMFunction_AlphaInterp(
    InterpSpeedDecreasing=8
    InterpSpeedIncreasing=10
    Scale=1
    bClampResult=False
    Value=&RigVMFunction_MathDoubleMul(A=RigVMVariableNode() {variable=CrouchWalk}.Value, B=RigVMVariableNode() {variable=MoveRight}.Value).Result
   )
   > &RigUnit_ModifyTransforms(
    ItemToModify="((Item=(Type=Bone,Name=\"dyn_skirt_r_master\"),Transform=(Rotation=(X=-0.0076,Y=-0.0868,Z=-0.0868,W=0.9924),Translation=(X=0,Y=0,Z=0),Scale3D=(X=1,Y=1,Z=1))),(Item=(Type=Bone,Name=\"skirt_r_f_Master\"),Transform=(Rotation=(X=0,Y=-0.0872,Z=0,W=0.9962),Translati ...(truncated)"
    WeightMinimum=0
    Weight=@n3.Result
   )
   > &RigUnit_ModifyTransforms(
    ItemToModify="((Item=(Type=Bone,Name=\"skirt_L_f_Master\"),Transform=(Rotation=(X=0,Y=-0.1305,Z=0,W=0.9914),Translation=(X=0,Y=0,Z=0),Scale3D=(X=1,Y=1,Z=1))),(Item=(Type=Bone,Name=\"dyn_skirt_l_master\"),Transform=(Rotation=(X=0,Y=-0.1305,Z=0,W=0.9914),Translation=(X=0,Y=0, ...(truncated)"
    WeightMinimum=0
    Weight=@n3.Result
   )
   n4 = &RigVMFunction_AlphaInterp(
    InterpSpeedDecreasing=8
    InterpSpeedIncreasing=10
    Scale=1
    bClampResult=False
    Value=&RigVMFunction_MathDoubleMul(A=RigVMVariableNode() {variable=CrouchWalk}.Value, B=RigVMVariableNode() {variable=MoveLeft}.Value).Result
   )
   > &RigUnit_ModifyTransforms(
    ItemToModify="((Item=(Type=Bone,Name=\"dyn_skirt_r_master\"),Transform=(Rotation=(X=-0.0076,Y=-0.0868,Z=-0.0868,W=0.9924),Translation=(X=0,Y=0,Z=0),Scale3D=(X=1,Y=1,Z=1))),(Item=(Type=Bone,Name=\"skirt_r_f_Master\"),Transform=(Rotation=(X=0,Y=-0.0872,Z=0,W=0.9962),Translati ...(truncated)"
    WeightMinimum=0
    Weight=@n4.Result
   )
   > &RigUnit_ModifyTransforms(
    ItemToModify="((Item=(Type=Bone,Name=\"skirt_L_f_Master\"),Transform=(Rotation=(X=0,Y=-0.1305,Z=0,W=0.9914),Translation=(X=0,Y=0,Z=0),Scale3D=(X=1,Y=1,Z=1))),(Item=(Type=Bone,Name=\"dyn_skirt_l_master\"),Transform=(Rotation=(X=0,Y=-0.1305,Z=0,W=0.9914),Translation=(X=0,Y=0, ...(truncated)"
    WeightMinimum=0
    Weight=@n4.Result
   )
   n5 = &RigVMFunction_AlphaInterp(InterpSpeedDecreasing=10, InterpSpeedIncreasing=10, Scale=1, bClampResult=False, Value=RigVMVariableNode() {variable=Slide}.Value)
   > &RigUnit_ModifyTransforms(
    ItemToModify="((Item=(Type=Bone,Name=\"None\"),Transform=(Rotation=(X=0,Y=0,Z=0,W=1),Translation=(X=0,Y=0,Z=0),Scale3D=(X=1,Y=1,Z=1))),(Item=(Type=Bone,Name=\"skirt_b_m_Master\"),Transform=(Rotation=(X=-0.0261,Y=0.812,Z=-0.0408,W=-0.5817),Translation=(X=0,Y=0,Z=0),Scale3D=( ...(truncated)"
    WeightMinimum=0
    Weight=@n5.Result
   )
   > &RigUnit_ModifyTransforms(
    ItemToModify="((Item=(Type=Bone,Name=\"None\"),Transform=(Rotation=(X=0,Y=0,Z=0,W=1),Translation=(X=0,Y=0,Z=0),Scale3D=(X=1,Y=1,Z=1))),(Item=(Type=Bone,Name=\"dyn_skirt_b_02\"),Transform=(Rotation=(X=0,Y=-0.1049,Z=0,W=0.9945),Translation=(X=0,Y=0,Z=0),Scale3D=(X=1,Y=1,Z=1)) ...(truncated)"
    WeightMinimum=0
    Weight=@n5.Result
   )
   > &RigUnit_ModifyTransforms(
    ItemToModify="((Item=(Type=Bone,Name=\"None\"),Transform=(Rotation=(X=0,Y=0,Z=0,W=1),Translation=(X=0,Y=0,Z=0),Scale3D=(X=1,Y=1,Z=1))),(Item=(Type=Bone,Name=\"dyn_skirt_b_03\"),Transform=(Rotation=(X=0,Y=0,Z=0,W=1),Translation=(X=0,Y=0,Z=0),Scale3D=(X=1,Y=1,Z=1))),(Item=(Ty ...(truncated)"
    WeightMinimum=0
    Weight=@n5.Result
   )
   RigVMFunctionReturnNode(ExecuteContext=@RigVMFunction_Sequence.C)
   n6 = &RigVMFunction_AlphaInterp(InterpSpeedDecreasing=10, InterpSpeedIncreasing=10, Scale=1, bClampResult=False, Value=RigVMVariableNode() {variable=Baseball Slide}.Value)
   &RigUnit_ModifyTransforms(
    ExecutePin=()
    ItemToModify="((Item=(Type=Bone,Name=\"skirt_b_m_Master\"),Transform=(Rotation=(X=-0.0708,Y=-0.0448,Z=0.0435,W=0.9955),Translation=(X=-0.0645,Y=0.0927,Z=0.6667),Scale3D=(X=1,Y=1,Z=1))),(Item=(Type=Bone,Name=\"skirt_r_b_Master\"),Transform=(Rotation=(X=0.4977,Y=-0.3639,Z=-0. ...(truncated)"
    WeightMinimum=0
    Weight=@n6.Result
   )
   > &RigUnit_ModifyTransforms(
    ItemToModify="((Item=(Type=Bone,Name=\"dyn_skirt_b_02\"),Transform=(Rotation=(X=-0.1966,Y=-0.0195,Z=-0.1054,W=0.9746),Translation=(X=0,Y=0,Z=0),Scale3D=(X=1,Y=1,Z=1))),(Item=(Type=Bone,Name=\"dyn_skirt_r_b_02\"),Transform=(Rotation=(X=0.1093,Y=-0.0649,Z=0.0805,W=0.9886),Tra ...(truncated)"
    WeightMinimum=0
    Weight=@n6.Result
   )
  ]}
 New Function_ContainedGraph {nodes [
   n1 = RigVMFunctionEntryNode()
   RigVMFunction_MathBoolToInteger = &RigVMFunction_MathBoolToInteger(Value=@n1.onlyTwist)
   RigVMFunction_MathQuaternionSwingTwist = &RigVMFunction_MathQuaternionSwingTwist(
    twistAxis=@n1.twistAxis
    Input=&RigUnit_GetRelativeTransformForItem(bChildInitial=False, bParentInitial=False, child=@n1.child, parent=@n1.parent).RelativeTransform.Rotation
   )
   n2 = RigVMDispatchNode(
    Values=((),())
    Index=@RigVMFunction_MathBoolToInteger.Result
    Values.0=@RigVMFunction_MathQuaternionSwingTwist.Swing
    Values.1=@RigVMFunction_MathQuaternionSwingTwist.Twist
   ) {template="DISPATCH_RigVMDispatch_SelectInt32(in Index,in Values,out Result)"}
   RigVMFunction_MathQuaternionToEuler = &RigVMFunction_MathQuaternionToEuler(Value=@n2.Result, RotationOrder=@n1.Rotation Order)
   > DISPATCH_RigVMDispatch_Print = RigVMDispatchNode(ScreenColor=(R=1,G=1,B=1,A=1), ScreenDuration=1.5, Enabled=@n1.debugRotation, Value=@RigVMFunction_MathQuaternionToEuler.Result) {template="DISPATCH_RigVMDispatch_Print(in Prefix,in Value,in Enabled,in ScreenDuration,in ScreenColor)"}
   Cast = RigVMDispatchNode(Value=@n1.triggerAxis) {template="DISPATCH_RigVMDispatch_CastEnumToInt(in Value,out Result)"}
   DISPATCH_RigVMDispatch_SelectInt32 = RigVMDispatchNode(
    Values=(0,0,0)
    Index=@Cast.Result
    Values.0=@RigVMFunction_MathQuaternionToEuler.Result.X
    Values.1=@RigVMFunction_MathQuaternionToEuler.Result.Y
    Values.2=@RigVMFunction_MathQuaternionToEuler.Result.Z
   ) {template="DISPATCH_RigVMDispatch_SelectInt32(in Index,in Values,out Result)"}
   Remap = &RigVMFunction_MathDoubleRemap(
    bClamp=False
    Value=@DISPATCH_RigVMDispatch_SelectInt32.Result
    TargetMinimum=@n1.outputMinMax.X
    TargetMaximum=@n1.outputMinMax.Y
    SourceMinimum=@n1.rotationMinMax.X
    SourceMaximum=@n1.rotationMinMax.Y
   )
   Absolute = &RigVMFunction_MathDoubleAbs(Value=@Remap.Result)
   Multiply = &RigVMFunction_MathDoubleMul(
    A=&RigVMFunction_MathDoubleClamp(Value=RigVMDispatchNode(True=@Absolute.Result, Condition=@n1.absoluteValue, False=@Remap.Result) {template="DISPATCH_RigVMDispatch_If(in Condition,in True,in False,out Result)"}.Result).Result
    B=@n1.premult
   )
   > RigVMFunctionReturnNode(outputQuat=@n2.Result, outputWeight=@Multiply.Result)
  ]}
 New Function_ContainedGraph {nodes [
   RigVMFunctionEntryNode()
   > n1 = &RigUnit_SphericalPoseReader(
    ActiveRegionScaleFactors=(PositiveWidth=0.5,NegativeWidth=0.5,PositiveHeight=0.5,NegativeHeight=0.5)
    ActiveRegionSize=0.1
    Debug=(bDrawDebug=false,bDraw2D=False,bDrawLocalAxes=False,DebugScale=25,DebugSegments=20,DebugThickness=0.25)
    DriverAxis=(X=1,Y=0,Z=0)
    DriverItem=(Type=Bone,Name="clavicle_l")
    FalloffRegionScaleFactors=(PositiveWidth=0.5,NegativeWidth=0.5,PositiveHeight=0.5,NegativeHeight=0.5)
    FalloffSize=0.2
    FlipHeightScaling=False
    FlipWidthScaling=False
    OptionalParentItem=(Type=Bone,Name="None")
    RotationOffset=(X=20,Y=-40,Z=0)
   )
   > &RigUnit_ModifyTransforms(
    ItemToModify=((Item=(Type=Bone,Name="dyn_l_corler_master"),Transform=(Rotation=(X=0,Y=-0.2588,Z=0,W=0.9659),Translation=(X=0,Y=0,Z=0),Scale3D=(X=1,Y=1,Z=1))))
    WeightMinimum=0
    Weight=@n1.OutputParam
   )
   > n2 = &RigUnit_SphericalPoseReader(
    ActiveRegionScaleFactors=(PositiveWidth=0.5,NegativeWidth=0.5,PositiveHeight=0.5,NegativeHeight=0.5)
    ActiveRegionSize=0.1
    Debug=(bDrawDebug=false,bDraw2D=False,bDrawLocalAxes=False,DebugScale=25,DebugSegments=20,DebugThickness=0.25)
    DriverAxis=(X=-1,Y=0,Z=0)
    DriverItem=(Type=Bone,Name="clavicle_r")
    FalloffRegionScaleFactors=(PositiveWidth=0.5,NegativeWidth=0.5,PositiveHeight=0.5,NegativeHeight=0.5)
    FalloffSize=0.2
    FlipHeightScaling=False
    FlipWidthScaling=False
    OptionalParentItem=(Type=Bone,Name="None")
    RotationOffset=(X=-10,Y=135,Z=0)
   )
   > &RigUnit_ModifyTransforms(
    ItemToModify=((Item=(Type=Bone,Name="dyn_r_corler_master"),Transform=(Rotation=(X=0,Y=-0.2588,Z=0,W=0.9659),Translation=(X=0,Y=0,Z=0),Scale3D=(X=1,Y=1,Z=1))))
    WeightMinimum=0
    Weight=@n2.OutputParam
   )
   > &RigUnit_ModifyTransforms(
    ItemToModify=((Item=(Type=Bone,Name="None"),Transform=(Rotation=(X=0,Y=0,Z=0,W=1),Translation=(X=0,Y=0,Z=0),Scale3D=(X=1,Y=1,Z=1))))
    Weight=1
    WeightMinimum=0
   )
   > RigVMFunctionReturnNode()
  ]}
 RigVMFunctionLibrary {nodes [
   CrouchFix = RigVMCollapseNode(ExecuteContext=())
   DeltaRotation = RigVMCollapseNode(
    ExecuteContext=()
    Rotation Order=ZYX
    child=(Type=Bone,Name="None")
    outputMinMax=(X=0,Y=1)
    parent=(Type=Bone,Name="None")
    premult=1
    rotationMinMax=(X=0,Y=0)
    twistAxis=(X=1,Y=0,Z=0)
   )
   Emotes = RigVMCollapseNode(ExecuteContext=())
   SkirtFix = RigVMCollapseNode(ExecuteContext=())
   UniversalSkirtOffsets = RigVMCollapseNode(ExecuteContext=())
  ]}
 RigVMFunction_Sequence_1_ContainedGraph {nodes [
   n1 = RigVMFunctionEntryNode()
   RigVMFunction_Sequence = &RigVMFunction_Sequence(ExecuteContext=@n1.ExecuteContext)
   n2 = &RigVMFunction_Sequence(ExecuteContext=@RigVMFunction_Sequence.B)
   RigVMFunctionReturnNode(A=@RigVMFunction_Sequence.A, B=@n2.A, C=@n2.B)
  ]}
 RigVMFunction_Sequence_2_ContainedGraph {nodes [
   n1 = RigVMFunctionEntryNode()
   n2 = &RigVMFunction_Sequence(ExecuteContext=@n1.ExecuteContext)
   n3 = &RigVMFunction_Sequence(ExecuteContext=@n2.B)
   RigVMFunctionReturnNode(A=@n2.A, B=@n3.A, C=@n3.B)
  ]}
 RigVMFunction_Sequence_2_ContainedGraph {nodes [
   n1 = RigVMFunctionEntryNode()
   RigVMFunction_Sequence = &RigVMFunction_Sequence(ExecuteContext=@n1.ExecuteContext)
   n2 = &RigVMFunction_Sequence(ExecuteContext=@RigVMFunction_Sequence.B)
   n3 = &RigVMFunction_Sequence(ExecuteContext=@n2.B)
   RigVMFunctionReturnNode(A=@RigVMFunction_Sequence.A, B=@n2.A, C=@n3.A, D=@n3.B)
  ]}
 RigVMModel {nodes [
   RigUnit_BeginExecution = &RigUnit_BeginExecution()
   RigVMFunction_Sequence = &RigVMFunction_Sequence(ExecuteContext=@RigUnit_BeginExecution.ExecutePin)
   FN_DeformRig_FNC = RigVMFunctionReferenceNode(
    DeformRig_ Curve="(deformRig_l_upperarm_up_alpha_72_#1=False,deformRig_r_upperarm_up_alpha_73_#2=False,deformRig_l_upperarm_fr_alpha_74_#3=False,deformRig_r_upperarm_fr_alpha_75_CB51B8 ...(truncated)"
    DeformRig_Bool="(upperarm_1_#4=True,elbow_3_#5=True,wrist_7_#6=True,groin_8_#7=True,glute_17_#8=True,knee_9_40C408324CA261 ...(truncated)"
    DeformRig_Debug=(upperarm_debug_18_#4=False,wrist_debug_20_#6=False,lat_debug_26_#9=False,groinGlute_debug_27_#7=False)
    DeformRig_Value="(upperarm_up_18_#10=(Rotation=(X=0.0,Y=-0.0,Z=0.0,W=1.0),Translation=(X=0.0,Y=0.0,Z=8.0),Scale3D=(X=0.70,Y=1.0,Z=1.0)),upperarm_fwd_fr_19_#11=(Rotation=(X=0.0,Y=-0.0,Z=0.0,W=1.0),Translation=(X=0.0, ...(truncated)"
    ExecuteContext=@RigVMFunction_Sequence.A
   )
   SkirtFix = RigVMFunctionReferenceNode(ExecuteContext=@RigVMFunction_Sequence.B)
   > UniversalSkirtOffsets = RigVMFunctionReferenceNode()
   > CrouchFix = RigVMFunctionReferenceNode()
   > Emotes = RigVMFunctionReferenceNode()
  ]}
]
hierarchy [
 root {loc=<0 0 0>, rot=<0 0 0 1>}
 pelvis {parent=root, loc=<0 -0.5301 98.0571>, rot=<0.0127 0.707 -0.0127 -0.707>}
 spine_01 {parent=pelvis, loc=<6.1208 -0.9328 0>, rot=<0 0 0.034 -0.9994>}
 spine_02 {parent=spine_01, loc=<6.5997 -0.2095 0>, rot=<0 0 0.0238 -0.9997>}
 spine_03 {parent=spine_02, loc=<6.4914 -0.1237 0>, rot=<0 0 0.0149 -0.9999>}
 spine_04 {parent=spine_03, loc=<6.9598 0.0794 0>, rot=<0 0 -0.0132 -0.9999>}
 spine_05 {parent=spine_04, loc=<6.9987 0 0>, rot=<0 0 -0.0015 -1>}
 neck_01 {parent=spine_05, loc=<11.3138 0.1529 0>, rot=<0 0 -0.2003 -0.9797>}
 neck_02 {parent=neck_01, loc=<4.5627 0 0>, rot=<0 0 0.0504 -0.9987>}
 head {parent=neck_02, loc=<5.3914 0 0>, rot=<0 0 0.115 -0.9934>}
 clavicle_l {parent=spine_05, loc=<5.4831 -0.1053 -3.2946>, rot=<0.0465 -0.7289 -0.009 -0.6829>}
 upperarm_l {parent=clavicle_l, loc=<9.8658 -1.2136 -0.4105>, rot=<0.0164 -0.3245 -0.0269 -0.9454>}
 lowerarm_l {parent=upperarm_l, loc=<26.465 -0.8345 0.259>, rot=<0.0036 0.0161 -0.2298 -0.9731>}
 hand_l {parent=lowerarm_l, loc=<22.7355 -0.3203 -0.6105>, rot=<-0.6784 -0.0223 0.0735 -0.7307>}
 thumb_01_l {parent=hand_l, loc=<1.3118 -0.3191 -2.459>, rot=<0.6546 -0.4042 0.0664 -0.6354>}
 thumb_02_l {parent=thumb_01_l, loc=<5.9336 -0.0448 0.0401>, rot=<-0.0037 0.0213 0.298 -0.9543>}
 thumb_03_l {parent=thumb_02_l, loc=<2.2223 0.3682 -0.0907>, rot=<0.0118 -0.013 0.0066 -0.9998>}
 index_metacarpal_l {parent=hand_l, loc=<1.6535 0.0183 -2.96>, rot=<0.0158 -0.0124 -0.0173 -0.9996>}
 index_01_l {parent=index_metacarpal_l, loc=<8.3008 0.0357 -0.0524>, rot=<-0.0044 0.0485 0.0924 -0.9945>}
 index_02_l {parent=index_01_l, loc=<3.3587 -0.3004 0.0452>, rot=<-0.002 0.0027 0.3225 -0.9466>}
 index_03_l {parent=index_02_l, loc=<2.5578 -0.123 -0.0252>, rot=<-0.002 0.0067 0.2558 -0.9667>}
 middle_metacarpal_l {parent=hand_l, loc=<2.1856 -0.1725 -0.5984>, rot=<0.006 0.0168 -0.0329 -0.9993>}
 middle_01_l {parent=middle_metacarpal_l, loc=<7.3275 -0.2447 0.0129>, rot=<-0.0208 0.0042 0.0858 -0.9961>}
 middle_02_l {parent=middle_01_l, loc=<4.0123 0.1401 0.1175>, rot=<-0.0218 0.0258 0.3512 -0.9357>}
 middle_03_l {parent=middle_02_l, loc=<2.6077 0.0534 0.1373>, rot=<0.0082 0.0188 0.1708 -0.9851>}
 ring_metacarpal_l {parent=hand_l, loc=<2.2074 -0.3625 1.3638>, rot=<-0.0284 0.0722 -0.0458 -0.9959>}
 ring_01_l {parent=ring_metacarpal_l, loc=<7.0295 -0.6197 -0.1446>, rot=<-0.0489 -0.0252 0.1892 -0.9804>}
 ring_02_l {parent=ring_01_l, loc=<3.6174 0.4647 -0.1027>, rot=<0 0 0.2806 -0.9598>}
 ring_03_l {parent=ring_02_l, loc=<2.4152 0.114 0.0199>, rot=<0 0 0.1189 -0.9929>}
 pinky_metacarpal_l {parent=hand_l, loc=<1.7616 -0.9753 3.3814>, rot=<-0.0397 0.0814 -0.069 -0.9935>}
 pinky_01_l {parent=pinky_metacarpal_l, loc=<7.1709 -0.4831 -0.1094>, rot=<-0.0654 -0.0035 0.2242 -0.9723>}
 pinky_02_l {parent=pinky_01_l, loc=<2.8274 0.0893 -0.1043>, rot=<0.0037 0.0011 0.2854 -0.9584>}
 pinky_03_l {parent=pinky_02_l, loc=<2.2046 0.0307 -0.0078>, rot=<-0.0002 0.0098 0.1445 -0.9895>}
 deform_wrist_dn_base_l {parent=hand_l, loc=<0 -1.5326 0>, rot=<0 0 0 -1>}
 deform_wrist_dn_l {parent=deform_wrist_dn_base_l, loc=<0 0 0>, rot=<0 0 0 1>}
 deform_wrist_up_base_l {parent=hand_l, loc=<0 2.3009 0>, rot=<0 0 0 -1>}
 deform_wrist_up_l {parent=deform_wrist_up_base_l, loc=<0 0 0>, rot=<0 0 0 1>}
 hand_attach_l {parent=hand_l, loc=<8.5 -3 0>, rot=<0.7071 0.7071 0 0>}
 weapon_l {parent=hand_l, loc=<-0.1974 -3.4111 0.3811>, rot=<0 0 0 1>}
 lowerarm_twist_02_l {parent=lowerarm_l, loc=<10.6707 0 0>, rot=<0 0 0 1>}
 lowerarm_twist_01_l {parent=lowerarm_l, loc=<18.316 0 0>, rot=<0 0 0 1>}
 deform_elbow_in_base_l {parent=lowerarm_l, loc=<0.1834 3.6062 -0.2572>, rot=<0 0 0 -1>}
 deform_elbow_in_l {parent=deform_elbow_in_base_l, loc=<0 0 0>, rot=<0 0 0 1>}
 deform_elbow_out_base_l {parent=lowerarm_l, loc=<0 -4 0>, rot=<0 0 0 1>}
 deform_elbow_out_l {parent=deform_elbow_out_base_l, loc=<0 0 0>, rot=<0 0 0 1>}
 upperarm_twist_01_l {parent=upperarm_l, loc=<7.2135 -0.1134 0>, rot=<0 0 0 1>}
 deform_upperarm_bk_base_l {parent=upperarm_twist_01_l, loc=<-6.4053 -4.3746 0.314>, rot=<0 0 0 -1>}
 deform_upperarm_bk_l {parent=deform_upperarm_bk_base_l, loc=<0 0 0>, rot=<0 0 0 1>}
 deform_upperarm_fr_base_l {parent=upperarm_twist_01_l, loc=<-5.6312 5.092 0.3354>, rot=<0 0 0 -1>}
 deform_upperarm_fr_l {parent=deform_upperarm_fr_base_l, loc=<0 0 0>, rot=<0 0 0 1>}
 deform_upperarm_up_base_l {parent=upperarm_twist_01_l, loc=<-8.891 0.4825 3.9517>, rot=<0 0 0 -1>}
 deform_upperarm_up_l {parent=deform_upperarm_up_base_l, loc=<0 0 0>, rot=<0 0 0 1>}
 upperarm_twist_02_l {parent=upperarm_l, loc=<19.7998 -0.1134 0>, rot=<0 0 0 1>}
 clavicle_r {parent=spine_05, loc=<5.4831 -0.1053 3.2946>, rot=<0.7289 0.0465 -0.6829 0.009>}
 upperarm_r {parent=clavicle_r, loc=<-9.8658 1.2136 0.4105>, rot=<0.0164 -0.3245 -0.0269 -0.9454>}
 lowerarm_r {parent=upperarm_r, loc=<-26.465 0.8345 -0.259>, rot=<0.0036 0.0161 -0.2298 -0.9731>}
 hand_r {parent=lowerarm_r, loc=<-22.7355 0.3203 0.6106>, rot=<-0.6784 -0.0223 0.0735 -0.7307>}
 thumb_01_r {parent=hand_r, loc=<-1.3118 0.3191 2.459>, rot=<0.6546 -0.4042 0.0664 -0.6354>}
 thumb_02_r {parent=thumb_01_r, loc=<-5.9336 0.0448 -0.0401>, rot=<-0.0037 0.0213 0.298 -0.9543>}
 thumb_03_r {parent=thumb_02_r, loc=<-2.2223 -0.3682 0.0907>, rot=<0.0118 -0.013 0.0066 -0.9998>}
 index_metacarpal_r {parent=hand_r, loc=<-1.6535 -0.0183 2.96>, rot=<0.0158 -0.0124 -0.0173 -0.9996>}
 index_01_r {parent=index_metacarpal_r, loc=<-8.3008 -0.0357 0.0524>, rot=<-0.0044 0.0485 0.0924 -0.9945>}
 index_02_r {parent=index_01_r, loc=<-3.3587 0.3004 -0.0452>, rot=<-0.002 0.0027 0.3225 -0.9466>}
 index_03_r {parent=index_02_r, loc=<-2.5578 0.123 0.0252>, rot=<-0.002 0.0067 0.2558 -0.9667>}
 middle_metacarpal_r {parent=hand_r, loc=<-2.1856 0.1725 0.5984>, rot=<0.006 0.0168 -0.0329 -0.9993>}
 middle_01_r {parent=middle_metacarpal_r, loc=<-7.3275 0.2447 -0.0129>, rot=<-0.0208 0.0042 0.0858 -0.9961>}
 middle_02_r {parent=middle_01_r, loc=<-4.0123 -0.1401 -0.1175>, rot=<-0.0218 0.0258 0.3512 -0.9357>}
 middle_03_r {parent=middle_02_r, loc=<-2.6077 -0.0534 -0.1373>, rot=<0.0082 0.0188 0.1708 -0.9851>}
 ring_metacarpal_r {parent=hand_r, loc=<-2.2074 0.3625 -1.3638>, rot=<-0.0284 0.0722 -0.0458 -0.9959>}
 ring_01_r {parent=ring_metacarpal_r, loc=<-7.0295 0.6197 0.1446>, rot=<-0.0489 -0.0252 0.1892 -0.9804>}
 ring_02_r {parent=ring_01_r, loc=<-3.6174 -0.4647 0.1027>, rot=<0 0 0.2806 -0.9598>}
 ring_03_r {parent=ring_02_r, loc=<-2.4152 -0.114 -0.0199>, rot=<0 0 0.1189 -0.9929>}
 pinky_metacarpal_r {parent=hand_r, loc=<-1.7616 0.9753 -3.3814>, rot=<-0.0397 0.0814 -0.069 -0.9935>}
 pinky_01_r {parent=pinky_metacarpal_r, loc=<-7.1709 0.4831 0.1094>, rot=<-0.0654 -0.0035 0.2242 -0.9723>}
 pinky_02_r {parent=pinky_01_r, loc=<-2.8274 -0.0893 0.1043>, rot=<0.0037 0.0011 0.2854 -0.9584>}
 pinky_03_r {parent=pinky_02_r, loc=<-2.2046 -0.0307 0.0078>, rot=<-0.0002 0.0098 0.1445 -0.9895>}
 deform_wrist_dn_base_r {parent=hand_r, loc=<-0.0001 1.5328 0>, rot=<0 0 0 -1>}
 deform_wrist_dn_r {parent=deform_wrist_dn_base_r, loc=<0 0 0>, rot=<0 0 0 1>}
 deform_wrist_up_base_r {parent=hand_r, loc=<-0.0002 -2.3007 -0.0001>, rot=<0 0 0 -1>}
 deform_wrist_up_r {parent=deform_wrist_up_base_r, loc=<0 0 0>, rot=<0 0 0 1>}
 weapon_r {parent=hand_r, loc=<0.1974 3.4111 -0.3811>, rot=<0 0 0 1>}
 hand_attach_r {parent=hand_r, loc=<-8.5001 2.9999 0>, rot=<0 0 -0.7071 -0.7071>}
 lowerarm_twist_02_r {parent=lowerarm_r, loc=<-10.6707 0 0>, rot=<0 0 0 -1>}
 lowerarm_twist_01_r {parent=lowerarm_r, loc=<-18.316 0 0>, rot=<0 0 0 -1>}
 deform_elbow_in_base_r {parent=lowerarm_r, loc=<-0.1834 -3.6062 0.2572>, rot=<0 0 0 -1>}
 deform_elbow_in_r {parent=deform_elbow_in_base_r, loc=<0 0 0>, rot=<0 0 0 1>}
 deform_elbow_out_base_r {parent=lowerarm_r, loc=<0 4 -0.0001>, rot=<0 0 0 -1>}
 deform_elbow_out_r {parent=deform_elbow_out_base_r, loc=<0 0 0>, rot=<0 0 0 1>}
 upperarm_twist_01_r {parent=upperarm_r, loc=<-7.2135 0.1134 0>, rot=<0 0 0 -1>}
 deform_upperarm_bk_base_r {parent=upperarm_twist_01_r, loc=<6.4053 4.3746 -0.314>, rot=<0 0 0 -1>}
 deform_upperarm_bk_r {parent=deform_upperarm_bk_base_r, loc=<0 0 0>, rot=<0 0 0 1>}
 deform_upperarm_fr_base_r {parent=upperarm_twist_01_r, loc=<5.6312 -5.092 -0.3354>, rot=<0 0 0 -1>}
 deform_upperarm_fr_r {parent=deform_upperarm_fr_base_r, loc=<0 0 0>, rot=<0 0 0 1>}
 deform_upperarm_up_base_r {parent=upperarm_twist_01_r, loc=<8.891 -0.4825 -3.9517>, rot=<0 0 0 -1>}
 deform_upperarm_up_r {parent=deform_upperarm_up_base_r, loc=<0 0 0>, rot=<0 0 0 1>}
 upperarm_twist_02_r {parent=upperarm_r, loc=<-19.7998 0.1134 0>, rot=<0 0 0 -1>}
 attach_backpack {parent=spine_05, loc=<-0.123 -13.473 0>, rot=<-0.0093 -0.707 0.0093 -0.707>}
 Backpack_BR {parent=attach_backpack, type=Null, loc=<0 5.5929 0>, rot=<0 0 0 1>}
 attach_cape {parent=spine_05, loc=<9.017 -11.714 0>, rot=<0.0616 -0.7044 -0.0616 -0.7044>}
 cape {parent=attach_cape, type=Null, loc=<0 5.6082 0>, rot=<0 0 0 1>}
 dyn_l_corler_master {parent=spine_05, loc=<11.6027 0.8357 -6.3356>, rot=<0.2184 -0.5832 -0.0917 -0.777>}
 dyn_l_corler_01 {parent=dyn_l_corler_master, loc=<0 0 0>, rot=<0 0 0 1>}
 dyn_l_corler_02 {parent=dyn_l_corler_01, loc=<7.7954 0 0>, rot=<0 0 0 1>}
 dyn_r_corler_master {parent=spine_05, loc=<10.5961 1.5916 6.4976>, rot=<0.7151 0.0561 0.6046 -0.3463>}
 dyn_r_corler_01 {parent=dyn_r_corler_master, loc=<0 0 0>, rot=<0 0 0 1>}
 dyn_r_corler_02 {parent=dyn_r_corler_01, loc=<7.7954 0 0>, rot=<0 0 0 1>}
 deform_pec_aim_base_l {parent=spine_04, loc=<3.2456 9.3309 -1.2237>, rot=<0.0603 -0.4033 0.3318 -0.8507>}
 deform_pec_aim_l {parent=deform_pec_aim_base_l, loc=<0 0 0>, rot=<0 0 0 1>}
 deform_pec_base_l {parent=deform_pec_aim_l, loc=<9.3665 0 0>, rot=<0 0 0 1>}
 deform_pec_l {parent=deform_pec_base_l, loc=<0 0 0>, rot=<0 0 0 1>}
 deform_pec_aim_base_r {parent=spine_04, loc=<3.2452 9.3309 1.2237>, rot=<-0.4033 -0.0603 0.8507 0.3318>}
 deform_pec_aim_r {parent=deform_pec_aim_base_r, loc=<0 0 0>, rot=<0 0 0 1>}
 deform_pec_base_r {parent=deform_pec_aim_r, loc=<-9.3667 -0.0003 -0.0004>, rot=<0 0 0 1>}
 deform_pec_r {parent=deform_pec_base_r, loc=<0 0 0>, rot=<0 0 0 1>}
 dyn_zipper_master {parent=spine_04, loc=<5.4478 10.4678 0>, rot=<0.0859 0.7019 0.7019 0.0859>}
 dyn_zipper_01 {parent=dyn_zipper_master, loc=<0 0 0>, rot=<0 0 0 1>}
 dyn_jacketback_r_b_master {parent=spine_03, loc=<7.2131 -5.1939 6.2039>, rot=<0.0876 -0.5737 0.8113 -0.0707>}
 dyn_jacketback_r_b_01 {parent=dyn_jacketback_r_b_master, loc=<0 0 0>, rot=<0 0 0 1>}
 dyn_jacketback_r_b_02 {parent=dyn_jacketback_r_b_01, loc=<5.5066 0 0>, rot=<0 0 0 1>}
 FX_Jacket_BR_01 {parent=dyn_jacketback_r_b_01, type=Null, loc=<0 0 0>, rot=<0 0 0 1>}
 dyn_jacketSide_r_01_master {parent=spine_03, loc=<5.3794 0.358 9.8703>, rot=<0.0699 0.0051 0.9967 -0.0404>}
 dyn_jacketSide_r_01 {parent=dyn_jacketSide_r_01_master, loc=<0 0 0>, rot=<0 0 0 1>}
 dyn_jacketSide_r_02 {parent=dyn_jacketSide_r_01, loc=<5.5066 0 0>, rot=<0 0 0 1>}
 FX_Jacket_SR_01 {parent=dyn_jacketSide_r_01, type=Null, loc=<0 0 0>, rot=<0 0 0 1>}
 dyn_jacketFront_r_a_master {parent=spine_03, loc=<4.6688 6.9325 8.4959>, rot=<0.0778 0.2607 0.9615 -0.0395>}
 dyn_jacketFront_r_a_01 {parent=dyn_jacketFront_r_a_master, loc=<0 0 0>, rot=<0 0 0 1>}
 dyn_jacketFront_r_a_02 {parent=dyn_jacketFront_r_a_01, loc=<5.5066 0 0>, rot=<0 0 0 1>}
 FX_Jacket_FR_01 {parent=dyn_jacketFront_r_a_01, type=Null, loc=<0 0 0>, rot=<0 0 0 1>}
 dyn_jacketFront_r_master {parent=spine_03, loc=<4.6872 10.4711 4.378>, rot=<-0.0057 -0.7187 -0.6953 -0.0063>}
 dyn_jacketFront_r_01 {parent=dyn_jacketFront_r_master, loc=<0 0 0>, rot=<0 0 0 1>}
 dyn_jacketFront_r_02 {parent=dyn_jacketFront_r_01, loc=<5.5066 0 0>, rot=<0 0 0 1>}
 dyn_jacketBack_master {parent=spine_03, loc=<7.9751 -6.3307 -0.5687>, rot=<0.0586 -0.7047 0.7047 -0.0586>}
 dyn_jacketBack_01 {parent=dyn_jacketBack_master, loc=<0 0 0>, rot=<0 0 0 1>}
 dyn_jacketBack_02 {parent=dyn_jacketBack_01, loc=<5.5066 0 0>, rot=<0 0 0 1>}
 FX_Jacket_BM_01 {parent=dyn_jacketBack_01, type=Null, loc=<0 0 0>, rot=<0 0 0 1>}
 dyn_jacketback_l_b_master {parent=spine_03, loc=<6.7121 -4.4482 -7.3766>, rot=<0.0808 -0.8722 0.4715 -0.1022>}
 dyn_jacketback_l_b_01 {parent=dyn_jacketback_l_b_master, loc=<0 0 0>, rot=<0 0 0 1>}
 dyn_jacketback_l_b_02 {parent=dyn_jacketback_l_b_01, loc=<5.5066 0 0>, rot=<0 0 0 1>}
 FX_Jacket_BL_01 {parent=dyn_jacketback_l_b_01, type=Null, loc=<0 0 0>, rot=<0 0 0 1>}
 dyn_jacketSide_l_01_master {parent=spine_03, loc=<5.8886 0.3118 -9.6995>, rot=<0.0407 -0.9985 0.0008 -0.0361>}
 dyn_jacketSide_l_01 {parent=dyn_jacketSide_l_01_master, loc=<0 0 0>, rot=<0 0 0 1>}
 dyn_jacketSide_l_02 {parent=dyn_jacketSide_l_01, loc=<5.5066 0 0>, rot=<0 0 0 1>}
 FX_Jacket_SL_01 {parent=dyn_jacketSide_l_01, type=Null, loc=<0 0 0>, rot=<0 0 0 1>}
 dyn_jacketFront_l_a_master {parent=spine_03, loc=<4.7533 7.0943 -8.6313>, rot=<0.0553 -0.9379 -0.3422 -0.0135>}
 dyn_jacketFront_l_a_01 {parent=dyn_jacketFront_l_a_master, loc=<0 0 0>, rot=<0 0 0 1>}
 dyn_jacketFront_l_a_02 {parent=dyn_jacketFront_l_a_01, loc=<5.5066 0 0>, rot=<0 0 0 1>}
 FX_Jacket_FL_01 {parent=dyn_jacketFront_l_a_01, type=Null, loc=<0 0 0>, rot=<0 0 0 1>}
 dyn_jacketFront_l_master {parent=spine_03, loc=<4.6879 10.4377 -4.1135>, rot=<-0.0069 0.7071 0.7071 -0.0069>}
 dyn_jacketFront_l_01 {parent=dyn_jacketFront_l_master, loc=<0 0 0>, rot=<0 0 0 1>}
 dyn_jacketFront_l_02 {parent=dyn_jacketFront_l_01, loc=<5.5066 0 0>, rot=<0 0 0 1>}
 dyn_bag_master {parent=spine_02, loc=<-0.2172 1.0018 -10.7897>, rot=<0.101 -0.9744 0.1004 -0.174>}
 dyn_bag_01 {parent=dyn_bag_master, loc=<0 0 0>, rot=<0 0 0 1>}
 thigh_l {parent=pelvis, loc=<-1.0855 -0.2109 -8.7053>, rot=<0.0009 0.0355 0.0155 -0.9993>}
 calf_l {parent=thigh_l, loc=<-42.5284 -0.0019 -0.7297>, rot=<0.0006 -0.0161 -0.0391 -0.9991>}
 foot_l {parent=calf_l, loc=<-46.7401 -2.5014 -0.592>, rot=<0.0427 -0.0238 0.0353 -0.9982>}
 ball_l {parent=foot_l, loc=<-3.5658 10.8066 -0.1544>, rot=<0.0001 -0.0001 -0.7063 -0.7079>}
 calf_twist_02_l {parent=calf_l, loc=<-19.1546 -1.7943 -0.2638>, rot=<0.0237 0.0095 -0.0035 -0.9997>}
 calf_twist_01_l {parent=calf_l, loc=<-35.8483 -1.9197 -0.5794>, rot=<0.0237 0.0095 -0.0035 -0.9997>}
 deform_knee_in_base_l {parent=calf_l, loc=<0.8031 -5.0231 0.2456>, rot=<0 0 0 -1>}
 deform_knee_in_l {parent=deform_knee_in_base_l, loc=<0 0 0>, rot=<0 0 0 1>}
 deform_knee_out_base_l {parent=calf_l, loc=<0 4 0>, rot=<0 0 0 -1>}
 deform_knee_out_l {parent=deform_knee_out_base_l, loc=<0 0 0>, rot=<0 0 0 1>}
 thigh_twist_01_l {parent=thigh_l, loc=<-27.952 0 0>, rot=<0.0013 0 0 -1>}
 deform_groin_base_l {parent=thigh_l, loc=<0.4801 5.6776 -0.2486>, rot=<0 0 0 -1>}
 deform_groin_l {parent=deform_groin_base_l, loc=<0 0 0>, rot=<0 0 0 1>}
 thigh_r {parent=pelvis, loc=<-1.0855 -0.2109 8.7053>, rot=<0.0355 -0.0009 0.9993 0.0155>}
 calf_r {parent=thigh_r, loc=<42.5284 0.0019 0.7297>, rot=<0.0006 -0.0161 -0.0391 -0.9991>}
 foot_r {parent=calf_r, loc=<46.7401 2.5014 0.592>, rot=<0.0427 -0.0238 0.0353 -0.9982>}
 ball_r {parent=foot_r, loc=<3.5658 -10.8066 0.1544>, rot=<0.0001 -0.0001 -0.7063 -0.7079>}
 calf_twist_02_r {parent=calf_r, loc=<19.1546 1.7943 0.2638>, rot=<0.0237 0.0095 -0.0035 -0.9997>}
 calf_twist_01_r {parent=calf_r, loc=<35.8483 1.9197 0.5794>, rot=<0.0237 0.0095 -0.0035 -0.9997>}
 deform_knee_in_base_r {parent=calf_r, loc=<-0.8031 5.0231 -0.2457>, rot=<0 0 0 -1>}
 deform_knee_in_r {parent=deform_knee_in_base_r, loc=<0 0 0>, rot=<0 0 0 1>}
 deform_knee_out_base_r {parent=calf_r, loc=<0 -4 0>, rot=<0 0 0 -1>}
 deform_knee_out_r {parent=deform_knee_out_base_r, loc=<0 0 0>, rot=<0 0 0 1>}
 thigh_twist_01_r {parent=thigh_r, loc=<27.952 0 0>, rot=<0.0013 0 0 -1>}
 deform_groin_base_r {parent=thigh_r, loc=<-0.4801 -5.6776 0.2486>, rot=<0 0 0 -1>}
 deform_groin_r {parent=deform_groin_base_r, loc=<0 0 0>, rot=<0 0 0 1>}
 deform_glute_base_l {parent=pelvis, loc=<-7.3466 -9.0213 -9.1245>, rot=<0.0172 0.031 -0.4477 -0.8935>}
 deform_glute_l {parent=deform_glute_base_l, loc=<0 0 0>, rot=<0 0 0 1>}
 deform_glute_base_r {parent=pelvis, loc=<-7.3464 -9.0213 9.1245>, rot=<0.031 -0.0172 0.8935 -0.4477>}
 deform_glute_r {parent=deform_glute_base_r, loc=<0 0 0>, rot=<0 0 0 1>}
 skirt_b_m_Master {parent=pelvis, loc=<2.2192 -11.3073 0>, rot=<-0.0062 -0.7071 0.7071 0.0062>}
 dyn_skirt_b_01 {parent=skirt_b_m_Master, loc=<0 0 0>, rot=<0 0 0 1>}
 dyn_skirt_b_02 {parent=dyn_skirt_b_01, loc=<7 0 0>, rot=<0 0 0 1>}
 dyn_skirt_b_03 {parent=dyn_skirt_b_02, loc=<12 0 0>, rot=<0 0 0 1>}
 dyn_skirt_b_04 {parent=dyn_skirt_b_03, loc=<12 0 0>, rot=<0 0 0 1>}
 dyn_skirt_b_05 {parent=dyn_skirt_b_04, loc=<12 0 0>, rot=<0 0 0 1>}
 dyn_skirt_b_06 {parent=dyn_skirt_b_05, loc=<12 0 0>, rot=<0 0 0 1>}
 FX_Skirt_Back_M_01 {parent=dyn_skirt_b_05, type=Null, loc=<0 0 0>, rot=<0 0 0 1>}
 dyn_skirt_r_master {parent=pelvis, loc=<4.9525 -0.2167 12.9815>, rot=<0.1792 -0.0032 0.9837 0.0176>}
 dyn_skirt_r_01 {parent=dyn_skirt_r_master, loc=<0 0 0>, rot=<0 0 0 1>}
 dyn_skirt_r_02 {parent=dyn_skirt_r_01, loc=<7 0 0>, rot=<0 -0.0755 0 -0.9971>}
 dyn_skirt_r_03 {parent=dyn_skirt_r_02, loc=<12 0 0>, rot=<0 -0.0703 0 -0.9975>}
 dyn_skirt_r_04 {parent=dyn_skirt_r_03, loc=<12 0 0>, rot=<0 0 0 1>}
 dyn_skirt_r_05 {parent=dyn_skirt_r_04, loc=<12 0 0>, rot=<0 0 0 1>}
 dyn_skirt_r_06 {parent=dyn_skirt_r_05, loc=<12 0 0>, rot=<0 0 0 1>}
 FX_Skirt_Side_R_01 {parent=dyn_skirt_r_04, type=Null, loc=<-4.76 3.39 -0.66>, rot=<0 0 0 1>}
 skirt_r_f_Master {parent=pelvis, loc=<5.2208 5.6279 8.5793>, rot=<0.1107 0.5823 0.8033 0.0579>}
 dyn_skirt_r_f_01 {parent=skirt_r_f_Master, loc=<0 0 0>, rot=<0 0.0146 0 -0.9999>}
 dyn_skirt_r_f_02 {parent=dyn_skirt_r_f_01, loc=<7 0 0>, rot=<0 0 0 1>}
 dyn_skirt_r_f_03 {parent=dyn_skirt_r_f_02, loc=<12 0 0>, rot=<0 -0.0624 0 -0.9981>}
 dyn_skirt_r_f_04 {parent=dyn_skirt_r_f_03, loc=<12 0 0>, rot=<0 0 0 1>}
 dyn_skirt_r_f_05 {parent=dyn_skirt_r_f_04, loc=<12 0 0>, rot=<0 0 0 1>}
 dyn_skirt_r_f_06 {parent=dyn_skirt_r_f_05, loc=<12 0 0>, rot=<0 0 0 1>}
 fx_skirt_r_front {parent=dyn_skirt_r_f_04, loc=<12 0 0>, rot=<0 0 0 1>}
 FX_Skirt_Front_R_02 {parent=fx_skirt_r_front, type=Null, loc=<-2.5488 7.2069 -3.0764>, rot=<0 0 0 1>}
 FX_Skirt_Front_R_01 {parent=dyn_skirt_r_f_03, type=Null, loc=<0 0 0>, rot=<0 0 0 1>}
 dyn_skirt_l_master {parent=pelvis, loc=<3.5328 -0.1658 -14.1095>, rot=<-0.0177 -0.9906 0.0024 -0.1354>}
 dyn_skirt_l_01 {parent=dyn_skirt_l_master, loc=<0 0 0>, rot=<0 0 0 1>}
 dyn_skirt_l_02 {parent=dyn_skirt_l_01, loc=<7 0 0>, rot=<0 -0.0688 0 -0.9976>}
 dyn_skirt_l_03 {parent=dyn_skirt_l_02, loc=<12 0 0>, rot=<0 -0.0445 0 -0.999>}
 dyn_skirt_l_04 {parent=dyn_skirt_l_03, loc=<12 0 0>, rot=<0 0 0 1>}
 dyn_skirt_l_05 {parent=dyn_skirt_l_04, loc=<12 0 0>, rot=<0 0 0 1>}
 dyn_skirt_l_06 {parent=dyn_skirt_l_05, loc=<12 0 0>, rot=<0 0 0 1>}
 FX_Skirt_Side_L_01 {parent=dyn_skirt_l_04, type=Null, loc=<-6.69 -3.39 -0.4>, rot=<0 0 0 1>}
 skirt_L_f_Master {parent=pelvis, loc=<3.4589 5.0543 -11.0742>, rot=<-0.0528 -0.8002 -0.586 -0.1161>}
 dyn_skirt_l_f1 {parent=skirt_L_f_Master, loc=<0.0128 0 0.0002>, rot=<0 0.0185 0 -0.9998>}
 dyn_skirt_l_f2 {parent=dyn_skirt_l_f1, loc=<7 0 0>, rot=<0 -0.0239 0 -0.9997>}
 dyn_skirt_l_f3 {parent=dyn_skirt_l_f2, loc=<12 0 0>, rot=<0 -0.0598 0 -0.9982>}
 dyn_skirt_l_f4 {parent=dyn_skirt_l_f3, loc=<12 0 0>, rot=<0 0 0 1>}
 dyn_skirt_l_f5 {parent=dyn_skirt_l_f4, loc=<12 0 0>, rot=<0 0 0 1>}
 dyn_skirt_l_f6 {parent=dyn_skirt_l_f5, loc=<12 0 0>, rot=<0 0 0 1>}
 FX_Skirt_Front_L_02 {parent=dyn_skirt_l_f5, type=Null, loc=<-4.18 -4.84 -1.89>, rot=<0 0 0 1>}
 fx_skirt_l_front {parent=dyn_skirt_l_f4, loc=<12 0 0>, rot=<0 0 0 1>}
 FX_Skirt_Front_L_01 {parent=dyn_skirt_l_f3, type=Null, loc=<0 0 0>, rot=<0 0 0 1>}
 skirt_L_b_Master {parent=pelvis, loc=<1.9448 -9.2724 -9.7828>, rot=<-0.0103 -0.8517 0.5239 0.0013>}
 dyn_skirt_l_b_01 {parent=skirt_L_b_Master, loc=<0 0 0>, rot=<0 0 0 1>}
 dyn_skirt_l_b_02 {parent=dyn_skirt_l_b_01, loc=<7 0 0>, rot=<0 0 0 1>}
 dyn_skirt_l_b_03 {parent=dyn_skirt_l_b_02, loc=<12 0 0>, rot=<0 0 0 1>}
 dyn_skirt_l_b_04 {parent=dyn_skirt_l_b_03, loc=<12 0 0>, rot=<0 0 0 1>}
 dyn_skirt_l_b_05 {parent=dyn_skirt_l_b_04, loc=<12 0 0>, rot=<0 0 0 1>}
 dyn_skirt_l_b_06 {parent=dyn_skirt_l_b_05, loc=<12 0 0>, rot=<0 0 0 1>}
 FX_Skirt_Back_L_01 {parent=dyn_skirt_l_b_04, type=Null, loc=<0 0 0>, rot=<0 0 0 1>}
 skirt_r_b_Master {parent=pelvis, loc=<1.7859 -8.8725 9.5725>, rot=<-0.0089 -0.5693 0.822 0.0053>}
 dyn_skirt_r_b_01 {parent=skirt_r_b_Master, loc=<0 0 0>, rot=<0 0 0 1>}
 dyn_skirt_r_b_02 {parent=dyn_skirt_r_b_01, loc=<7 0 0>, rot=<0 0 0 1>}
 dyn_skirt_r_b_03 {parent=dyn_skirt_r_b_02, loc=<12 0 0>, rot=<0 0 0 1>}
 dyn_skirt_r_b_04 {parent=dyn_skirt_r_b_03, loc=<12 0 0>, rot=<0 0 0 1>}
 dyn_skirt_r_b_05 {parent=dyn_skirt_r_b_04, loc=<12 0 0>, rot=<0 0 0 1>}
 dyn_skirt_r_b_06 {parent=dyn_skirt_r_b_05, loc=<12 0 0>, rot=<0 0 0 1>}
 FX_Skirt_Back_R_01 {parent=dyn_skirt_r_b_04, type=Null, loc=<0 0 0>, rot=<0 0 0 1>}
 dyn_chain_master {parent=pelvis, loc=<9.1812 0.2014 11.4378>, rot=<0.2199 -0.0039 0.9754 0.0175>}
 dyn_chain_01 {parent=dyn_chain_master, loc=<0 0 0>, rot=<0 0 0 1>}
 dyn_chain_02 {parent=dyn_chain_01, loc=<8.111 0 0>, rot=<0 0 0 1>}
 thigh_l_null {parent=pelvis, type=Null, loc=<-6.5952 0.5039 -10.8978>, rot=<-0.0888 0.0655 0.2355 -0.9656>}
 thigh_r_null {parent=pelvis, type=Null, loc=<-6.1935 -2.2704 1.9465>, rot=<-0.1506 0.2545 -0.9497 -0.1027>}
 C_Jnt_1 {parent=pelvis, loc=<0.8638 6.6257 9.076>, rot=<0.1107 0.5823 0.8033 0.0579>}
 C_Jnt_2 {parent=C_Jnt_1, loc=<7 0 0>, rot=<0 0 0 1>}
 C_Jnt_3 {parent=C_Jnt_2, loc=<12 0 0>, rot=<0 0 0 1>}
 C_Jnt_4 {parent=C_Jnt_3, loc=<12 0 0>, rot=<0 0 0 1>}
 C_Jnt_5 {parent=C_Jnt_4, loc=<12 0 0>, rot=<0 0 0 1>}
 C_Jnt_6 {parent=C_Jnt_5, loc=<12 0 0>, rot=<0 0 0 1>}
 D_Jnt_01 {parent=pelvis, loc=<2.0718 -0.1134 14.5395>, rot=<0.074 -0.0013 0.9971 0.0179>}
 D_Jnt_02 {parent=D_Jnt_01, loc=<7 0 0>, rot=<0 0 0 1>}
 D_Jnt_03 {parent=D_Jnt_02, loc=<12 0 0>, rot=<0 0 0 1>}
 D_Jnt_04 {parent=D_Jnt_03, loc=<12 0 0>, rot=<0 0 0 1>}
 D_Jnt_05 {parent=D_Jnt_04, loc=<12 0 0>, rot=<0 0 0 1>}
 D_Jnt_06 {parent=D_Jnt_05, loc=<12 0 0>, rot=<0 0 0 1>}
 G_Jnt_01 {parent=pelvis, loc=<1.7859 -8.8725 9.5725>, rot=<-0.0089 -0.5693 0.822 0.0053>}
 G_Jnt_02 {parent=G_Jnt_01, loc=<7 0 0>, rot=<0 0 0 1>}
 G_Jnt_03 {parent=G_Jnt_02, loc=<12 0 0>, rot=<0 0 0 1>}
 G_Jnt_04 {parent=G_Jnt_03, loc=<12 0 0>, rot=<0 0 0 1>}
 G_Jnt_05 {parent=G_Jnt_04, loc=<12 0 0>, rot=<0 0 0 1>}
 G_Jnt_06 {parent=G_Jnt_05, loc=<12 0 0>, rot=<0 0 0 1>}
 E_Jnt_01 {parent=pelvis, loc=<2.2192 -11.3073 0>, rot=<-0.0062 -0.7071 0.7071 0.0062>}
 E_Jnt_02 {parent=E_Jnt_01, loc=<7 0 0>, rot=<0 0 0 1>}
 E_Jnt_03 {parent=E_Jnt_02, loc=<12 0 0>, rot=<0 0 0 1>}
 E_Jnt_04 {parent=E_Jnt_03, loc=<12 0 0>, rot=<0 0 0 1>}
 E_Jnt_05 {parent=E_Jnt_04, loc=<12 0 0>, rot=<0 0 0 1>}
 E_Jnt_06 {parent=E_Jnt_05, loc=<12 0 0>, rot=<0 0 0 1>}
 F_Jnt_01 {parent=pelvis, loc=<1.9448 -9.2724 -9.7828>, rot=<-0.0103 -0.8517 0.5239 0.0013>}
 F_Jnt_02 {parent=F_Jnt_01, loc=<7 0 0>, rot=<0 0 0 1>}
 F_Jnt_03 {parent=F_Jnt_02, loc=<12 0 0>, rot=<0 0 0 1>}
 F_Jnt_04 {parent=F_Jnt_03, loc=<12 0 0>, rot=<0 0 0 1>}
 F_Jnt_05 {parent=F_Jnt_04, loc=<12 0 0>, rot=<0 0 0 1>}
 F_Jnt_06 {parent=F_Jnt_05, loc=<12 0 0>, rot=<0 0 0 1>}
 B_Jnt_01 {parent=pelvis, loc=<2.0718 -0.1134 -15.5874>, rot=<-0.0179 -0.9982 0.001 -0.0566>}
 B_Jnt_02 {parent=B_Jnt_01, loc=<7 0 0>, rot=<0 0 0 1>}
 B_Jnt_03 {parent=B_Jnt_02, loc=<12 0 0>, rot=<0 0 0 1>}
 B_Jnt_04 {parent=B_Jnt_03, loc=<12 0 0>, rot=<0 0 0 1>}
 B_Jnt_05 {parent=B_Jnt_04, loc=<12 0 0>, rot=<0 0 0 1>}
 B_Jnt_06 {parent=B_Jnt_05, loc=<12 0 0>, rot=<0 0 0 1>}
 A_Jnt_1 {parent=pelvis, loc=<1.7301 5.7884 -11.4057>, rot=<-0.0528 -0.8002 -0.586 -0.1161>}
 A_Jnt_2 {parent=A_Jnt_1, loc=<7 0 0>, rot=<0 0 0 1>}
 A_Jnt_3 {parent=A_Jnt_2, loc=<12 0 0>, rot=<0 0 0 1>}
 A_Jnt_4 {parent=A_Jnt_3, loc=<12 0 0>, rot=<0 0 0 1>}
 A_Jnt_5 {parent=A_Jnt_4, loc=<12 0 0>, rot=<0 0 0 1>}
 A_Jnt_6 {parent=A_Jnt_5, loc=<12 0 0>, rot=<0 0 0 1>}
]
variables [
 Baseball Slide {type=double, default=0}
 CrouchWalk {type=double, default=0}
 GoldenGloryEmote {type=bool, default=False}
 L Leg Front Mid SPR {type=double, default=0}
 L Leg Front SPR {type=double, default=0}
 L_Leg_Rotback {type=double, default=0}
 LeftMove {type=double, default=0}
 MoveBackward {type=double, default=0}
 MoveForward {type=double, default=0}
 MoveLeft {type=double, default=0}
 MoveRight {type=double, default=0}
 OnlyCrouching {type=double, default=0}
 PlayingEmote {type=double, default=0}
 R Leg Front Mid SPR {type=double, default=0}
 R Leg Front SPR {type=double, default=0}
 RBAN Enable {type=bool, default=False}
 R_Leg_Rotback {type=double, default=0}
 Running {type=double, default=0}
 SPR L Leg Itw {type=double, default=0}
 SPR R Leg Itw {type=double, default=0}
 Slide {type=double, default=0}
 Sprinting {type=double, default=0}
 jump {type=double, default=0}
]
