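# Post-process Control Rig for the F_MED_BurnLeer female player body that corrects skirt and chain bones per movement or emote state.
# A three-branch sequence runs chained ModifyTransforms offsets weighted by state variables, many smoothed with AlphaInterp, plus one constant back-skirt tweak.
# GroundMovement is derived from pelvis height through a remap and curve, then scales the crouch chain correction.
# The FrontEnd skirt offsets are not connected to execution.
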ControlRigBlueprint /BRCosmetics/Characters/Player/Female/Medium/Bodies/F_MED_BurnLeer/Meshes/F_MED_BurnLeer_CtrlRig_Post
graphs [
 RigVMFunctionLibrary {links [], nodes []}
 RigVMFunction_Sequence_1_ContainedGraph {nodes [
   n1 = RigVMFunctionEntryNode()
   RigVMFunction_Sequence = &RigVMFunction_Sequence(ExecuteContext=@n1.ExecuteContext)
   n2 = &RigVMFunction_Sequence(ExecuteContext=@RigVMFunction_Sequence.B)
   RigVMFunctionReturnNode(A=@RigVMFunction_Sequence.A, B=@n2.A, C=@n2.B)
  ]}
 RigVMModel {nodes [
   n1 = RigVMVariableNode() {variable=FrontEnd}
   &RigUnit_ModifyTransforms(
    ExecutePin=()
    ItemToModify="((Item=(Type=Bone,Name=\"skirt_L_f_Master\"),Transform=(Rotation=(X=0,Y=0.2164,Z=0,W=0.9763),Translation=(X=0,Y=0,Z=0),Scale3D=(X=1,Y=1,Z=1))),(Item=(Type=Bone,Name=\"dyn_skirt_l_f2\"),Transform=(Rotation=(X=0,Y=-0.1564,Z=0,W=0.9877),Translation=(X=0,Y=0,Z=0), ...(truncated)"
    WeightMinimum=0
    Weight=@n1.Value
   )
   > &RigUnit_ModifyTransforms(
    ItemToModify="((Item=(Type=Bone,Name=\"skirt_b_m_Master\"),Transform=(Rotation=(X=0,Y=-0.0436,Z=0,W=0.999),Translation=(X=0,Y=0,Z=0),Scale3D=(X=1,Y=1,Z=1))),(Item=(Type=Bone,Name=\"dyn_skirt_b_02\"),Transform=(Rotation=(X=0,Y=0,Z=0,W=1),Translation=(X=0,Y=0,Z=0),Scale3D=(X= ...(truncated)"
    WeightMinimum=0
    Weight=@n1.Value
   )
   RigUnit_BeginExecution = &RigUnit_BeginExecution()
   RigVMFunction_Sequence = RigVMAggregateNode(ExecuteContext=@RigUnit_BeginExecution.ExecutePin)
   n2 = RigVMVariableNode() {variable=BaseBallSprint}
   &RigUnit_ModifyTransforms(
    ItemToModify="((Item=(Type=Bone,Name=\"skirt_L_b_Master\"),Transform=(Rotation=(X=0,Y=0,Z=0,W=1),Translation=(X=0,Y=0,Z=0),Scale3D=(X=1,Y=1,Z=1))),(Item=(Type=Bone,Name=\"skirt_b_m_Master\"),Transform=(Rotation=(X=0,Y=0,Z=0,W=1),Translation=(X=0,Y=0,Z=0),Scale3D=(X=1,Y=1,Z= ...(truncated)"
    WeightMinimum=0
    ExecutePin=@RigVMFunction_Sequence.B
    Weight=@n2.Value
   )
   > &RigUnit_ModifyTransforms(
    ItemToModify="((Item=(Type=Bone,Name=\"None\"),Transform=(Rotation=(X=0,Y=0,Z=0,W=1),Translation=(X=0,Y=0,Z=0),Scale3D=(X=1,Y=1,Z=1))),(Item=(Type=Bone,Name=\"dyn_skirt_r_02\"),Transform=(Rotation=(X=-0.3153,Y=-0.0679,Z=0.0772,W=0.9434),Translation=(X=-0.1718,Y=0.568,Z=0.76 ...(truncated)"
    WeightMinimum=0
    Weight=@n2.Value
   )
   n3 = &RigVMFunction_AlphaInterp(InterpSpeedDecreasing=10, InterpSpeedIncreasing=15, Scale=1, bClampResult=False, Value=RigVMVariableNode() {variable=Skydiving}.Value)
   > &RigUnit_ModifyTransforms(
    ItemToModify="((Item=(Type=Bone,Name=\"None\"),Transform=(Rotation=(X=0,Y=0,Z=0,W=1),Translation=(X=0,Y=0,Z=0),Scale3D=(X=1,Y=1,Z=1))),(Item=(Type=Bone,Name=\"dyn_skirt_r_master\"),Transform=(Rotation=(X=0,Y=-0.0872,Z=0,W=0.9962),Translation=(X=0,Y=0,Z=0),Scale3D=(X=1,Y=1,Z ...(truncated)"
    WeightMinimum=0
    Weight=@n3.Result
   )
   n4 = &RigVMFunction_AlphaInterp(InterpSpeedDecreasing=10, InterpSpeedIncreasing=10, Scale=1, bClampResult=False, Value=RigVMVariableNode() {variable=Emote}.Value)
   > &RigUnit_ModifyTransforms(
    ItemToModify="((Item=(Type=Bone,Name=\"skirt_b_m_Master\"),Transform=(Rotation=(X=0,Y=0,Z=0,W=1),Translation=(X=0,Y=0,Z=0.7336),Scale3D=(X=1,Y=1,Z=1))),(Item=(Type=Bone,Name=\"skirt_r_b_Master\"),Transform=(Rotation=(X=0,Y=0,Z=0,W=1),Translation=(X=0,Y=-0.7336,Z=0.1507),Sca ...(truncated)"
    WeightMinimum=0
    Weight=@n4.Result
   )
   > &RigUnit_ModifyTransforms(
    ItemToModify="((Item=(Type=Bone,Name=\"None\"),Transform=(Rotation=(X=0,Y=0,Z=0,W=1),Translation=(X=0,Y=0,Z=0),Scale3D=(X=1,Y=1,Z=1))),(Item=(Type=Bone,Name=\"dyn_skirt_r_f_02\"),Transform=(Rotation=(X=0,Y=-0.0436,Z=0,W=0.999),Translation=(X=0,Y=0,Z=0),Scale3D=(X=1,Y=1,Z=1) ...(truncated)"
    WeightMinimum=0
    Weight=RigVMVariableNode() {variable=CrouchForward}.Value
   )
   RigVMFunction_MathBoolToFloat = &RigVMFunction_MathBoolToFloat(Value=RigVMVariableNode() {variable=Zipline}.Value)
   n5 = &RigVMFunction_AlphaInterp(InterpSpeedDecreasing=8, InterpSpeedIncreasing=10, Scale=1, bClampResult=False, Value=@RigVMFunction_MathBoolToFloat.Result)
   > &RigUnit_ModifyTransforms(
    ItemToModify=((Item=(Type=Bone,Name="dyn_skirt_l_b_01"),Transform=(Rotation=(X=0,Y=0.0872,Z=0,W=0.9962),Translation=(X=0,Y=0,Z=5),Scale3D=(X=1,Y=1,Z=1))))
    WeightMinimum=0
    Weight=@n5.Result
   )
   > &RigUnit_ModifyTransforms(
    ItemToModify=((Item=(Type=Bone,Name="None"),Transform=(Rotation=(X=0,Y=0,Z=0,W=1),Translation=(X=0,Y=0,Z=0),Scale3D=(X=1,Y=1,Z=1))))
    Weight=1
    WeightMinimum=0
   )
   &RigUnit_ModifyTransforms(
    ItemToModify=((Item=(Type=Bone,Name="dyn_skirt_b_02"),Transform=(Rotation=(X=0,Y=0.0698,Z=0,W=0.9976),Translation=(X=0,Y=0,Z=0),Scale3D=(X=1,Y=1,Z=1))))
    Weight=1
    WeightMinimum=0
    ExecutePin=@RigVMFunction_Sequence.C
   )
   n6 = &RigVMFunction_AnimEvalRichCurve(
    Curve="(EditorCurveData=(Keys=((InterpMode=RCIM_Cubic,TangentMode=RCTM_Break,TangentWeightMode=RCTWM_WeightedNone,Time=0,Value=0,ArriveTangent=0,ArriveTangentWeight=0,LeaveTangent=0,LeaveTangentWeight=0),(InterpMode=RCIM_Cubic,TangentMode=RCTM_Break,TangentWeight ...(truncated)"
    SourceMinimum=0.5
    TargetMinimum=0
    Value=&RigVMFunction_MathDoubleRemap(
     SourceMaximum=30
     SourceMinimum=60
     TargetMaximum=0
     TargetMinimum=1
     Value=&RigUnit_GetRelativeTransformForItem(Child=(Type=Bone,Name="pelvis"), Parent=(Type=Bone,Name="root"), bChildInitial=False, bParentInitial=False).RelativeTransform.Translation.Z
    ).Result
   )
   RigVMVariableNode(Value=@n6.Result, ExecuteContext=@RigVMFunction_Sequence.A) {variable=GroundMovement}
   AlphaInterp = &RigVMFunction_AlphaInterp(InterpSpeedDecreasing=15, InterpSpeedIncreasing=15, Scale=1, bClampResult=False, Value=RigVMVariableNode() {variable=SkydiveFloating}.Value)
   > &RigUnit_ModifyTransforms(
    ItemToModify="((Item=(Type=Bone,Name=\"dyn_skirt_b_01\"),Transform=(Rotation=(X=0,Y=0,Z=0,W=1),Translation=(X=0,Y=0,Z=0),Scale3D=(X=1,Y=1,Z=1))),(Item=(Type=Bone,Name=\"dyn_skirt_b_01\"),Transform=(Rotation=(X=0,Y=-0.0872,Z=0,W=0.9962),Translation=(X=0,Y=0,Z=0),Scale3D=(X=1 ...(truncated)"
    WeightMinimum=0
    Weight=@AlphaInterp.Result
   )
   > &RigUnit_ModifyTransforms(
    ItemToModify="((Item=(Type=Bone,Name=\"None\"),Transform=(Rotation=(X=0,Y=0,Z=0,W=1),Translation=(X=0,Y=0,Z=0),Scale3D=(X=1,Y=1,Z=1))),(Item=(Type=Bone,Name=\"dyn_skirt_b_01\"),Transform=(Rotation=(X=0,Y=-0.0087,Z=0,W=1),Translation=(X=0,Y=0,Z=0),Scale3D=(X=1,Y=1,Z=1))),(It ...(truncated)"
    WeightMinimum=0
    Weight=@AlphaInterp.Result
   )
   n7 = &RigVMFunction_AlphaInterp(InterpSpeedDecreasing=10, InterpSpeedIncreasing=6, Scale=1, bClampResult=False, Value=RigVMVariableNode() {variable=running}.Value)
   > &RigUnit_ModifyTransforms(
    ItemToModify="((Item=(Type=Bone,Name=\"skirt_b_m_Master\"),Transform=(Rotation=(X=0,Y=-0.0436,Z=0,W=0.999),Translation=(X=0,Y=0,Z=0),Scale3D=(X=1,Y=1,Z=1))),(Item=(Type=Bone,Name=\"skirt_r_b_Master\"),Transform=(Rotation=(X=0,Y=-0.0436,Z=0,W=0.999),Translation=(X=0,Y=0,Z=0) ...(truncated)"
    WeightMinimum=0
    Weight=@n7.Result
   )
   > &RigUnit_ModifyTransforms(
    ItemToModify="((Item=(Type=Bone,Name=\"skirt_b_m_Master\"),Transform=(Rotation=(X=0,Y=-0.0872,Z=0,W=0.9962),Translation=(X=0,Y=0,Z=0),Scale3D=(X=1,Y=1,Z=1))),(Item=(Type=Bone,Name=\"skirt_r_b_Master\"),Transform=(Rotation=(X=0,Y=-0.0872,Z=0,W=0.9962),Translation=(X=0,Y=0,Z= ...(truncated)"
    WeightMinimum=0
    Weight=@n7.Result
   )
   > &RigUnit_ModifyTransforms(
    ItemToModify="((Item=(Type=Bone,Name=\"skirt_L_f_Master\"),Transform=(Rotation=(X=0,Y=-0.0872,Z=0,W=0.9962),Translation=(X=0,Y=-1,Z=1),Scale3D=(X=1,Y=1,Z=1))),(Item=(Type=Bone,Name=\"dyn_skirt_l_f2\"),Transform=(Rotation=(X=0,Y=0,Z=0,W=1),Translation=(X=0,Y=0,Z=0),Scale3D=( ...(truncated)"
    WeightMinimum=0
    Weight=@n7.Result
   )
   n8 = &RigVMFunction_AlphaInterp(InterpSpeedDecreasing=10, InterpSpeedIncreasing=8, Scale=1, bClampResult=False, Value=RigVMVariableNode() {variable=Sprinting}.Value)
   > &RigUnit_ModifyTransforms(
    ItemToModify="((Item=(Type=Bone,Name=\"skirt_r_b_Master\"),Transform=(Rotation=(X=-0.0019,Y=-0.0436,Z=-0.0436,W=0.9981),Translation=(X=0,Y=0,Z=0),Scale3D=(X=1,Y=1,Z=1))),(Item=(Type=Bone,Name=\"skirt_L_b_Master\"),Transform=(Rotation=(X=0.0019,Y=-0.0436,Z=0.0436,W=0.9981),T ...(truncated)"
    WeightMinimum=0
    Weight=@n8.Result
   )
   n9 = &RigVMFunction_AlphaInterp(InterpSpeedDecreasing=12, InterpSpeedIncreasing=8, Scale=1, bClampResult=False, Value=RigVMVariableNode() {variable=jump}.Value)
   > &RigUnit_ModifyTransforms(
    ItemToModify="((Item=(Type=Bone,Name=\"dyn_chain_master\"),Transform=(Rotation=(X=-0.0028,Y=-0.0435,Z=0.0634,W=0.997),Translation=(X=-1.3052,Y=0,Z=-0.1141),Scale3D=(X=1,Y=1,Z=1))),(Item=(Type=Bone,Name=\"dyn_chain_01\"),Transform=(Rotation=(X=0,Y=-0.0872,Z=0,W=0.9962),Trans ...(truncated)"
    WeightMinimum=0
    Weight=@n9.Result
   )
   > &RigUnit_ModifyTransforms(
    ItemToModify="((Item=(Type=Bone,Name=\"None\"),Transform=(Rotation=(X=0,Y=0,Z=0,W=1),Translation=(X=0,Y=0,Z=0),Scale3D=(X=1,Y=1,Z=1))),(Item=(Type=Bone,Name=\"dyn_chain_master\"),Transform=(Rotation=(X=0,Y=-0.0262,Z=0,W=0.9997),Translation=(X=0,Y=0,Z=0),Scale3D=(X=1,Y=1,Z=1 ...(truncated)"
    WeightMinimum=0
    Weight=RigVMVariableNode() {variable=yokequick_custom}.Value
   )
   > &RigUnit_ModifyTransforms(
    ItemToModify="((Item=(Type=Bone,Name=\"None\"),Transform=(Rotation=(X=0,Y=0,Z=0,W=1),Translation=(X=0,Y=0,Z=0),Scale3D=(X=1,Y=1,Z=1))),(Item=(Type=Bone,Name=\"dyn_chain_master\"),Transform=(Rotation=(X=0,Y=0.0262,Z=0,W=0.9997),Translation=(X=0,Y=0,Z=0),Scale3D=(X=1,Y=1,Z=1) ...(truncated)"
    WeightMinimum=0
    Weight=RigVMVariableNode() {variable=FE_Emote}.Value
   )
   > &RigUnit_ModifyTransforms(
    ItemToModify=((Item=(Type=Bone,Name="dyn_chain_01"),Transform=(Rotation=(X=-0.0299,Y=0.0522,Z=-0.0012,W=-0.9982),Translation=(X=0,Y=0,Z=0),Scale3D=(X=1,Y=1,Z=1))))
    WeightMinimum=0
    Weight=RigVMVariableNode() {variable=only crouch}.Value
   )
   Multiply = &RigVMFunction_MathDoubleMul(
    A=&RigVMFunction_MathBoolToFloat(Value=RigVMVariableNode() {variable=Crouch}.Value).Result
    B=RigVMVariableNode() {variable=GroundMovement}.Value
   )
   > &RigUnit_ModifyTransforms(
    ItemToModify=((Item=(Type=Bone,Name="dyn_chain_01"),Transform=(Rotation=(X=-0.0004,Y=-0.0174,Z=0.0205,W=0.9996),Translation=(X=0,Y=0,Z=0),Scale3D=(X=1,Y=1,Z=1))))
    WeightMinimum=0
    Weight=@Multiply.Result
   )
  ]}
]
hierarchy [
 root {loc=<0 0 0>, rot=<0 0 0 1>}
 pelvis {parent=root, loc=<0 -0.5301 98.0571>, rot=<0.0127 0.707 -0.0127 -0.707>}
 spine_01 {parent=pelvis, loc=<6.1208 -0.9328 0>, rot=<0 0 0.034 -0.9994>}
 spine_02 {parent=spine_01, loc=<6.5997 -0.2095 0>, rot=<0 0 0.0238 -0.9997>}
 spine_03 {parent=spine_02, loc=<6.4914 -0.1237 0>, rot=<0 0 0.0149 -0.9999>}
 spine_04 {parent=spine_03, loc=<6.9598 0.0794 0>, rot=<0 0 -0.0132 -0.9999>}
 spine_05 {parent=spine_04, loc=<6.9987 0 0>, rot=<0 0 -0.0015 -1>}
 neck_01 {parent=spine_05, loc=<11.3138 0.1529 0>, rot=<0 0 -0.2003 -0.9797>}
 neck_02 {parent=neck_01, loc=<4.5627 0 0>, rot=<0 0 0.0504 -0.9987>}
 head {parent=neck_02, loc=<5.3914 0 0>, rot=<0 0 0.115 -0.9934>}
 clavicle_l {parent=spine_05, loc=<5.4831 -0.1053 -3.2946>, rot=<0.0465 -0.7289 -0.009 -0.6829>}
 upperarm_l {parent=clavicle_l, loc=<9.8658 -1.2136 -0.4105>, rot=<0.0164 -0.3245 -0.0269 -0.9454>}
 lowerarm_l {parent=upperarm_l, loc=<26.465 -0.8345 0.259>, rot=<0.0036 0.0161 -0.2298 -0.9731>}
 hand_l {parent=lowerarm_l, loc=<22.7355 -0.3203 -0.6105>, rot=<-0.6784 -0.0223 0.0735 -0.7307>}
 thumb_01_l {parent=hand_l, loc=<1.3118 -0.3191 -2.459>, rot=<0.6546 -0.4042 0.0664 -0.6354>}
 thumb_02_l {parent=thumb_01_l, loc=<5.9336 -0.0448 0.0401>, rot=<-0.0037 0.0213 0.298 -0.9543>}
 thumb_03_l {parent=thumb_02_l, loc=<2.2223 0.3682 -0.0907>, rot=<0.0118 -0.013 0.0066 -0.9998>}
 index_metacarpal_l {parent=hand_l, loc=<1.6535 0.0183 -2.96>, rot=<0.0158 -0.0124 -0.0173 -0.9996>}
 index_01_l {parent=index_metacarpal_l, loc=<8.3008 0.0357 -0.0524>, rot=<-0.0044 0.0485 0.0924 -0.9945>}
 index_02_l {parent=index_01_l, loc=<3.3587 -0.3004 0.0452>, rot=<-0.002 0.0027 0.3225 -0.9466>}
 index_03_l {parent=index_02_l, loc=<2.5578 -0.123 -0.0252>, rot=<-0.002 0.0067 0.2558 -0.9667>}
 middle_metacarpal_l {parent=hand_l, loc=<2.1856 -0.1725 -0.5984>, rot=<0.006 0.0168 -0.0329 -0.9993>}
 middle_01_l {parent=middle_metacarpal_l, loc=<7.3275 -0.2447 0.0129>, rot=<-0.0208 0.0042 0.0858 -0.9961>}
 middle_02_l {parent=middle_01_l, loc=<4.0123 0.1401 0.1175>, rot=<-0.0218 0.0258 0.3512 -0.9357>}
 middle_03_l {parent=middle_02_l, loc=<2.6077 0.0534 0.1373>, rot=<0.0082 0.0188 0.1708 -0.9851>}
 ring_metacarpal_l {parent=hand_l, loc=<2.2074 -0.3625 1.3638>, rot=<-0.0284 0.0722 -0.0458 -0.9959>}
 ring_01_l {parent=ring_metacarpal_l, loc=<7.0295 -0.6197 -0.1446>, rot=<-0.0489 -0.0252 0.1892 -0.9804>}
 ring_02_l {parent=ring_01_l, loc=<3.6174 0.4647 -0.1027>, rot=<0 0 0.2806 -0.9598>}
 ring_03_l {parent=ring_02_l, loc=<2.4152 0.114 0.0199>, rot=<0 0 0.1189 -0.9929>}
 pinky_metacarpal_l {parent=hand_l, loc=<1.7616 -0.9753 3.3814>, rot=<-0.0397 0.0814 -0.069 -0.9935>}
 pinky_01_l {parent=pinky_metacarpal_l, loc=<7.1709 -0.4831 -0.1094>, rot=<-0.0654 -0.0035 0.2242 -0.9723>}
 pinky_02_l {parent=pinky_01_l, loc=<2.8274 0.0893 -0.1043>, rot=<0.0037 0.0011 0.2854 -0.9584>}
 pinky_03_l {parent=pinky_02_l, loc=<2.2046 0.0307 -0.0078>, rot=<-0.0002 0.0098 0.1445 -0.9895>}
 deform_wrist_dn_base_l {parent=hand_l, loc=<0 -1.5326 0>, rot=<0 0 0 -1>}
 deform_wrist_dn_l {parent=deform_wrist_dn_base_l, loc=<0 0 0>, rot=<0 0 0 1>}
 deform_wrist_up_base_l {parent=hand_l, loc=<0 2.3009 0>, rot=<0 0 0 -1>}
 deform_wrist_up_l {parent=deform_wrist_up_base_l, loc=<0 0 0>, rot=<0 0 0 1>}
 lowerarm_twist_02_l {parent=lowerarm_l, loc=<10.6707 0 0>, rot=<0 0 0 1>}
 lowerarm_twist_01_l {parent=lowerarm_l, loc=<18.316 0 0>, rot=<0 0 0 1>}
 deform_elbow_in_base_l {parent=lowerarm_l, loc=<0.1834 3.6062 -0.2572>, rot=<0 0 0 -1>}
 deform_elbow_in_l {parent=deform_elbow_in_base_l, loc=<0 0 0>, rot=<0 0 0 1>}
 deform_elbow_out_base_l {parent=lowerarm_l, loc=<0 -4 0>, rot=<0 0 0 1>}
 deform_elbow_out_l {parent=deform_elbow_out_base_l, loc=<0 0 0>, rot=<0 0 0 1>}
 upperarm_twist_01_l {parent=upperarm_l, loc=<7.2135 -0.1134 0>, rot=<0 0 0 1>}
 deform_upperarm_bk_base_l {parent=upperarm_twist_01_l, loc=<-6.4053 -4.3746 0.314>, rot=<0 0 0 -1>}
 deform_upperarm_bk_l {parent=deform_upperarm_bk_base_l, loc=<0 0 0>, rot=<0 0 0 1>}
 deform_upperarm_fr_base_l {parent=upperarm_twist_01_l, loc=<-5.6312 5.092 0.3354>, rot=<0 0 0 -1>}
 deform_upperarm_fr_l {parent=deform_upperarm_fr_base_l, loc=<0 0 0>, rot=<0 0 0 1>}
 deform_upperarm_up_base_l {parent=upperarm_twist_01_l, loc=<-8.891 0.4825 3.9517>, rot=<0 0 0 -1>}
 deform_upperarm_up_l {parent=deform_upperarm_up_base_l, loc=<0 0 0>, rot=<0 0 0 1>}
 upperarm_twist_02_l {parent=upperarm_l, loc=<19.7998 -0.1134 0>, rot=<0 0 0 1>}
 clavicle_r {parent=spine_05, loc=<5.4831 -0.1053 3.2946>, rot=<0.7289 0.0465 -0.6829 0.009>}
 upperarm_r {parent=clavicle_r, loc=<-9.8658 1.2136 0.4105>, rot=<0.0164 -0.3245 -0.0269 -0.9454>}
 lowerarm_r {parent=upperarm_r, loc=<-26.465 0.8345 -0.259>, rot=<0.0036 0.0161 -0.2298 -0.9731>}
 hand_r {parent=lowerarm_r, loc=<-22.7355 0.3203 0.6106>, rot=<-0.6784 -0.0223 0.0735 -0.7307>}
 thumb_01_r {parent=hand_r, loc=<-1.3118 0.3191 2.459>, rot=<0.6546 -0.4042 0.0664 -0.6354>}
 thumb_02_r {parent=thumb_01_r, loc=<-5.9336 0.0448 -0.0401>, rot=<-0.0037 0.0213 0.298 -0.9543>}
 thumb_03_r {parent=thumb_02_r, loc=<-2.2223 -0.3682 0.0907>, rot=<0.0118 -0.013 0.0066 -0.9998>}
 index_metacarpal_r {parent=hand_r, loc=<-1.6535 -0.0183 2.96>, rot=<0.0158 -0.0124 -0.0173 -0.9996>}
 index_01_r {parent=index_metacarpal_r, loc=<-8.3008 -0.0357 0.0524>, rot=<-0.0044 0.0485 0.0924 -0.9945>}
 index_02_r {parent=index_01_r, loc=<-3.3587 0.3004 -0.0452>, rot=<-0.002 0.0027 0.3225 -0.9466>}
 index_03_r {parent=index_02_r, loc=<-2.5578 0.123 0.0252>, rot=<-0.002 0.0067 0.2558 -0.9667>}
 middle_metacarpal_r {parent=hand_r, loc=<-2.1856 0.1725 0.5984>, rot=<0.006 0.0168 -0.0329 -0.9993>}
 middle_01_r {parent=middle_metacarpal_r, loc=<-7.3275 0.2447 -0.0129>, rot=<-0.0208 0.0042 0.0858 -0.9961>}
 middle_02_r {parent=middle_01_r, loc=<-4.0123 -0.1401 -0.1175>, rot=<-0.0218 0.0258 0.3512 -0.9357>}
 middle_03_r {parent=middle_02_r, loc=<-2.6077 -0.0534 -0.1373>, rot=<0.0082 0.0188 0.1708 -0.9851>}
 ring_metacarpal_r {parent=hand_r, loc=<-2.2074 0.3625 -1.3638>, rot=<-0.0284 0.0722 -0.0458 -0.9959>}
 ring_01_r {parent=ring_metacarpal_r, loc=<-7.0295 0.6197 0.1446>, rot=<-0.0489 -0.0252 0.1892 -0.9804>}
 ring_02_r {parent=ring_01_r, loc=<-3.6174 -0.4647 0.1027>, rot=<0 0 0.2806 -0.9598>}
 ring_03_r {parent=ring_02_r, loc=<-2.4152 -0.114 -0.0199>, rot=<0 0 0.1189 -0.9929>}
 pinky_metacarpal_r {parent=hand_r, loc=<-1.7616 0.9753 -3.3814>, rot=<-0.0397 0.0814 -0.069 -0.9935>}
 pinky_01_r {parent=pinky_metacarpal_r, loc=<-7.1709 0.4831 0.1094>, rot=<-0.0654 -0.0035 0.2242 -0.9723>}
 pinky_02_r {parent=pinky_01_r, loc=<-2.8274 -0.0893 0.1043>, rot=<0.0037 0.0011 0.2854 -0.9584>}
 pinky_03_r {parent=pinky_02_r, loc=<-2.2046 -0.0307 0.0078>, rot=<-0.0002 0.0098 0.1445 -0.9895>}
 deform_wrist_dn_base_r {parent=hand_r, loc=<-0.0001 1.5328 0>, rot=<0 0 0 -1>}
 deform_wrist_dn_r {parent=deform_wrist_dn_base_r, loc=<0 0 0>, rot=<0 0 0 1>}
 deform_wrist_up_base_r {parent=hand_r, loc=<-0.0002 -2.3007 -0.0001>, rot=<0 0 0 -1>}
 deform_wrist_up_r {parent=deform_wrist_up_base_r, loc=<0 0 0>, rot=<0 0 0 1>}
 lowerarm_twist_02_r {parent=lowerarm_r, loc=<-10.6707 0 0>, rot=<0 0 0 -1>}
 lowerarm_twist_01_r {parent=lowerarm_r, loc=<-18.316 0 0>, rot=<0 0 0 -1>}
 deform_elbow_in_base_r {parent=lowerarm_r, loc=<-0.1834 -3.6062 0.2572>, rot=<0 0 0 -1>}
 deform_elbow_in_r {parent=deform_elbow_in_base_r, loc=<0 0 0>, rot=<0 0 0 1>}
 deform_elbow_out_base_r {parent=lowerarm_r, loc=<0 4 -0.0001>, rot=<0 0 0 -1>}
 deform_elbow_out_r {parent=deform_elbow_out_base_r, loc=<0 0 0>, rot=<0 0 0 1>}
 upperarm_twist_01_r {parent=upperarm_r, loc=<-7.2135 0.1134 0>, rot=<0 0 0 -1>}
 deform_upperarm_bk_base_r {parent=upperarm_twist_01_r, loc=<6.4053 4.3746 -0.314>, rot=<0 0 0 -1>}
 deform_upperarm_bk_r {parent=deform_upperarm_bk_base_r, loc=<0 0 0>, rot=<0 0 0 1>}
 deform_upperarm_fr_base_r {parent=upperarm_twist_01_r, loc=<5.6312 -5.092 -0.3354>, rot=<0 0 0 -1>}
 deform_upperarm_fr_r {parent=deform_upperarm_fr_base_r, loc=<0 0 0>, rot=<0 0 0 1>}
 deform_upperarm_up_base_r {parent=upperarm_twist_01_r, loc=<8.891 -0.4825 -3.9517>, rot=<0 0 0 -1>}
 deform_upperarm_up_r {parent=deform_upperarm_up_base_r, loc=<0 0 0>, rot=<0 0 0 1>}
 upperarm_twist_02_r {parent=upperarm_r, loc=<-19.7998 0.1134 0>, rot=<0 0 0 -1>}
 attach_backpack {parent=spine_05, loc=<-0.123 -13.473 0>, rot=<-0.0093 -0.707 0.0093 -0.707>}
 attach_cape {parent=spine_05, loc=<9.017 -11.714 0>, rot=<0.0616 -0.7044 -0.0616 -0.7044>}
 deform_pec_aim_base_l {parent=spine_04, loc=<3.2456 9.3309 -1.2237>, rot=<0.0603 -0.4033 0.3318 -0.8507>}
 deform_pec_aim_l {parent=deform_pec_aim_base_l, loc=<0 0 0>, rot=<0 0 0 1>}
 deform_pec_base_l {parent=deform_pec_aim_l, loc=<9.3665 0 0>, rot=<0 0 0 1>}
 deform_pec_l {parent=deform_pec_base_l, loc=<0 0 0>, rot=<0 0 0 1>}
 deform_pec_aim_base_r {parent=spine_04, loc=<3.2452 9.3309 1.2237>, rot=<-0.4033 -0.0603 0.8507 0.3318>}
 deform_pec_aim_r {parent=deform_pec_aim_base_r, loc=<0 0 0>, rot=<0 0 0 1>}
 deform_pec_base_r {parent=deform_pec_aim_r, loc=<-9.3667 -0.0003 -0.0004>, rot=<0 0 0 1>}
 deform_pec_r {parent=deform_pec_base_r, loc=<0 0 0>, rot=<0 0 0 1>}
 dyn_zipper_master {parent=spine_04, loc=<5.4478 10.4678 0>, rot=<0.0859 0.7019 0.7019 0.0859>}
 dyn_zipper_01 {parent=dyn_zipper_master, loc=<0 0 0>, rot=<0 0 0 1>}
 dyn_jacketback_r_b_master {parent=spine_03, loc=<7.2131 -5.1939 6.2039>, rot=<0.0876 -0.5737 0.8113 -0.0707>}
 dyn_jacketback_r_b_01 {parent=dyn_jacketback_r_b_master, loc=<0 0 0>, rot=<0 0 0 1>}
 dyn_jacketSide_r_01_master {parent=spine_03, loc=<5.3794 0.358 9.8703>, rot=<0.0699 0.0051 0.9967 -0.0404>}
 dyn_jacketSide_r_01 {parent=dyn_jacketSide_r_01_master, loc=<0 0 0>, rot=<0 0 0 1>}
 dyn_jacketFront_r_a_master {parent=spine_03, loc=<4.6688 6.9325 8.4959>, rot=<0.0778 0.2607 0.9615 -0.0395>}
 dyn_jacketFront_r_a_01 {parent=dyn_jacketFront_r_a_master, loc=<0 0 0>, rot=<0 0 0 1>}
 dyn_jacketFront_r_master {parent=spine_03, loc=<4.6872 10.4711 4.378>, rot=<-0.0057 -0.7187 -0.6953 -0.0063>}
 dyn_jacketFront_r_01 {parent=dyn_jacketFront_r_master, loc=<0 0 0>, rot=<0 0 0 1>}
 dyn_jacketBack_master {parent=spine_03, loc=<7.9751 -6.3307 -0.5687>, rot=<0.0586 -0.7047 0.7047 -0.0586>}
 dyn_jacketBack_01 {parent=dyn_jacketBack_master, loc=<0 0 0>, rot=<0 0 0 1>}
 dyn_jacketback_l_b_master {parent=spine_03, loc=<6.7121 -4.4482 -7.3766>, rot=<0.0808 -0.8722 0.4715 -0.1022>}
 dyn_jacketback_l_b_01 {parent=dyn_jacketback_l_b_master, loc=<0 0 0>, rot=<0 0 0 1>}
 dyn_jacketSide_l_01_master {parent=spine_03, loc=<5.8886 0.3118 -9.6995>, rot=<0.0407 -0.9985 0.0008 -0.0361>}
 dyn_jacketSide_l_01 {parent=dyn_jacketSide_l_01_master, loc=<0 0 0>, rot=<0 0 0 1>}
 dyn_jacketFront_l_a_master {parent=spine_03, loc=<4.7533 7.0943 -8.6313>, rot=<0.0553 -0.9379 -0.3422 -0.0135>}
 dyn_jacketFront_l_a_01 {parent=dyn_jacketFront_l_a_master, loc=<0 0 0>, rot=<0 0 0 1>}
 dyn_jacketFront_l_master {parent=spine_03, loc=<4.6879 10.4377 -4.1135>, rot=<-0.0069 0.7071 0.7071 -0.0069>}
 dyn_jacketFront_l_01 {parent=dyn_jacketFront_l_master, loc=<0 0 0>, rot=<0 0 0 1>}
 dyn_bag_master {parent=spine_02, loc=<-0.2172 1.0018 -10.7897>, rot=<0.101 -0.9744 0.1004 -0.174>}
 dyn_bag_01 {parent=dyn_bag_master, loc=<0 0 0>, rot=<0 0 0 1>}
 thigh_l {parent=pelvis, loc=<-1.0855 -0.2109 -8.7053>, rot=<0.0009 0.0355 0.0155 -0.9993>}
 calf_l {parent=thigh_l, loc=<-42.5284 -0.0019 -0.7297>, rot=<0.0006 -0.0161 -0.0391 -0.9991>}
 foot_l {parent=calf_l, loc=<-46.7401 -2.5014 -0.592>, rot=<0.0427 -0.0238 0.0353 -0.9982>}
 ball_l {parent=foot_l, loc=<-3.5658 10.8066 -0.1544>, rot=<0.0001 -0.0001 -0.7063 -0.7079>}
 calf_twist_02_l {parent=calf_l, loc=<-19.1546 -1.7943 -0.2638>, rot=<0.0237 0.0095 -0.0035 -0.9997>}
 calf_twist_01_l {parent=calf_l, loc=<-35.8483 -1.9197 -0.5794>, rot=<0.0237 0.0095 -0.0035 -0.9997>}
 deform_knee_in_base_l {parent=calf_l, loc=<0.8031 -5.0231 0.2456>, rot=<0 0 0 -1>}
 deform_knee_in_l {parent=deform_knee_in_base_l, loc=<0 0 0>, rot=<0 0 0 1>}
 deform_knee_out_base_l {parent=calf_l, loc=<0 4 0>, rot=<0 0 0 -1>}
 deform_knee_out_l {parent=deform_knee_out_base_l, loc=<0 0 0>, rot=<0 0 0 1>}
 thigh_twist_01_l {parent=thigh_l, loc=<-27.952 0 0>, rot=<0.0013 0 0 -1>}
 deform_groin_base_l {parent=thigh_l, loc=<0.4801 5.6776 -0.2486>, rot=<0 0 0 -1>}
 deform_groin_l {parent=deform_groin_base_l, loc=<0 0 0>, rot=<0 0 0 1>}
 thigh_r {parent=pelvis, loc=<-1.0855 -0.2109 8.7053>, rot=<0.0355 -0.0009 0.9993 0.0155>}
 calf_r {parent=thigh_r, loc=<42.5284 0.0019 0.7297>, rot=<0.0006 -0.0161 -0.0391 -0.9991>}
 foot_r {parent=calf_r, loc=<46.7401 2.5014 0.592>, rot=<0.0427 -0.0238 0.0353 -0.9982>}
 ball_r {parent=foot_r, loc=<3.5658 -10.8066 0.1544>, rot=<0.0001 -0.0001 -0.7063 -0.7079>}
 calf_twist_02_r {parent=calf_r, loc=<19.1546 1.7943 0.2638>, rot=<0.0237 0.0095 -0.0035 -0.9997>}
 calf_twist_01_r {parent=calf_r, loc=<35.8483 1.9197 0.5794>, rot=<0.0237 0.0095 -0.0035 -0.9997>}
 deform_knee_in_base_r {parent=calf_r, loc=<-0.8031 5.0231 -0.2457>, rot=<0 0 0 -1>}
 deform_knee_in_r {parent=deform_knee_in_base_r, loc=<0 0 0>, rot=<0 0 0 1>}
 deform_knee_out_base_r {parent=calf_r, loc=<0 -4 0>, rot=<0 0 0 -1>}
 deform_knee_out_r {parent=deform_knee_out_base_r, loc=<0 0 0>, rot=<0 0 0 1>}
 thigh_twist_01_r {parent=thigh_r, loc=<27.952 0 0>, rot=<0.0013 0 0 -1>}
 deform_groin_base_r {parent=thigh_r, loc=<-0.4801 -5.6776 0.2486>, rot=<0 0 0 -1>}
 deform_groin_r {parent=deform_groin_base_r, loc=<0 0 0>, rot=<0 0 0 1>}
 deform_glute_base_l {parent=pelvis, loc=<-7.3466 -9.0213 -9.1245>, rot=<0.0172 0.031 -0.4477 -0.8935>}
 deform_glute_l {parent=deform_glute_base_l, loc=<0 0 0>, rot=<0 0 0 1>}
 deform_glute_base_r {parent=pelvis, loc=<-7.3464 -9.0213 9.1245>, rot=<0.031 -0.0172 0.8935 -0.4477>}
 deform_glute_r {parent=deform_glute_base_r, loc=<0 0 0>, rot=<0 0 0 1>}
 skirt_b_m_Master {parent=pelvis, loc=<2.2192 -11.3073 0>, rot=<-0.0062 -0.7071 0.7071 0.0062>}
 dyn_skirt_b_01 {parent=skirt_b_m_Master, loc=<0 0 0>, rot=<0 0 0 1>}
 dyn_skirt_b_02 {parent=dyn_skirt_b_01, loc=<7 0 0>, rot=<0 0 0 1>}
 dyn_skirt_b_03 {parent=dyn_skirt_b_02, loc=<12 0 0>, rot=<0 0 0 1>}
 dyn_skirt_b_04 {parent=dyn_skirt_b_03, loc=<12 0 0>, rot=<0 0 0 1>}
 dyn_skirt_b_05 {parent=dyn_skirt_b_04, loc=<12 0 0>, rot=<0 0 0 1>}
 dyn_skirt_b_06 {parent=dyn_skirt_b_05, loc=<12 0 0>, rot=<0 0 0 1>}
 dyn_skirt_r_master {parent=pelvis, loc=<4.9525 -0.2167 12.9815>, rot=<0.1792 -0.0032 0.9837 0.0176>}
 dyn_skirt_r_01 {parent=dyn_skirt_r_master, loc=<0 0 0>, rot=<0 0 0 1>}
 dyn_skirt_r_02 {parent=dyn_skirt_r_01, loc=<7 0 0>, rot=<0 -0.0755 0 -0.9971>}
 dyn_skirt_r_03 {parent=dyn_skirt_r_02, loc=<12 0 0>, rot=<0 -0.0703 0 -0.9975>}
 dyn_skirt_r_04 {parent=dyn_skirt_r_03, loc=<12 0 0>, rot=<0 0 0 1>}
 dyn_skirt_r_05 {parent=dyn_skirt_r_04, loc=<12 0 0>, rot=<0 0 0 1>}
 dyn_skirt_r_06 {parent=dyn_skirt_r_05, loc=<12 0 0>, rot=<0 0 0 1>}
 skirt_r_f_Master {parent=pelvis, loc=<5.2208 5.6279 8.5793>, rot=<0.1107 0.5823 0.8033 0.0579>}
 dyn_skirt_r_f_01 {parent=skirt_r_f_Master, loc=<0 0 0>, rot=<0 0.0146 0 -0.9999>}
 dyn_skirt_r_f_02 {parent=dyn_skirt_r_f_01, loc=<7 0 0>, rot=<0 0 0 1>}
 dyn_skirt_r_f_03 {parent=dyn_skirt_r_f_02, loc=<12 0 0>, rot=<0 -0.0624 0 -0.9981>}
 dyn_skirt_r_f_04 {parent=dyn_skirt_r_f_03, loc=<12 0 0>, rot=<0 0 0 1>}
 dyn_skirt_r_f_05 {parent=dyn_skirt_r_f_04, loc=<12 0 0>, rot=<0 0 0 1>}
 dyn_skirt_r_f_06 {parent=dyn_skirt_r_f_05, loc=<12 0 0>, rot=<0 0 0 1>}
 dyn_skirt_l_master {parent=pelvis, loc=<3.5328 -0.1658 -14.1095>, rot=<-0.0177 -0.9906 0.0024 -0.1354>}
 dyn_skirt_l_01 {parent=dyn_skirt_l_master, loc=<0 0 0>, rot=<0 0 0 1>}
 dyn_skirt_l_02 {parent=dyn_skirt_l_01, loc=<7 0 0>, rot=<0 -0.0688 0 -0.9976>}
 dyn_skirt_l_03 {parent=dyn_skirt_l_02, loc=<12 0 0>, rot=<0 -0.0445 0 -0.999>}
 dyn_skirt_l_04 {parent=dyn_skirt_l_03, loc=<12 0 0>, rot=<0 0 0 1>}
 dyn_skirt_l_05 {parent=dyn_skirt_l_04, loc=<12 0 0>, rot=<0 0 0 1>}
 dyn_skirt_l_06 {parent=dyn_skirt_l_05, loc=<12 0 0>, rot=<0 0 0 1>}
 skirt_L_f_Master {parent=pelvis, loc=<3.4589 5.0543 -11.0742>, rot=<-0.0528 -0.8002 -0.586 -0.1161>}
 dyn_skirt_l_f1 {parent=skirt_L_f_Master, loc=<0.0128 0 0.0002>, rot=<0 0.0185 0 -0.9998>}
 dyn_skirt_l_f2 {parent=dyn_skirt_l_f1, loc=<7 0 0>, rot=<0 -0.0239 0 -0.9997>}
 dyn_skirt_l_f3 {parent=dyn_skirt_l_f2, loc=<12 0 0>, rot=<0 -0.0598 0 -0.9982>}
 dyn_skirt_l_f4 {parent=dyn_skirt_l_f3, loc=<12 0 0>, rot=<0 0 0 1>}
 dyn_skirt_l_f5 {parent=dyn_skirt_l_f4, loc=<12 0 0>, rot=<0 0 0 1>}
 dyn_skirt_l_f6 {parent=dyn_skirt_l_f5, loc=<12 0 0>, rot=<0 0 0 1>}
 skirt_L_b_Master {parent=pelvis, loc=<1.9448 -9.2724 -9.7828>, rot=<-0.0103 -0.8517 0.5239 0.0013>}
 dyn_skirt_l_b_01 {parent=skirt_L_b_Master, loc=<0 0 0>, rot=<0 0 0 1>}
 dyn_skirt_l_b_02 {parent=dyn_skirt_l_b_01, loc=<7 0 0>, rot=<0 0 0 1>}
 dyn_skirt_l_b_03 {parent=dyn_skirt_l_b_02, loc=<12 0 0>, rot=<0 0 0 1>}
 dyn_skirt_l_b_04 {parent=dyn_skirt_l_b_03, loc=<12 0 0>, rot=<0 0 0 1>}
 dyn_skirt_l_b_05 {parent=dyn_skirt_l_b_04, loc=<12 0 0>, rot=<0 0 0 1>}
 dyn_skirt_l_b_06 {parent=dyn_skirt_l_b_05, loc=<12 0 0>, rot=<0 0 0 1>}
 skirt_r_b_Master {parent=pelvis, loc=<1.7859 -8.8725 9.5725>, rot=<-0.0089 -0.5693 0.822 0.0053>}
 dyn_skirt_r_b_01 {parent=skirt_r_b_Master, loc=<0 0 0>, rot=<0 0 0 1>}
 dyn_skirt_r_b_02 {parent=dyn_skirt_r_b_01, loc=<7 0 0>, rot=<0 0 0 1>}
 dyn_skirt_r_b_03 {parent=dyn_skirt_r_b_02, loc=<12 0 0>, rot=<0 0 0 1>}
 dyn_skirt_r_b_04 {parent=dyn_skirt_r_b_03, loc=<12 0 0>, rot=<0 0 0 1>}
 dyn_skirt_r_b_05 {parent=dyn_skirt_r_b_04, loc=<12 0 0>, rot=<0 0 0 1>}
 dyn_skirt_r_b_06 {parent=dyn_skirt_r_b_05, loc=<12 0 0>, rot=<0 0 0 1>}
 dyn_chain_master {parent=pelvis, loc=<9.1812 0.2014 11.4378>, rot=<0.2199 -0.0039 0.9754 0.0175>}
 dyn_chain_01 {parent=dyn_chain_master, loc=<0 0 0>, rot=<0 0 0 1>}
 dyn_chain_02 {parent=dyn_chain_01, loc=<8.111 0 0>, rot=<0 0 0 1>}
 thigh_l_null {parent=pelvis, type=Null, loc=<-6.5952 0.5039 -10.8978>, rot=<-0.0888 0.0655 0.2355 -0.9656>}
 thigh_r_null {parent=pelvis, type=Null, loc=<-6.1935 -2.2704 1.9465>, rot=<-0.1506 0.2545 -0.9497 -0.1027>}
]
variables [
 BaseBallSprint {type=double, default=0}
 Crouch {type=bool, default=False}
 CrouchForward {type=double, default=0}
 Emote {type=double, default=0}
 FE_Emote {type=double, default=0}
 FrontEnd {type=double, default=0}
 GroundMovement {type=double, default=0}
 SkydiveFloating {type=double, default=0}
 Skydiving {type=double, default=0}
 Sprinting {type=double, default=0}
 Zipline {type=bool, default=False}
 jump {type=double, default=0}
 only crouch {type=double, default=0}
 running {type=double, default=0}
 yokequick_custom {type=double, default=0}
]
