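# Physics asset for the BurnLeer female medium body cosmetic, providing capsule bodies for its dynamic skirt and chain.
# Pelvis, thigh and calf capsules form the body around seven four-segment skirt strips and a two-segment chain.
# Each strip and the chain hangs from a master body with free rotation on the pelvis, then narrow swing and zero twist limits.
# An extra free-rotation constraint links the tips of the right and right-front strips.
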PhysicsAsset /BRCosmetics/Characters/Player/Female/Medium/Bodies/F_MED_BurnLeer/Meshes/F_MED_BurnLeer_FE_Physics
bodies {
 calf_l [{shapeType=Capsule, length=35, radius=8, center=<-22.9034 -1.8748 0>, rotation=<rot 90 3.1447 0>}]
 calf_r [{shapeType=Capsule, length=39.9629, radius=8, center=<22.9034 1.8748 0>, rotation=<rot 90 -176.8553 0>}]
 dyn_chain_01 [{shapeType=Capsule, length=8.2297, radius=5.2579, shapeName=dyn_chain_01_capsule, center=<5.6699 0 2.9345>, rotation=<rot 90 90 0>}]
 dyn_chain_02 [{shapeType=Capsule, length=6.8452, radius=6.3694, shapeName=dyn_chain_02_capsule, center=<4.1726 0 3.3625>, rotation=<rot 90 90 0>}]
 dyn_chain_master [{shapeType=Capsule, length=0.505, radius=0.505, shapeName=dyn_chain_master_capsule, rotation=<rot 0 -90 0>}]
 dyn_skirt_b_01 [{shapeType=Capsule, length=5.2388, radius=4, shapeName=dyn_skirt_b_01_capsule, center=<5 0 5>, rotation=<rot 83.3587 180 180>}]
 dyn_skirt_b_02 [{shapeType=Capsule, length=10, radius=6, shapeName=dyn_skirt_b_02_capsule, center=<5 0 3.5>, rotation=<rot 90 90 0>}]
 dyn_skirt_b_03 [{shapeType=Capsule, length=12, radius=7, shapeName=dyn_skirt_b_03_capsule, center=<5 0 4.3>, rotation=<rot 90 90 0>}]
 dyn_skirt_b_04 [{shapeType=Capsule, length=15, radius=8, shapeName=dyn_skirt_b_04_capsule, center=<5 0 5>, rotation=<rot 90 90 0>}]
 dyn_skirt_l_01 [{
   shapeType=Capsule
   length=4.6689
   radius=4
   shapeName=dyn_skirt_l_01_capsule
   center=<5.0248 0 4.3202>
   rotation=<rot 81.9532 -179.9994 -179.9996>
  }]
 dyn_skirt_l_02 [{shapeType=Capsule, length=10, radius=6, shapeName=dyn_skirt_l_02_capsule, center=<5 0 6.7515>, rotation=<rot 90 90 0>}]
 dyn_skirt_l_03 [{shapeType=Capsule, length=15, radius=7, shapeName=dyn_skirt_l_03_capsule, center=<5 0 6.6369>, rotation=<rot 90 90 0>}]
 dyn_skirt_l_04 [{shapeType=Capsule, length=20, radius=8, shapeName=dyn_skirt_l_04_capsule, center=<5 0 7.4684>, rotation=<rot 90 90 0>}]
 dyn_skirt_l_b_01 [{shapeType=Capsule, length=5.1308, radius=4, shapeName=dyn_skirt_l_b_01_capsule, center=<5 0 5.7>, rotation=<rot 90 0 0>}]
 dyn_skirt_l_b_02 [{shapeType=Capsule, length=10, radius=6, shapeName=dyn_skirt_l_b_02_capsule, center=<5 0 5>, rotation=<rot 90 90 0>}]
 dyn_skirt_l_b_03 [{shapeType=Capsule, length=15, radius=7.5, shapeName=dyn_skirt_l_b_03_capsule, center=<5 0 5.5>, rotation=<rot 90 90 0>}]
 dyn_skirt_l_b_04 [{shapeType=Capsule, length=20, radius=8.5, shapeName=dyn_skirt_l_b_04_capsule, center=<5 0 6>, rotation=<rot 90 90 0>}]
 dyn_skirt_l_f1 [{
   shapeType=Capsule
   length=6.4756
   radius=4.5244
   shapeName=dyn_skirt_l_f1_capsule
   center=<4.9698 0.0063 5.6468>
   rotation=<rot 89.1678 168.1561 -179.9979>
  }]
 dyn_skirt_l_f2 [{
   shapeType=Capsule
   length=8
   radius=5
   shapeName=dyn_skirt_l_f2_capsule
   center=<4.9808 -0.0256 5.6787>
   rotation=<rot 88.8738 167.0101 -179.999>
  }]
 dyn_skirt_l_f3 [{
   shapeType=Capsule
   length=8
   radius=5
   shapeName=dyn_skirt_l_f3_capsule
   center=<5.4821 -4.4895 3.4655>
   rotation=<rot 85.0078 -15.5106 -5.1918>
  }]
 dyn_skirt_l_f4 [{
   shapeType=Capsule
   length=8
   radius=5
   shapeName=dyn_skirt_l_f4_capsule
   center=<5.2062 -6.6336 1.6305>
   rotation=<rot 73.7276 -55.6013 -46.7696>
  }]
 dyn_skirt_l_master [{shapeType=Capsule, length=2, radius=2, shapeName=dyn_skirt_l_master_capsule, rotation=<rot 90 0 0>}]
 dyn_skirt_r_01 [{
   shapeType=Capsule
   length=5.3621
   radius=4.4126
   shapeName=dyn_skirt_r_01_capsule
   center=<2.5296 1.4369 3.8163>
   rotation=<rot 90 -90.0006 0>
  }]
 dyn_skirt_r_02 [{
   shapeType=Capsule
   length=12.8656
   radius=5.7661
   shapeName=dyn_skirt_r_02_capsule
   center=<2.7442 1.202 6.3041>
   rotation=<rot 85.4365 0.0001 -89.9998>
  }]
 dyn_skirt_r_03 [{shapeType=Capsule, length=14.7331, radius=6.2669, shapeName=dyn_skirt_r_03_capsule, center=<3.7919 0 5.157>, rotation=<rot 90 90 0>}]
 dyn_skirt_r_04 [{shapeType=Capsule, length=15.024, radius=5.976, shapeName=dyn_skirt_r_04_capsule, center=<4.3171 0 4.8518>, rotation=<rot 90 90 0>}]
 dyn_skirt_r_b_01 [{
   shapeType=Capsule
   length=7.9278
   radius=5.2768
   shapeName=dyn_skirt_r_b_01_capsule
   center=<-2.3343 0 5.0668>
   rotation=<rot 82.7487 179.9999 90>
  }]
 dyn_skirt_r_b_02 [{shapeType=Capsule, length=13.6496, radius=6.8054, shapeName=dyn_skirt_r_b_02_capsule, center=<-0.2249 0 5.661>, rotation=<rot 90 90 0>}]
 dyn_skirt_r_b_03 [{shapeType=Capsule, length=15, radius=7.5, shapeName=dyn_skirt_r_b_03_capsule, center=<5 0 7>, rotation=<rot 90 90 0>}]
 dyn_skirt_r_b_04 [{shapeType=Capsule, length=20, radius=8.5, shapeName=dyn_skirt_r_b_04_capsule, center=<5 0 7.5>, rotation=<rot 90 90 0>}]
 dyn_skirt_r_f_01 [{
   shapeType=Capsule
   length=4.5845
   radius=3.4917
   shapeName=dyn_skirt_r_f_01_capsule
   center=<5.0232 0 2.7607>
   rotation=<rot 84.4687 179.9999 179.9999>
  }]
 dyn_skirt_r_f_02 [{
   shapeType=Capsule
   length=8.8773
   radius=4.6227
   shapeName=dyn_skirt_r_f_02_capsule
   center=<4.3087 2.7198 4.8817>
   rotation=<rot 80.3504 107.6074 90.7171>
  }]
 dyn_skirt_r_f_03 [{
   shapeType=Capsule
   length=8.7176
   radius=4.7824
   shapeName=dyn_skirt_r_f_03_capsule
   center=<4.9647 5.7539 3.4495>
   rotation=<rot 75.1651 75.7926 64.7814>
  }]
 dyn_skirt_r_f_04 [{
   shapeType=Capsule
   length=9.44
   radius=4.06
   shapeName=dyn_skirt_r_f_04_capsule
   center=<5.057 7.1298 1.3162>
   rotation=<rot 59.8179 72.4294 63.5583>
  }]
 dyn_skirt_r_master [{shapeType=Capsule, length=2, radius=2, shapeName=dyn_skirt_r_master_capsule, rotation=<rot 90 0 0>}]
 pelvis [{shapeType=Capsule, length=12.8892, radius=9.1991, center=<-5.2761 0 0>}]
 skirt_L_b_Master [{shapeType=Capsule, length=2, radius=2, shapeName=skirt_L_b_Master_capsule, rotation=<rot 90 0 0>}]
 skirt_L_f_Master [{shapeType=Capsule, length=2, radius=2, shapeName=skirt_L_f_Master_capsule, rotation=<rot 90 0 0>}]
 skirt_b_m_Master [{shapeType=Capsule, length=2, radius=2, shapeName=skirt_b_m_Master_capsule, rotation=<rot 90 0 0>}]
 skirt_r_b_Master [{shapeType=Capsule, length=2, radius=2, shapeName=skirt_r_b_Master_capsule, rotation=<rot 90 0 0>}]
 skirt_r_f_Master [{shapeType=Capsule, length=2, radius=2, shapeName=skirt_r_f_Master_capsule, rotation=<rot 90 0 0>}]
 thigh_l [{shapeType=Capsule, length=30.1198, radius=9.8802, center=<-20.872 0 1.2675>, rotation=<rot 89.3861 0.0037 0.0046>}]
 thigh_r [{shapeType=Capsule, length=34.4285, radius=8.395, center=<20.1098 -0.5427 -0.4663>, rotation=<rot 89.3861 0.0037 -179.9954>}]
}
constraints [
 {bone1Name=thigh_l, bone2Name=pelvis, swing1LimitDegrees=45, swing2LimitDegrees=45, twistLimitDegrees=45}
 {bone1Name=calf_l, bone2Name=thigh_l, swing1LimitDegrees=45, swing2LimitDegrees=45, twistLimitDegrees=45}
 {bone1Name=thigh_r, bone2Name=pelvis, swing1LimitDegrees=45, swing2LimitDegrees=45, twistLimitDegrees=45}
 {bone1Name=calf_r, bone2Name=thigh_r, swing1LimitDegrees=45, swing2LimitDegrees=45, twistLimitDegrees=45}
 {
  bone1Name=dyn_skirt_l_master
  bone2Name=pelvis
  swing1LimitDegrees=45
  swing1Motion=Free
  swing2LimitDegrees=45
  swing2Motion=Free
  twistLimitDegrees=45
  twistMotion=Free
 }
 {bone1Name=dyn_skirt_l_01, bone2Name=dyn_skirt_l_master, swing1LimitDegrees=2, swing2LimitDegrees=10, twistLimitDegrees=0}
 {bone1Name=dyn_skirt_l_02, bone2Name=dyn_skirt_l_01, swing1LimitDegrees=2.5, swing2LimitDegrees=10, twistLimitDegrees=0}
 {bone1Name=dyn_skirt_l_03, bone2Name=dyn_skirt_l_02, swing1LimitDegrees=2.5, swing2LimitDegrees=10, twistLimitDegrees=0}
 {bone1Name=dyn_skirt_l_04, bone2Name=dyn_skirt_l_03, swing1LimitDegrees=2.5, swing2LimitDegrees=10, twistLimitDegrees=0}
 {
  bone1Name=skirt_r_f_Master
  bone2Name=pelvis
  swing1LimitDegrees=45
  swing1Motion=Free
  swing2LimitDegrees=45
  swing2Motion=Free
  twistLimitDegrees=45
  twistMotion=Free
 }
 {
  bone1Name=dyn_skirt_r_master
  bone2Name=pelvis
  swing1LimitDegrees=45
  swing1Motion=Free
  swing2LimitDegrees=45
  swing2Motion=Free
  twistLimitDegrees=45
  twistMotion=Free
 }
 {bone1Name=dyn_skirt_r_01, bone2Name=dyn_skirt_r_master, swing1LimitDegrees=2, swing2LimitDegrees=5, twistLimitDegrees=0}
 {bone1Name=dyn_skirt_r_02, bone2Name=dyn_skirt_r_01, swing1LimitDegrees=2.5, swing2LimitDegrees=10, twistLimitDegrees=0}
 {bone1Name=dyn_skirt_r_03, bone2Name=dyn_skirt_r_02, swing1LimitDegrees=2.5, swing2LimitDegrees=10, twistLimitDegrees=0}
 {bone1Name=dyn_skirt_r_04, bone2Name=dyn_skirt_r_03, swing1LimitDegrees=2.5, swing2LimitDegrees=10, twistLimitDegrees=0}
 {bone1Name=dyn_skirt_r_f_01, bone2Name=skirt_r_f_Master, swing1LimitDegrees=2, swing2LimitDegrees=5, twistLimitDegrees=0}
 {
  bone1Name=skirt_L_b_Master
  bone2Name=pelvis
  swing1LimitDegrees=45
  swing1Motion=Free
  swing2LimitDegrees=45
  swing2Motion=Free
  twistLimitDegrees=45
  twistMotion=Free
 }
 {bone1Name=dyn_skirt_l_b_01, bone2Name=skirt_L_b_Master, swing1LimitDegrees=2, swing2LimitDegrees=10, twistLimitDegrees=0}
 {bone1Name=dyn_skirt_l_b_02, bone2Name=dyn_skirt_l_b_01, swing1LimitDegrees=2.5, swing2LimitDegrees=20, twistLimitDegrees=0}
 {bone1Name=dyn_skirt_l_b_03, bone2Name=dyn_skirt_l_b_02, swing1LimitDegrees=2.5, swing2LimitDegrees=20, twistLimitDegrees=0}
 {bone1Name=dyn_skirt_l_b_04, bone2Name=dyn_skirt_l_b_03, swing1LimitDegrees=2.5, swing2LimitDegrees=20, twistLimitDegrees=0}
 {bone1Name=dyn_skirt_r_f_02, bone2Name=dyn_skirt_r_f_01, swing1LimitDegrees=2.5, swing2LimitDegrees=10, twistLimitDegrees=0}
 {
  bone1Name=skirt_L_f_Master
  bone2Name=pelvis
  swing1LimitDegrees=45
  swing1Motion=Free
  swing2LimitDegrees=45
  swing2Motion=Free
  twistLimitDegrees=45
  twistMotion=Free
 }
 {bone1Name=dyn_skirt_l_f1, bone2Name=skirt_L_f_Master, swing1LimitDegrees=2.5, swing2LimitDegrees=15, twistLimitDegrees=0}
 {bone1Name=dyn_skirt_l_f2, bone2Name=dyn_skirt_l_f1, swing1LimitDegrees=2.5, swing2LimitDegrees=5, twistLimitDegrees=0}
 {bone1Name=dyn_skirt_l_f3, bone2Name=dyn_skirt_l_f2, swing1LimitDegrees=2.5, swing2LimitDegrees=10, twistLimitDegrees=0}
 {bone1Name=dyn_skirt_l_f4, bone2Name=dyn_skirt_l_f3, swing1LimitDegrees=2.5, swing2LimitDegrees=10, twistLimitDegrees=0}
 {bone1Name=dyn_skirt_r_f_03, bone2Name=dyn_skirt_r_f_02, swing1LimitDegrees=2.5, swing2LimitDegrees=10, twistLimitDegrees=0}
 {
  bone1Name=skirt_b_m_Master
  bone2Name=pelvis
  swing1LimitDegrees=45
  swing1Motion=Free
  swing2LimitDegrees=45
  swing2Motion=Free
  twistLimitDegrees=45
  twistMotion=Free
 }
 {bone1Name=dyn_skirt_b_01, bone2Name=skirt_b_m_Master, swing1LimitDegrees=2, swing2LimitDegrees=10, twistLimitDegrees=0}
 {bone1Name=dyn_skirt_b_02, bone2Name=dyn_skirt_b_01, swing1LimitDegrees=2.5, swing2LimitDegrees=20, twistLimitDegrees=0}
 {bone1Name=dyn_skirt_b_03, bone2Name=dyn_skirt_b_02, swing1LimitDegrees=2.5, swing2LimitDegrees=20, twistLimitDegrees=0}
 {bone1Name=dyn_skirt_b_04, bone2Name=dyn_skirt_b_03, swing1LimitDegrees=2.5, swing2LimitDegrees=20, twistLimitDegrees=0}
 {bone1Name=dyn_skirt_r_f_04, bone2Name=dyn_skirt_r_f_03, swing1LimitDegrees=2.5, swing2LimitDegrees=10, twistLimitDegrees=0}
 {
  bone1Name=skirt_r_b_Master
  bone2Name=pelvis
  swing1LimitDegrees=45
  swing1Motion=Free
  swing2LimitDegrees=45
  swing2Motion=Free
  twistLimitDegrees=45
  twistMotion=Free
 }
 {bone1Name=dyn_skirt_r_b_01, bone2Name=skirt_r_b_Master, swing1LimitDegrees=2, swing2LimitDegrees=10, twistLimitDegrees=0}
 {bone1Name=dyn_skirt_r_b_02, bone2Name=dyn_skirt_r_b_01, swing1LimitDegrees=2.5, swing2LimitDegrees=20, twistLimitDegrees=0}
 {bone1Name=dyn_skirt_r_b_03, bone2Name=dyn_skirt_r_b_02, swing1LimitDegrees=2.5, swing2LimitDegrees=20, twistLimitDegrees=0}
 {bone1Name=dyn_skirt_r_b_04, bone2Name=dyn_skirt_r_b_03, swing1LimitDegrees=2.5, swing2LimitDegrees=20, twistLimitDegrees=0}
 {
  bone1Name=dyn_chain_master
  bone2Name=pelvis
  swing1LimitDegrees=45
  swing1Motion=Free
  swing2LimitDegrees=45
  swing2Motion=Free
  twistLimitDegrees=45
  twistMotion=Free
 }
 {bone1Name=dyn_chain_01, bone2Name=dyn_chain_master, swing1LimitDegrees=2, swing2LimitDegrees=10, twistLimitDegrees=0}
 {bone1Name=dyn_chain_02, bone2Name=dyn_chain_01, swing1LimitDegrees=2, swing2LimitDegrees=10, twistLimitDegrees=0}
 {
  bone1Name=dyn_skirt_r_04
  bone2Name=dyn_skirt_r_f_04
  swing1LimitDegrees=45
  swing1Motion=Free
  swing2LimitDegrees=45
  swing2Motion=Free
  twistLimitDegrees=45
  twistMotion=Free
 }
]
