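# Physics asset for the BurnLeer medium female player body, providing secondary motion for its skirt and a short chain.
# Pelvis, thigh and calf capsules form the body volume around seven four-bone skirt strands and a two-link chain.
# Each strand hangs from a master bone with free angular motion on the pelvis.
# Links below the masters allow very narrow swing on one axis, wider on the other, with zero-degree twist limits.
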PhysicsAsset /BRCosmetics/Characters/Player/Female/Medium/Bodies/F_MED_BurnLeer/Meshes/F_MED_BurnLeer_Physics
bodies {
 calf_l [{shapeType=Capsule, length=35, radius=8, center=<-22.9034 -1.8748 0>, rotation=<rot 90 3.1447 0>}]
 calf_r [{shapeType=Capsule, length=39.9629, radius=8, center=<22.9034 1.8748 0>, rotation=<rot 90 -176.8553 0>}]
 dyn_chain_01 [{shapeType=Capsule, length=5, radius=5.3, shapeName=dyn_chain_01_capsule, center=<5 0 2.3>, rotation=<rot 90 90 0>}]
 dyn_chain_02 [{shapeType=Capsule, length=5, radius=6, shapeName=dyn_chain_02_capsule, center=<5 0 4>, rotation=<rot 90 90 0>}]
 dyn_chain_master [{shapeType=Capsule, length=0.505, radius=0.505, shapeName=dyn_chain_master_capsule, rotation=<rot 0 -90 0>}]
 dyn_skirt_b_01 [{shapeType=Capsule, length=5.2388, radius=4, shapeName=dyn_skirt_b_01_capsule, center=<5 0 5>, rotation=<rot 83.3587 180 180>}]
 dyn_skirt_b_02 [{shapeType=Capsule, length=10, radius=6, shapeName=dyn_skirt_b_02_capsule, center=<5 0 3.5>, rotation=<rot 90 90 0>}]
 dyn_skirt_b_03 [{shapeType=Capsule, length=12, radius=7, shapeName=dyn_skirt_b_03_capsule, center=<5 0 4.3>, rotation=<rot 90 90 0>}]
 dyn_skirt_b_04 [{shapeType=Capsule, length=15, radius=8, shapeName=dyn_skirt_b_04_capsule, center=<5 0 5>, rotation=<rot 90 90 0>}]
 dyn_skirt_l_01 [{shapeType=Capsule, length=4.6689, radius=4, shapeName=dyn_skirt_l_01_capsule, center=<5 0 4.5>, rotation=<rot 90 0 0>}]
 dyn_skirt_l_02 [{shapeType=Capsule, length=10, radius=6, shapeName=dyn_skirt_l_02_capsule, center=<5 0 5.5>, rotation=<rot 90 90 0>}]
 dyn_skirt_l_03 [{shapeType=Capsule, length=15, radius=7, shapeName=dyn_skirt_l_03_capsule, center=<5 0 6>, rotation=<rot 90 90 0>}]
 dyn_skirt_l_04 [{shapeType=Capsule, length=20, radius=8, shapeName=dyn_skirt_l_04_capsule, center=<5 0 7>, rotation=<rot 90 90 0>}]
 dyn_skirt_l_b_01 [{shapeType=Capsule, length=5.1308, radius=4, shapeName=dyn_skirt_l_b_01_capsule, center=<5 0 5.7>, rotation=<rot 90 0 0>}]
 dyn_skirt_l_b_02 [{shapeType=Capsule, length=10, radius=6, shapeName=dyn_skirt_l_b_02_capsule, center=<5 0 5>, rotation=<rot 90 90 0>}]
 dyn_skirt_l_b_03 [{shapeType=Capsule, length=15, radius=7.5, shapeName=dyn_skirt_l_b_03_capsule, center=<5 0 5.5>, rotation=<rot 90 90 0>}]
 dyn_skirt_l_b_04 [{shapeType=Capsule, length=20, radius=8.5, shapeName=dyn_skirt_l_b_04_capsule, center=<5 0 6>, rotation=<rot 90 90 0>}]
 dyn_skirt_l_f1 [{shapeType=Capsule, length=6, radius=5, shapeName=dyn_skirt_l_f1_capsule, center=<5 0 3>, rotation=<rot 90 -11.846 0>}]
 dyn_skirt_l_f2 [{shapeType=Capsule, length=8, radius=5, shapeName=dyn_skirt_l_f2_capsule, center=<4.8699 0 4>, rotation=<rot 80.25 -13 -0.0001>}]
 dyn_skirt_l_f3 [{shapeType=Capsule, length=8, radius=5, shapeName=dyn_skirt_l_f3_capsule, center=<5.0912 -4 1>, rotation=<rot 76.8768 -61.1923 -50.1762>}]
 dyn_skirt_l_f4 [{shapeType=Capsule, length=8, radius=5, shapeName=dyn_skirt_l_f4_capsule, center=<4.7724 -6 -1>, rotation=<rot 73.7276 -55.6013 -46.7696>}]
 dyn_skirt_l_master [{shapeType=Capsule, length=2, radius=2, shapeName=dyn_skirt_l_master_capsule, rotation=<rot 90 0 0>}]
 dyn_skirt_r_01 [{shapeType=Capsule, length=4.7481, radius=4, shapeName=dyn_skirt_r_01_capsule, center=<5 0 3>, rotation=<rot 90 0 0>}]
 dyn_skirt_r_02 [{shapeType=Capsule, length=10, radius=6, shapeName=dyn_skirt_r_02_capsule, center=<5 0 5>, rotation=<rot 90 90 0>}]
 dyn_skirt_r_03 [{shapeType=Capsule, length=15, radius=6, shapeName=dyn_skirt_r_03_capsule, center=<5 0 5>, rotation=<rot 90 90 0>}]
 dyn_skirt_r_04 [{shapeType=Capsule, length=15, radius=6, shapeName=dyn_skirt_r_04_capsule, center=<5 0 5>, rotation=<rot 90 90 0>}]
 dyn_skirt_r_b_01 [{shapeType=Capsule, length=6.0265, radius=4, shapeName=dyn_skirt_r_b_01_capsule, center=<5 0 6>, rotation=<rot 82.7483 180 -180>}]
 dyn_skirt_r_b_02 [{shapeType=Capsule, length=10, radius=6, shapeName=dyn_skirt_r_b_02_capsule, center=<5 0 6>, rotation=<rot 90 90 0>}]
 dyn_skirt_r_b_03 [{shapeType=Capsule, length=15, radius=7.5, shapeName=dyn_skirt_r_b_03_capsule, center=<5 0 7>, rotation=<rot 90 90 0>}]
 dyn_skirt_r_b_04 [{shapeType=Capsule, length=20, radius=8.5, shapeName=dyn_skirt_r_b_04_capsule, center=<5 0 7.5>, rotation=<rot 90 90 0>}]
 dyn_skirt_r_f_01 [{shapeType=Capsule, length=4.0762, radius=4, shapeName=dyn_skirt_r_f_01_capsule, center=<5 0 3>, rotation=<rot 84.4687 179.9999 179.9999>}]
 dyn_skirt_r_f_02 [{shapeType=Capsule, length=8, radius=5.5, shapeName=dyn_skirt_r_f_02_capsule, center=<4.8639 1.3 4.5>, rotation=<rot 75.3526 54.6848 38.7237>}]
 dyn_skirt_r_f_03 [{shapeType=Capsule, length=8, radius=5.5, shapeName=dyn_skirt_r_f_03_capsule, center=<5.04 4.6 2>, rotation=<rot 74.1082 69.4176 54.9151>}]
 dyn_skirt_r_f_04 [{shapeType=Capsule, length=8, radius=5.5, shapeName=dyn_skirt_r_f_04_capsule, center=<4.826 6.4 0>, rotation=<rot 59.8179 72.4294 63.5583>}]
 dyn_skirt_r_master [{shapeType=Capsule, length=2, radius=2, shapeName=dyn_skirt_r_master_capsule, rotation=<rot 90 0 0>}]
 pelvis [{shapeType=Capsule, length=8.812, radius=11.9082, center=<-5.2761 0 0>}]
 skirt_L_b_Master [{shapeType=Capsule, length=2, radius=2, shapeName=skirt_L_b_Master_capsule, rotation=<rot 90 0 0>}]
 skirt_L_f_Master [{shapeType=Capsule, length=2, radius=2, shapeName=skirt_L_f_Master_capsule, rotation=<rot 90 0 0>}]
 skirt_b_m_Master [{shapeType=Capsule, length=2, radius=2, shapeName=skirt_b_m_Master_capsule, rotation=<rot 90 0 0>}]
 skirt_r_b_Master [{shapeType=Capsule, length=2, radius=2, shapeName=skirt_r_b_Master_capsule, rotation=<rot 90 0 0>}]
 skirt_r_f_Master [{shapeType=Capsule, length=2, radius=2, shapeName=skirt_r_f_Master_capsule, rotation=<rot 90 0 0>}]
 thigh_l [{shapeType=Capsule, length=30, radius=8, center=<-20.8752 0 0.966>, rotation=<rot 89.3861 0.0037 0.0046>}]
 thigh_r [{shapeType=Capsule, length=32.0042, radius=8.2, center=<20.8752 0 -0.966>, rotation=<rot 89.3861 0.0037 -179.9954>}]
}
constraints [
 {bone1Name=thigh_l, bone2Name=pelvis, swing1LimitDegrees=45, swing2LimitDegrees=45, twistLimitDegrees=45}
 {bone1Name=calf_l, bone2Name=thigh_l, swing1LimitDegrees=45, swing2LimitDegrees=45, twistLimitDegrees=45}
 {bone1Name=thigh_r, bone2Name=pelvis, swing1LimitDegrees=45, swing2LimitDegrees=45, twistLimitDegrees=45}
 {bone1Name=calf_r, bone2Name=thigh_r, swing1LimitDegrees=45, swing2LimitDegrees=45, twistLimitDegrees=45}
 {
  bone1Name=dyn_skirt_l_master
  bone2Name=pelvis
  swing1LimitDegrees=45
  swing1Motion=Free
  swing2LimitDegrees=45
  swing2Motion=Free
  twistLimitDegrees=45
  twistMotion=Free
 }
 {bone1Name=dyn_skirt_l_01, bone2Name=dyn_skirt_l_master, swing1LimitDegrees=2, swing2LimitDegrees=10, twistLimitDegrees=0}
 {bone1Name=dyn_skirt_l_02, bone2Name=dyn_skirt_l_01, swing1LimitDegrees=2.5, swing2LimitDegrees=10, twistLimitDegrees=0}
 {bone1Name=dyn_skirt_l_03, bone2Name=dyn_skirt_l_02, swing1LimitDegrees=2.5, swing2LimitDegrees=10, twistLimitDegrees=0}
 {bone1Name=dyn_skirt_l_04, bone2Name=dyn_skirt_l_03, swing1LimitDegrees=2.5, swing2LimitDegrees=10, twistLimitDegrees=0}
 {
  bone1Name=skirt_r_f_Master
  bone2Name=pelvis
  swing1LimitDegrees=45
  swing1Motion=Free
  swing2LimitDegrees=45
  swing2Motion=Free
  twistLimitDegrees=45
  twistMotion=Free
 }
 {
  bone1Name=dyn_skirt_r_master
  bone2Name=pelvis
  swing1LimitDegrees=45
  swing1Motion=Free
  swing2LimitDegrees=45
  swing2Motion=Free
  twistLimitDegrees=45
  twistMotion=Free
 }
 {bone1Name=dyn_skirt_r_01, bone2Name=dyn_skirt_r_master, swing1LimitDegrees=2, swing2LimitDegrees=5, twistLimitDegrees=0}
 {bone1Name=dyn_skirt_r_02, bone2Name=dyn_skirt_r_01, swing1LimitDegrees=2.5, swing2LimitDegrees=10, twistLimitDegrees=0}
 {bone1Name=dyn_skirt_r_03, bone2Name=dyn_skirt_r_02, swing1LimitDegrees=2.5, swing2LimitDegrees=10, twistLimitDegrees=0}
 {bone1Name=dyn_skirt_r_04, bone2Name=dyn_skirt_r_03, swing1LimitDegrees=2.5, swing2LimitDegrees=10, twistLimitDegrees=0}
 {bone1Name=dyn_skirt_r_f_01, bone2Name=skirt_r_f_Master, swing1LimitDegrees=2, swing2LimitDegrees=5, twistLimitDegrees=0}
 {
  bone1Name=skirt_L_b_Master
  bone2Name=pelvis
  swing1LimitDegrees=45
  swing1Motion=Free
  swing2LimitDegrees=45
  swing2Motion=Free
  twistLimitDegrees=45
  twistMotion=Free
 }
 {bone1Name=dyn_skirt_l_b_01, bone2Name=skirt_L_b_Master, swing1LimitDegrees=2, swing2LimitDegrees=10, twistLimitDegrees=0}
 {bone1Name=dyn_skirt_l_b_02, bone2Name=dyn_skirt_l_b_01, swing1LimitDegrees=2.5, swing2LimitDegrees=20, twistLimitDegrees=0}
 {bone1Name=dyn_skirt_l_b_03, bone2Name=dyn_skirt_l_b_02, swing1LimitDegrees=2.5, swing2LimitDegrees=20, twistLimitDegrees=0}
 {bone1Name=dyn_skirt_l_b_04, bone2Name=dyn_skirt_l_b_03, swing1LimitDegrees=2.5, swing2LimitDegrees=20, twistLimitDegrees=0}
 {bone1Name=dyn_skirt_r_f_02, bone2Name=dyn_skirt_r_f_01, swing1LimitDegrees=2.5, swing2LimitDegrees=10, twistLimitDegrees=0}
 {
  bone1Name=skirt_L_f_Master
  bone2Name=pelvis
  swing1LimitDegrees=45
  swing1Motion=Free
  swing2LimitDegrees=45
  swing2Motion=Free
  twistLimitDegrees=45
  twistMotion=Free
 }
 {bone1Name=dyn_skirt_l_f1, bone2Name=skirt_L_f_Master, swing1LimitDegrees=2.5, swing2LimitDegrees=15, twistLimitDegrees=0}
 {bone1Name=dyn_skirt_l_f2, bone2Name=dyn_skirt_l_f1, swing1LimitDegrees=2.5, swing2LimitDegrees=5, twistLimitDegrees=0}
 {bone1Name=dyn_skirt_l_f3, bone2Name=dyn_skirt_l_f2, swing1LimitDegrees=2.5, swing2LimitDegrees=10, twistLimitDegrees=0}
 {bone1Name=dyn_skirt_l_f4, bone2Name=dyn_skirt_l_f3, swing1LimitDegrees=2.5, swing2LimitDegrees=10, twistLimitDegrees=0}
 {bone1Name=dyn_skirt_r_f_03, bone2Name=dyn_skirt_r_f_02, swing1LimitDegrees=2.5, swing2LimitDegrees=10, twistLimitDegrees=0}
 {
  bone1Name=skirt_b_m_Master
  bone2Name=pelvis
  swing1LimitDegrees=45
  swing1Motion=Free
  swing2LimitDegrees=45
  swing2Motion=Free
  twistLimitDegrees=45
  twistMotion=Free
 }
 {bone1Name=dyn_skirt_b_01, bone2Name=skirt_b_m_Master, swing1LimitDegrees=2, swing2LimitDegrees=10, twistLimitDegrees=0}
 {bone1Name=dyn_skirt_b_02, bone2Name=dyn_skirt_b_01, swing1LimitDegrees=2.5, swing2LimitDegrees=20, twistLimitDegrees=0}
 {bone1Name=dyn_skirt_b_03, bone2Name=dyn_skirt_b_02, swing1LimitDegrees=2.5, swing2LimitDegrees=20, twistLimitDegrees=0}
 {bone1Name=dyn_skirt_b_04, bone2Name=dyn_skirt_b_03, swing1LimitDegrees=2.5, swing2LimitDegrees=20, twistLimitDegrees=0}
 {bone1Name=dyn_skirt_r_f_04, bone2Name=dyn_skirt_r_f_03, swing1LimitDegrees=2.5, swing2LimitDegrees=10, twistLimitDegrees=0}
 {
  bone1Name=skirt_r_b_Master
  bone2Name=pelvis
  swing1LimitDegrees=45
  swing1Motion=Free
  swing2LimitDegrees=45
  swing2Motion=Free
  twistLimitDegrees=45
  twistMotion=Free
 }
 {bone1Name=dyn_skirt_r_b_01, bone2Name=skirt_r_b_Master, swing1LimitDegrees=2, swing2LimitDegrees=10, twistLimitDegrees=0}
 {bone1Name=dyn_skirt_r_b_02, bone2Name=dyn_skirt_r_b_01, swing1LimitDegrees=2.5, swing2LimitDegrees=20, twistLimitDegrees=0}
 {bone1Name=dyn_skirt_r_b_03, bone2Name=dyn_skirt_r_b_02, swing1LimitDegrees=2.5, swing2LimitDegrees=20, twistLimitDegrees=0}
 {bone1Name=dyn_skirt_r_b_04, bone2Name=dyn_skirt_r_b_03, swing1LimitDegrees=2.5, swing2LimitDegrees=20, twistLimitDegrees=0}
 {
  bone1Name=dyn_chain_master
  bone2Name=pelvis
  swing1LimitDegrees=45
  swing1Motion=Free
  swing2LimitDegrees=45
  swing2Motion=Free
  twistLimitDegrees=45
  twistMotion=Free
 }
 {bone1Name=dyn_chain_01, bone2Name=dyn_chain_master, swing1LimitDegrees=2, swing2LimitDegrees=10, twistLimitDegrees=0}
 {bone1Name=dyn_chain_02, bone2Name=dyn_chain_01, swing1LimitDegrees=2, swing2LimitDegrees=10, twistLimitDegrees=0}
]
