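# Animation Blueprint for the F_MED_BurnLeer cosmetic body, copying the attached parent's pose and adding secondary motion to clothing and accessories.
# A Control Rig feeds front-end or gameplay RigidBody simulation, faded out during knee and baseball slides.
# A post Control Rig, AnimDynamics for jacket, bag, zipper and collar, plus constraints and skirt offsets follow.
# The EventGraph converts movement states, specific emote montages and zipline events into rig inputs.
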AnimBlueprint /BRCosmetics/Characters/Player/Female/Medium/Bodies/F_MED_BurnLeer/Meshes/F_MED_BurnLeer_AnimBP
target_skeleton=&/F_MED_BurnLeer_Skeleton
graphs [
 AnimGraph {nodes [
   n1 = RigidBody!(
    ComponentPose=Local To Component(LocalPose=Use cached pose 'PreSim_1'())
    Alpha=Get DynAlpha() -> (DynAlpha:double)
    OverrideWorldGravity:Vector="0,0,-980"
    ExternalForce
    SimSpaceSettings:Sim Space Settings Structure=(WorldAlpha=1,VelocityScaleZ=1,DampingAlpha=1,MaxLinearVelocity=10000,MaxAngularVelocity=8,MaxLinearAcceleration=10000,MaxAngularAcceleration=40,ExternalLinearDragV=(X=1,Y=1,Z=1),ExternalLinearVelocity=(X=0,Y=0,Z=0),ExternalAngularVelocity=(X=1,Y=1,Z=1))
   ) -> (Pose) {
    bOverrideWorldGravity=true
    overridePhysicsAsset=&/F_MED_BurnLeer_FE_Physics
    simulationSpace=BaseBoneSpace
    baseBoneRef {boneName=pelvis}
    overrideWorldGravity {z=-980}
    simSpaceSettings {maxAngularAcceleration=40, maxAngularVelocity=8, worldAlpha=1, externalAngularVelocity=<1 1 1>, externalLinearDragV=<1 1 1>}
   }
   n2 = RigidBody(
    ComponentPose=Local To Component(LocalPose=Use cached pose 'PreSim_1'())
    Alpha=Get DynAlpha()
    OverrideWorldGravity="0,0,-980"
    SimSpaceSettings=(WorldAlpha=1,VelocityScaleZ=1,DampingAlpha=1,MaxLinearVelocity=10000,MaxAngularVelocity=8,MaxLinearAcceleration=10000,MaxAngularAcceleration=40,ExternalLinearDragV=(X=1,Y=1,Z=1),ExternalLinearVelocity=(X=0,Y=0,Z=0),ExternalAngularVelocity=(X=1,Y=1,Z=1))
   ) {
    bOverrideWorldGravity=true
    overridePhysicsAsset=&/F_MED_BurnLeer_Physics
    simulationSpace=BaseBoneSpace
    baseBoneRef {boneName=pelvis}
    overrideWorldGravity {z=-980}
    simSpaceSettings {maxAngularAcceleration=40, maxAngularVelocity=8, worldAlpha=1, externalAngularVelocity=<1 1 1>, externalLinearDragV=<1 1 1>}
   }
   n3 = Control Rig!(
    Source=Blend Poses by bool(BlendPose_0=Component To Local(ComponentPose=@n1), BlendPose_1=Component To Local(ComponentPose=@n2), bActiveValue=Get bIsInFrontEnd())
    Alpha
    SkydiveFloating:double=Get Skydive_floating() -> (Skydive_floating:double)
    running:double=Get 2D moving() -> (2D moving:double)
    Sprinting:double=Get Tactical Sprinting() -> (Tactical Sprinting:double)
    only crouch:double="Get Crouch "() -> ("Crouch ":double)
    jump:double=Get jump() -> (jump:double)
    FrontEnd:double=Get FrontEnd Float() -> (FrontEnd Float:double)
    BaseBallSprint:double=Get Baseball Sliding() -> (Baseball Sliding:double)
    Skydiving:double=Get Skydiving() -> (Skydiving:double)
    Emote:double=Get Emote() -> (Emote:double)
    CrouchForward:double=Get CrouchwalkMove forward() -> (CrouchwalkMove forward:double)
    Zipline:bool=Get OnZipline() -> (OnZipline:bool)
    Crouch:bool=Get bIsCrouching()
    yokequick_custom:double=Get yokequick_custom() -> (yokequick_custom:double)
    FE_Emote:double=Get FE_Emote() -> (FE_Emote:double)
   ) -> (Pose) {controlRigAssetReference {blueprintRigClass=&/F_MED_BurnLeer_CtrlRig_Post.F_MED_BurnLeer_CtrlRig_Post_C}}
   n4 = Anim Dynamics - Bone: dyn_jacketFront_l_01!(ComponentPose=Local To Component(LocalPose=@n3), Alpha=1, GravityOverride:Vector="0,0,-980", ExternalForce="0,0,0") -> (Pose) {
    angularDampingOverride=1
    angularSpringConstant=20000
    bAngularSpring=true
    bOverrideAngularDamping=true
    bUseGravityOverride=true
    gravityScale=0.35
    simulationSpace=World
    boundBone {boneName=dyn_jacketFront_l_01}
    componentAppliedLinearAccClamp=<5000 5000 5000>
    componentLinearAccScale=<1 1 1>
    componentLinearVelScale=<1 1 1>
    gravityOverride {z=-980}
    physicsBodyDefinitions [{
      boundBone {boneName=dyn_jacketFront_l_01}
      boxExtents=<20 25 25>
      constraintSetup {angularLimitsMax=<0 5 7>, angularLimitsMin=<0 -15 -7>, angularTarget=<1 0 0>}
      localJointOffset=<4 0 0>
     }]
   }
   n5 = Anim Dynamics - Bone: dyn_jacketFront_l_a_01!(ComponentPose=@n4, Alpha=1, GravityOverride:Vector="0,0,-980", ExternalForce="0,0,0") -> (Pose) {
    angularDampingOverride=1
    angularSpringConstant=20000
    bAngularSpring=true
    bOverrideAngularDamping=true
    bUseGravityOverride=true
    gravityScale=0.35
    simulationSpace=World
    boundBone {boneName=dyn_jacketFront_l_a_01}
    componentAppliedLinearAccClamp=<5000 5000 5000>
    componentLinearAccScale=<1 1 1>
    componentLinearVelScale=<1 1 1>
    gravityOverride {z=-980}
    physicsBodyDefinitions [{
      boundBone {boneName=dyn_jacketFront_l_a_01}
      boxExtents=<20 25 25>
      constraintSetup {angularLimitsMax=<0 5 7>, angularLimitsMin=<0 -15 -7>, angularTarget=<1 0 0>}
      localJointOffset=<4 0 0>
     }]
   }
   n6 = Anim Dynamics - Bone: dyn_jacketSide_l_01!(ComponentPose=@n5, Alpha=1, GravityOverride:Vector="0,0,-980", ExternalForce="0,0,0") -> (Pose) {
    angularDampingOverride=1
    angularSpringConstant=20000
    bAngularSpring=true
    bOverrideAngularDamping=true
    bUseGravityOverride=true
    gravityScale=0.35
    simulationSpace=World
    boundBone {boneName=dyn_jacketSide_l_01}
    componentAppliedLinearAccClamp=<5000 5000 5000>
    componentLinearAccScale=<1 1 1>
    componentLinearVelScale=<1 1 1>
    gravityOverride {z=-980}
    physicsBodyDefinitions [{
      boundBone {boneName=dyn_jacketSide_l_01}
      boxExtents=<20 25 25>
      constraintSetup {angularLimitsMax=<0 5 7>, angularLimitsMin=<0 -15 -7>, angularTarget=<1 0 0>}
      localJointOffset=<4 0 0>
     }]
   }
   n7 = Anim Dynamics - Bone: dyn_jacketback_l_b_01!(ComponentPose=@n6, Alpha=1, GravityOverride:Vector="0,0,-980", ExternalForce="0,0,0") -> (Pose) {
    angularDampingOverride=1
    angularSpringConstant=20000
    bAngularSpring=true
    bOverrideAngularDamping=true
    bUseGravityOverride=true
    gravityScale=0.35
    simulationSpace=RootRelative
    boundBone {boneName=dyn_jacketback_l_b_01}
    componentAppliedLinearAccClamp=<5000 5000 5000>
    componentLinearAccScale=<1 1 1>
    componentLinearVelScale=<1 1 1>
    gravityOverride {z=-980}
    physicsBodyDefinitions [{
      boundBone {boneName=dyn_jacketback_l_b_01}
      boxExtents=<20 25 25>
      constraintSetup {angularLimitsMax=<0 5 7>, angularLimitsMin=<0 -15 -7>, angularTarget=<1 0 0>}
      localJointOffset=<4 0 0>
     }]
   }
   n8 = Anim Dynamics - Bone: dyn_jacketBack_01!(ComponentPose=@n7, Alpha=1, GravityOverride:Vector="0,0,-980", ExternalForce="0,0,0") -> (Pose) {
    angularDampingOverride=1
    angularSpringConstant=20000
    bAngularSpring=true
    bOverrideAngularDamping=true
    bUseGravityOverride=true
    gravityScale=0.35
    simulationSpace=RootRelative
    boundBone {boneName=dyn_jacketBack_01}
    componentAppliedLinearAccClamp=<5000 5000 5000>
    componentLinearAccScale=<1 1 1>
    componentLinearVelScale=<1 1 1>
    gravityOverride {z=-980}
    physicsBodyDefinitions [{
      boundBone {boneName=dyn_jacketBack_01}
      boxExtents=<20 25 25>
      constraintSetup {angularLimitsMax=<0 5 7>, angularLimitsMin=<0 -15 -7>, angularTarget=<1 0 0>}
      localJointOffset=<4 0 0>
     }]
   }
   n9 = Anim Dynamics - Bone: dyn_jacketback_r_b_01!(ComponentPose=@n8, Alpha=1, GravityOverride:Vector="0,0,-980", ExternalForce="0,0,0") -> (Pose) {
    angularDampingOverride=1
    angularSpringConstant=20000
    bAngularSpring=true
    bOverrideAngularDamping=true
    bUseGravityOverride=true
    gravityScale=0.35
    simulationSpace=RootRelative
    boundBone {boneName=dyn_jacketback_r_b_01}
    componentAppliedLinearAccClamp=<5000 5000 5000>
    componentLinearAccScale=<1 1 1>
    componentLinearVelScale=<1 1 1>
    gravityOverride {z=-980}
    physicsBodyDefinitions [{
      boundBone {boneName=dyn_jacketback_r_b_01}
      boxExtents=<20 25 25>
      constraintSetup {angularLimitsMax=<0 5 7>, angularLimitsMin=<0 -15 -7>, angularTarget=<1 0 0>}
      localJointOffset=<4 0 0>
     }]
   }
   n10 = Anim Dynamics - Bone: dyn_jacketSide_r_01!(ComponentPose=@n9, Alpha=1, GravityOverride:Vector="0,0,-980", ExternalForce="0,0,0") -> (Pose) {
    angularDampingOverride=1
    angularSpringConstant=20000
    bAngularSpring=true
    bOverrideAngularDamping=true
    bUseGravityOverride=true
    gravityScale=0.35
    simulationSpace=RootRelative
    boundBone {boneName=dyn_jacketSide_r_01}
    componentAppliedLinearAccClamp=<5000 5000 5000>
    componentLinearAccScale=<1 1 1>
    componentLinearVelScale=<1 1 1>
    gravityOverride {z=-980}
    physicsBodyDefinitions [{
      boundBone {boneName=dyn_jacketSide_r_01}
      boxExtents=<20 25 25>
      constraintSetup {angularLimitsMax=<0 5 7>, angularLimitsMin=<0 -15 -7>, angularTarget=<1 0 0>}
      localJointOffset=<4 0 0>
     }]
   }
   n11 = Anim Dynamics - Bone: dyn_jacketFront_r_a_01!(ComponentPose=@n10, Alpha=1, GravityOverride:Vector="0,0,-980", ExternalForce="0,0,0") -> (Pose) {
    angularDampingOverride=1
    angularSpringConstant=20000
    bAngularSpring=true
    bOverrideAngularDamping=true
    bUseGravityOverride=true
    gravityScale=0.35
    simulationSpace=RootRelative
    boundBone {boneName=dyn_jacketFront_r_a_01}
    componentAppliedLinearAccClamp=<5000 5000 5000>
    componentLinearAccScale=<1 1 1>
    componentLinearVelScale=<1 1 1>
    gravityOverride {z=-980}
    physicsBodyDefinitions [{
      boundBone {boneName=dyn_jacketFront_r_a_01}
      boxExtents=<20 25 25>
      constraintSetup {angularLimitsMax=<0 5 7>, angularLimitsMin=<0 -15 -7>, angularTarget=<1 0 0>}
      localJointOffset=<4 0 0>
     }]
   }
   n12 = Anim Dynamics - Bone: dyn_jacketFront_r_01!(ComponentPose=@n11, Alpha=1, GravityOverride:Vector="0,0,-980", ExternalForce="0,0,0") -> (Pose) {
    angularDampingOverride=1
    angularSpringConstant=20000
    bAngularSpring=true
    bOverrideAngularDamping=true
    bUseGravityOverride=true
    gravityScale=0.35
    simulationSpace=RootRelative
    boundBone {boneName=dyn_jacketFront_r_01}
    componentAppliedLinearAccClamp=<5000 5000 5000>
    componentLinearAccScale=<1 1 1>
    componentLinearVelScale=<1 1 1>
    gravityOverride {z=-980}
    physicsBodyDefinitions [{
      boundBone {boneName=dyn_jacketFront_r_01}
      boxExtents=<20 25 25>
      constraintSetup {angularLimitsMax=<0 5 7>, angularLimitsMin=<0 -15 -7>, angularTarget=<1 0 0>}
      localJointOffset=<4 0 0>
     }]
   }
   n13 = Anim Dynamics - Bone: dyn_bag_01!(ComponentPose=@n12, Alpha=1, GravityOverride:Vector="0,0,-980", ExternalForce="0,0,0") -> (Pose) {
    angularDampingOverride=1
    angularSpringConstant=20000
    bAngularSpring=true
    bOverrideAngularDamping=true
    bUseGravityOverride=true
    gravityScale=0.35
    simulationSpace=RootRelative
    boundBone {boneName=dyn_bag_01}
    componentAppliedLinearAccClamp=<5000 5000 5000>
    componentLinearAccScale=<1 1 1>
    componentLinearVelScale=<1 1 1>
    gravityOverride {z=-980}
    physicsBodyDefinitions [{
      boundBone {boneName=dyn_bag_01}
      boxExtents=<20 30 30>
      constraintSetup {angularLimitsMax=<0 3 3>, angularLimitsMin=<0 -3 -3>, angularTarget=<1 0 0>}
      localJointOffset=<4 0 0>
     }]
   }
   n14 = "Constraint - dyn_zipper_01 (PPP)"!(ComponentPose=@n13, Alpha=1, ConstraintWeights_0=0.5046, ConstraintWeights_1:float=0.4841, ConstraintWeights_2:float=0.0113) -> (Pose) {
    boneToModify {boneName=dyn_zipper_01}
    constraintSetup [{targetBone {boneName=spine_05}}, {targetBone {boneName=spine_04}}, {targetBone {boneName=deform_pec_l}}]
    constraintWeights [1,1,1]
   }
   n15 = Anim Dynamics - Bone: dyn_zipper_01(ComponentPose=@n14, Alpha=1, ExternalForce="0,0,0") {
    angularDampingOverride=1
    angularSpringConstant=20000
    bAngularSpring=true
    bOverrideAngularDamping=true
    bUseGravityOverride=true
    gravityScale=0.35
    simulationSpace=RootRelative
    boundBone {boneName=dyn_zipper_01}
    componentLinearAccScale=<1 1 1>
    componentLinearVelScale=<1 1 1>
    gravityOverride {z=-980}
    physicsBodyDefinitions [{
      boundBone {boneName=dyn_zipper_01}
      boxExtents=<20 30 30>
      constraintSetup {angularLimitsMax=<0 1 0>, angularLimitsMin=<0 -50 0>, angularTarget=<1 0 0>}
      localJointOffset=<4 0 0>
     }]
   }
   n16 = Anim Dynamics - Bone: dyn_l_corler_master!(ComponentPose=@n15, Alpha=1, GravityOverride:Vector="0,0,-980", ExternalForce="0,0,0") -> (Pose) {
    angularDampingOverride=1
    angularSpringConstant=20000
    bAngularSpring=true
    bOverrideAngularDamping=true
    bUseGravityOverride=true
    gravityScale=0.35
    simulationSpace=RootRelative
    boundBone {boneName=dyn_l_corler_master}
    componentAppliedLinearAccClamp=<5000 5000 5000>
    componentLinearAccScale=<1 1 1>
    componentLinearVelScale=<1 1 1>
    gravityOverride {z=-980}
    physicsBodyDefinitions [{
      boundBone {boneName=dyn_l_corler_master}
      boxExtents=<20 30 30>
      constraintSetup {angularLimitsMax=<0 5 1>, angularLimitsMin=<0 -10 -1>, angularTarget=<1 0 0>}
      localJointOffset=<4 0 0>
     }]
   }
   n17 = Anim Dynamics - Bone: dyn_r_corler_master!(ComponentPose=@n16, Alpha=1, GravityOverride:Vector="0,0,-980", ExternalForce="0,0,0") -> (Pose) {
    angularDampingOverride=1
    angularSpringConstant=20000
    bAngularSpring=true
    bOverrideAngularDamping=true
    bUseGravityOverride=true
    gravityScale=0.35
    simulationSpace=RootRelative
    boundBone {boneName=dyn_r_corler_master}
    componentAppliedLinearAccClamp=<5000 5000 5000>
    componentLinearAccScale=<1 1 1>
    componentLinearVelScale=<1 1 1>
    gravityOverride {z=-980}
    physicsBodyDefinitions [{
      boundBone {boneName=dyn_r_corler_master}
      boxExtents=<20 30 30>
      constraintSetup {angularLimitsMax=<0 3 1>, angularLimitsMin=<0 -10 -1>, angularTarget=<1 0 0>}
      localJointOffset=<4 0 0>
     }]
   }
   n18 = "Constraint - fx_skirt_r_front (PPPP)"!(
    ComponentPose=@n17
    Alpha=1
    ConstraintWeights_0=0.4747
    ConstraintWeights_1:float=0.3826
    ConstraintWeights_2:float=0.079
    ConstraintWeights_3:float=0.0637
   ) -> (Pose) {
    boneToModify {boneName=fx_skirt_r_front}
    constraintSetup [
     {targetBone {boneName=dyn_skirt_r_f_04}}
     {targetBone {boneName=dyn_skirt_r_04}}
     {targetBone {boneName=dyn_skirt_r_f_03}}
     {targetBone {boneName=dyn_skirt_r_03}}
    ]
    constraintWeights [1,1,1,1]
   }
   n19 = "Transform (Modify) Bone - Bone: dyn_skirt_l_master"(ComponentPose=@n18, Alpha=1, Translation="-0.928,0.000,1.464", Rotation="0,0,0", Scale="1,1,1") {
    rotationMode=BMM_Additive
    rotationSpace=BCS_BoneSpace
    translationMode=BMM_Additive
    translationSpace=BCS_BoneSpace
    alphaBoolBlend {blendInTime=0.2, blendOutTime=0.2}
    boneToModify {boneName=dyn_skirt_l_master}
    translation {x=-0.928, z=1.464}
   }
   n20 = "Transform (Modify) Bone - Bone: skirt_L_b_Master"(ComponentPose=@n19, Alpha=1, Translation="0.000,0.000,0.598", Rotation="0,0,0", Scale="1,1,1") {
    rotationMode=BMM_Additive
    rotationSpace=BCS_BoneSpace
    translationMode=BMM_Additive
    translationSpace=BCS_BoneSpace
    alphaBoolBlend {blendInTime=0.2, blendOutTime=0.2}
    boneToModify {boneName=skirt_L_b_Master}
    translation {z=0.598}
   }
   Output Pose(Result=Component To Local(ComponentPose=@n20)) {layerGroup=None}
   n21 = Control Rig!(
    Source=Copy Pose From Mesh() {bUseAttachedParent=true}
    Alpha
    OnlyCrouching:double="Get Crouch "()
    CrouchWalk:double=Get Crouch Walk() -> (Crouch Walk:double)
    Slide:double=Get Crouch Slide() -> (Crouch Slide:double)
    jump=Get jump()
    Sprinting=Get Tactical Sprinting()
    MoveForward:double=Get moveForward() -> (moveForward:double)
    MoveBackward:double=Get moveBackward() -> (moveBackward:double)
    MoveRight:double=Get strafeRight float() -> (strafeRight float:double)
    MoveLeft:double=Get strafeLeft float() -> (strafeLeft float:double)
    Baseball Slide:double=Get Baseball Sliding()
    GoldenGloryEmote:bool=Get GoldenGloryEmote() -> (GoldenGloryEmote:bool)
    RBAN Enable:bool=Get bRigidBodyEnabled()
   ) -> (Pose) {controlRigAssetReference {blueprintRigClass=&/F_MED_BurnLeer_CtrlRig.F_MED_BurnLeer_CtrlRig_C}}
   Save cached pose 'PreSim_1'(Pose=@n21) {cacheName=PreSim_1}
  ]}
 EventGraph {nodes [
   Event BlueprintUpdateAnimation()
   > n1 = Sequence!(execute) -> (then_0, then_1, then_2:Exec, then_3:Exec, then_4:Exec, then_5:Exec)
   n2 = NOT Boolean(A=Get bIsCrouching())
   n3 = NOT Boolean(A=Get bIsSkydiveFloating())
   n4 = Get bIsMoving2D()
   Set 2D moving(execute=@n1.then_3, 2D moving:double="To Float (Boolean)"(InBool=AND Boolean!(A=@n4, B=@n2, C:bool=@n3) -> (ReturnValue))) -> (Output_Get:double)
   n5 = NOT Boolean(A=Get bIsCrouching())
   n6 = Get bIsMoving2D()
   n7 = Get bIsTacticalSprinting()
   > Set Tactical Sprinting(Tactical Sprinting:double="To Float (Boolean)"(InBool=AND Boolean(A=@n6, B=@n7, C=@n5))) -> (Output_Get:double)
   > Set Skydive_floating(Skydive_floating:double="To Float (Boolean)"(InBool=Get bIsSkydiveFloating())) -> (Output_Get:double)
   > n8 = Branch(Condition=OR Boolean(A=Get bIsKneeSliding(), B=Get bIsBaseballSliding()))
   > Set DynAlpha(DynAlpha:double=0) -> (Output_Get:double)
   Set DynAlpha(execute=@n8.else, DynAlpha=1)
   n9 = NOT Boolean(A=Get bIsSkydiveFloating())
   n10 = Get bIsSkyDiving()
   Set Skydiving(execute=@n1.then_4, Skydiving:double="To Float (Boolean)"(InBool=AND Boolean!(A=@n10, B=@n9) -> (ReturnValue))) -> (Output_Get:double)
   > Set Emote(Emote:double="To Float (Boolean)"(InBool=Get bIsPlayingEmote())) -> (Output_Get:double)
   > Set CrouchwalkMove forward(CrouchwalkMove forward:double=float + float(A=Get moveForward(), B="Get Crouch "())) -> (Output_Get:double)
   n11 = NOT Boolean(A=Get bIsPlayingEmote())
   n12 = Get bIsInFrontEnd()
   > Set FE_Emote(FE_Emote:double="To Float (Boolean)"(InBool=AND Boolean(A=@n11, B=@n12))) -> (Output_Get:double)
   n13 = GetAnimInstance(self=GetOwnerSkeletalMeshComponent())
   Set yokequick_custom(
    execute=@n1.then_5
    yokequick_custom:double="To Float (Boolean)"(InBool=Montage_IsPlaying(self=@n13, Montage=/BRCosmetics/Animation/Game/MainPlayer/Emotes/YokeQuick/CMF/Emote_YokeQuick_CMF_Joining_M))
   ) -> (Output_Get:double)
   n14 = Get bRigidBodyEnabled()
   n15 = OR Boolean(
    A=Montage_IsPlaying(self=@n13, Montage=/BRCosmetics/Animation/Game/MainPlayer/Emotes/GoldenGlory/CMF/Emote_GoldenGlory_CMF_M)
    B=Montage_IsPlaying(self=@n13, Montage=/BRCosmetics/Animation/Game/MainPlayer/Emotes/GoldenGlory/CMF/Emote_GoldenGlory_CMF_FrontEnd_M)
   )
   > Set GoldenGloryEmote(GoldenGloryEmote:bool=AND Boolean(A=@n15, B=NOT Boolean(A=@n14))) -> (Output_Get:bool)
   n16 = Get bRigidBodyEnabled()
   > Set GoldenGlory(GoldenGlory:double="To Float (Boolean)"(InBool=AND Boolean(A=@n15, B=@n16))) -> (Output_Get:double)
   n17 = "Gravity Override State Function (Parameter Asset) (Nativized)"(
    execute=@n1.then_0
    CharacterPartAnimInstance=Self-Reference()
    GameTimeInSeconds=GetGameTimeInSeconds()
    Parameters=&/F_MED_BurnLeer_Parameters
   )
   > Set GravityOverride(GravityOverride=@n17)
   > Set RBAN Settings Out(RBAN Settings Out=@n17.OutRigidBodyAnimNodeParameters)
   n18 = NOT Boolean(A=Get bIsMoving2D())
   n19 = Get bIsCrouching()
   "Set Crouch "(execute=@n1.then_2, "Crouch ":double="To Float (Boolean)"(InBool=AND Boolean(A=@n19, B=@n18))) -> (Output_Get:double)
   n20 = Get bIsCrouching()
   n21 = Get bIsMoving2D()
   > Set Crouch Walk(Crouch Walk:double="To Float (Boolean)"(InBool=AND Boolean(A=@n20, B=@n21))) -> (Output_Get:double)
   n22 = Get bIsCrouching()
   n23 = Get bIsSliding()
   > Set Crouch Slide(Crouch Slide:double="To Float (Boolean)"(InBool=AND Boolean(A=@n22, B=@n23))) -> (Output_Get:double)
   n24 = Get bIsJumping()
   n25 = Get bIsFalling()
   > Set jump(jump:double="To Float (Boolean)"(InBool=OR Boolean(A=@n24, B=@n25))) -> (Output_Get:double)
   n26 = Get bIsBaseballSliding()
   n27 = Get bIsCrouching()
   > Set Baseball Sliding(Baseball Sliding:double="To Float (Boolean)"(InBool=AND Boolean(A=@n27, B=@n26))) -> (Output_Get:double)
   > Set FrontEnd Float(FrontEnd Float:double="To Float (Boolean)"(InBool=Get bIsInFrontEnd())) -> (Output_Get:double)
   n28 = PlayerMovement(execute=@n1.then_1, AnimInstance=Self-Reference(), XDotProductCoefficient=1, YDotProductCoefficient=1, ZDotProductCoefficient=1)
   n29 = BreakVector(InVec=MakeVector(X=@n28, Y=@n28.YMovement))
   > Set strafeLeft float(strafeLeft float:double="Absolute (Float)"(A="Clamp (Float)"(Value=@n29, Min=-1, Max=0))) -> (Output_Get:double)
   > Set strafeRight float(strafeRight float:double="Clamp (Float)"(Value=@n29, Min=0, Max=1)) -> (Output_Get:double)
   > Set moveForward(moveForward:double="Clamp (Float)"(Value=@n29.Y, Min=0, Max=1)) -> (Output_Get:double)
   > Set moveBackward(moveBackward:double="Absolute (Float)"(A="Clamp (Float)"(Value=@n29.Y, Min=-1, Max=0))) -> (Output_Get:double)
   n30 = Event BlueprintProcessAnimInputEvent()
   > n31 = Branch(Condition=IsNamed(self=@n30.AnimInputEvent, InEventName=aie.Player.Zipline.Entered))
   > Set OnZipline(OnZipline:bool=1) -> (Output_Get:bool)
   Branch(execute=@n31.else, Condition=IsNamed(self=@n30.AnimInputEvent, InEventName=aie.Player.Zipline.Exited))
   > Set OnZipline(OnZipline=false)
  ]}
]
variables [
 2D moving {type=number}
 Baseball Sliding {type=number}
 "Crouch " {type=number}
 Crouch Slide {type=number}
 Crouch Walk {type=number}
 CrouchwalkMove forward {type=number}
 DynAlpha {type=number}
 Emote {type=number}
 FE_Emote {type=number}
 FrontEnd Float {type=number}
 GoldenGlory {type=number}
 GoldenGloryEmote {type=boolean}
 GravityOverride {type=Vector}
 jump {type=number}
 KneeSlide {type=number}
 moveBackward {type=number}
 moveForward {type=number}
 OnZipline {type=boolean}
 RBAN Settings Out {type=FortRigidBodyAnimNodeParameters}
 Skydive_floating {type=number}
 Skydiving {type=number}
 strafeLeft float {type=number}
 strafeRight float {type=number}
 Tactical Sprinting {type=number}
 yokequick_custom {type=number}
]
